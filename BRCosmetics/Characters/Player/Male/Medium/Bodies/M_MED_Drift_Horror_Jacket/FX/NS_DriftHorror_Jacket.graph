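# Niagara cosmetic effect for the Drift Horror jacket body that flickers short magenta lightning across the character's back, body and cape edge.
# Paired emitters per region fire randomized, looping bursts of mesh particles with attached lights.
# Particles track User.SkeletalMesh each update, mostly blending between neighboring bone or socket positions, with a local four-bone helper for the lower back.
# Random scaling, orientation and sine-driven alpha create the crackle.
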NiagaraSystem /BRCosmetics/Characters/Player/Male/Medium/Bodies/M_MED_Drift_Horror_Jacket/FX/NS_DriftHorror_Jacket
dependencies {
 usedDataInterfaces [
  &NiagaraDataInterfaceCurve
  &NiagaraDataInterfaceArrayInt32
  &NiagaraDataInterfaceColorCurve
  &NiagaraDataInterfaceVectorCurve
  &NiagaraDataInterfaceVector2DCurve
  &NiagaraDataInterfaceArrayFloat
  &NiagaraDataInterfaceArrayFloat2
  &NiagaraDataInterfaceArrayFloat3
  &NiagaraDataInterfaceArrayColor
 ]
 usedDynamicInputs [&Abs_Float, &Boolean_OrOperation, &Boolean_AndOperation]
 usedModules [&SystemState, &Timeline, &InterpolateOverTime]
}
emitters {
 Lightning_Back_Lower {
  emitterUpdateScript [
   &EmitterState(
    Life Cycle Mode=::NewEnumerator1
    Loop Duration=&RandomRangeFloat(Minimum=0.1, Maximum=1.6)
    Recalculate Duration Each Loop=true
    UseLoopDelay=false
   )
   &SpawnBurst_Instantaneous(
    Spawn Count=&MakeCustomIntFromBool(
     Bool=&SetBoolByFloatComparison(A=&RandomRangeFloat(Minimum=0, Maximum=0.5), B=0.25)
     True INT=1
     False INT=2
    )
    Spawn Probability=0.5
    Use Spawn Probability=true
    UseLoopCountLimit=false
   )
   &SpawnRate(SpawnRate=30, Use Spawn Probability=false) {enabled=false}
  ]
  particleSpawnScript [
   &InitializeParticle(
    Color=<rgba 2 0 1.5333 1>
    Lifetime Max=0.3
    Lifetime Min=0.25
    Lifetime Mode=::NewEnumerator1
    Mesh Scale=&Multiply_VectorByFloat(
     Vector=&RandomRangeVector(Minimum=<1 0.5 0.5>, Maximum=<0.5 0.5 0.5>)
     Float=&FloatFromCurve(FloatCurve=DataInterface() {propertyValues {Curve {keys [{arriveTangent=-3.9075, interpMode=RCIM_Cubic, leaveTangent=-3.9075, tangentMode=RCTM_User, time=0, value=1}, <at 1 0 RCIM_Cubic>]}}})
    )
    Mesh Scale Mode=::NewEnumerator4
    Position Mode=::NewEnumerator1
    Sprite Size Mode=::NewEnumerator3
    Uniform Sprite Size=2
    UsePositionOffset=false
    Write Material Random=true
   )
   /SetVariables(Particles.Pos Choice:NiagaraInt32=&Multiply_Int(A=&RandomRangeInt(Minimum=6, Maximum=10), B=1)) {isSetParameters=true, module=SetVariables_#1}
   &InitialMeshOrientation(
    Mesh Orientation Mode=::NewEnumerator7
    Rotation=&RandomRangeVector(Minimum=<0.25 0.25 -0.25>, Maximum=<-0.25 -0.25 -0.25>)
    Use Rotation Vector=true
   )
   &InitialMeshOrientation(
    Mesh Orientation Mode=::NewEnumerator7
    Rotation=&RandomRangeVector(Minimum=<0 0 0>, Maximum=<0.25 0 0>)
    Use Rotation Vector=true
   ) {enabled=false, module=InitialMeshOrientation001}
  ]
  particleUpdateScript [
   &ParticleState()
   &self:Drift_Horror_Get_Set_of_4_Bones(New Skeletal Mesh:NiagaraDataInterfaceSkeletalMesh=$User.SkeletalMesh, StartIndex:NiagaraInt32=$Particles.Initial.Pos Choice) {module=Drift_Horror_Get_Set_of_4_Bones}
   /Niagara/Modules/Update/Orientation/V2/OrientMeshToVector(
    Look At Direction:Vector3f=&Subtract_Vector(A=$Output.Drift_Horror_Get_Set_of_4_Bones.BonePos2, B=$Output.Drift_Horror_Get_Set_of_4_Bones.BonePos1)
    Override Quat=false
    Up Axis Direction:Vector3f=<0 0 1>
    UseOptionalSideAxis=false
    UseOptionalUpAxis=false
   ) {enabled=false}
   /SetVariables(
    Particles.Pos A:Vector3f=&Lerp_Vector(
     A=$Output.Drift_Horror_Get_Set_of_4_Bones.BonePos1
     B=$Output.Drift_Horror_Get_Set_of_4_Bones.BonePos2
     Alpha=$Particles.Initial.MaterialRandom
    )
    Particles.Pos B:Vector3f=&Lerp_Vector(
     A=$Output.Drift_Horror_Get_Set_of_4_Bones.BonePos3
     B=$Output.Drift_Horror_Get_Set_of_4_Bones.BonePos4
     Alpha=$Particles.Initial.MaterialRandom
    )
   ) {isSetParameters=true, module=SetVariables_#2}
   /SetVariables(Particles.Position=&Lerp_Vector(A=$Particles.Pos A, B=$Particles.Pos B, Alpha=0.5)) {isSetParameters=true, module=SetVariables_#3}
   /SetVariables(Particles.MeshOrientation=&LerpQuaternion(
     Quaternion A:Quat4f=$Particles.SkeletalMeshLocation.SampledRotation
     Quaternion B:Quat4f=$Particles.SkeletalMeshLocation001.SampledRotation
     Lerp Factor:NiagaraFloat=$Particles.Initial.MaterialRandom
     Enable Spherical Linear Interpolation:NiagaraBool=true
    )) {enabled=false, isSetParameters=true, module=SetVariables_#4}
   &DynamicMaterialParameters(
    Index 0 Param 1=1
    Index 0 Param 3=-0.1
    Module.Param0WriteEnabled=true
    Module.Param1WriteEnabled=true
    Module.Param2WriteEnabled=true
    Module.Param3WriteEnabled=true
   )
   &ScaleMeshSize(Scale Factor=&MakeVector(
     X=&FloatFromCurve(FloatCurve=DataInterface() {propertyValues {Curve {keys [{tangentMode=RCTM_User, time=0, value=0.5}, {tangentMode=RCTM_User, time=0.1, value=1}, <at 1 1.1>]}}})
     Y=&Lerp_Float(A=1, B=&RandomRangeFloat(Minimum=1.2, Maximum=2), Alpha=$Particles.NormalizedAge)
     Z=&Lerp_Float(A=1, B=&RandomRangeFloat(Minimum=1.2, Maximum=2), Alpha=$Particles.NormalizedAge)
    ))
   &FN_Light_Attributes(
    Color=<rgba 10 2.5 8.1923 1>
    Create Color Attribute=true
    Create Exponent Attribute=true
    Create Position Attribute=true
    Exponent=1
    Position=&Add_Vector(
     A=Linked() {linkedVariable {name=Particles.Position, type=NiagaraPosition}}
     B=&Multiply_VectorByFloat(
      Vector=&ReturnMeshOrientationAxis(Mesh Orientation Axis=::XAxis, Mesh Orientation Quat=$Particles.MeshOrientation)
      Float=10
     )
    )
   )
   &ScaleColor(
    Scale Alpha=&Multiply_Float(
     A=&FloatFromCurve(FloatCurve=DataInterface() {propertyValues {Curve {keys [
          {
           arriveTangentWeight=0.9771
           interpMode=RCIM_Cubic
           leaveTangentWeight=0.9771
           tangentMode=RCTM_User
           tangentWeightMode=RCTWM_WeightedBoth
           time=0
           value=1
          }
          {
           arriveTangent=-4.27
           arriveTangentWeight=0.8915
           interpMode=RCIM_Cubic
           leaveTangent=-4.27
           leaveTangentWeight=0.8915
           tangentMode=RCTM_User
           tangentWeightMode=RCTWM_WeightedBoth
           time=1
           value=0
          }
         ]}}})
     B=&Sine(
      Normalized Angle=$Emitter.Age
      Period=&RandomRangeFloat(Minimum=0.05, Maximum=0.2)
      Scale=2
      Bias=&RandomRangeFloat(Minimum=0.5, Maximum=5)
     )
    )
    Scale RGB=&VectorFromFloat(Value=&Sine(
      Normalized Angle=$Emitter.Age
      Period=&RandomRangeFloat(Minimum=0.05, Maximum=0.2)
      Scale=2
      Bias=&RandomRangeFloat(Minimum=0.5, Maximum=2)
     ))
    ScaleA=true
    ScaleRGB=true
   )
   &UpdateMeshOrientation(
    Rotation Coordinate Space=::NewEnumerator3
    Rotation Rate=&RandomRangeFloat(Minimum=-0.75, Maximum=0.75)
    Rotation Vector=<1 0 0>
   )
  ]
  rendererClasses [&NiagaraMeshRendererProperties, &NiagaraLightRendererProperties]
  renderers [
   {rendererClass=&NiagaraMeshRendererProperties, rendererIndex=0}
   {rendererClass=&NiagaraLightRendererProperties, rendererIndex=1}
  ]
 }
 Lightning_Back_Lower001 {
  emitterUpdateScript [
   &EmitterState(
    Life Cycle Mode=::NewEnumerator1
    Loop Duration=&RandomRangeFloat(Minimum=1.5, Maximum=2.5)
    Recalculate Duration Each Loop=true
    UseLoopDelay=false
   )
   &SpawnBurst_Instantaneous(
    Spawn Count=&MakeCustomIntFromBool(
     Bool=&SetBoolByFloatComparison(A=&RandomRangeFloat(Minimum=0, Maximum=0.5), B=0.25)
     True INT=1
     False INT=2
    )
    Spawn Probability=0.4
    Use Spawn Probability=true
    UseLoopCountLimit=false
   )
   &SpawnRate(SpawnRate=5, Use Spawn Probability=false) {enabled=false}
  ]
  particleSpawnScript [
   &InitializeParticle(
    Color=<rgba 1 0 0.7667 0.025>
    Lifetime Max=0.3
    Lifetime Min=0.2
    Lifetime Mode=::NewEnumerator1
    Mesh Scale=&Multiply_VectorByFloat(
     Vector=&RandomRangeVector(Minimum=<0.2 -0.5 1>, Maximum=<0.5 -0.25 0.5>)
     Float=&FloatFromCurve(FloatCurve=DataInterface() {propertyValues {Curve {keys [{arriveTangent=-3.9075, interpMode=RCIM_Cubic, leaveTangent=-3.9075, tangentMode=RCTM_User, time=0, value=1}, <at 1 0 RCIM_Cubic>]}}})
    )
    Mesh Scale Mode=::NewEnumerator4
    Position Mode=::NewEnumerator1
    Sprite Size Mode=::NewEnumerator3
    Uniform Sprite Size=2
    UsePositionOffset=false
    Write Material Random=true
   )
   /SetVariables(Particles.Pos Choice:NiagaraInt32=&Multiply_Int(A=&RandomRangeInt(Minimum=6, Maximum=10), B=1)) {isSetParameters=true, module=SetVariables_#1}
   &InitialMeshOrientation(
    Mesh Orientation Mode=::NewEnumerator7
    Rotation=&RandomRangeVector(Minimum=<0.1 0.25 -0.1>, Maximum=<-0.1 0.25 0.1>)
    Use Rotation Vector=true
   )
   &InitialMeshOrientation(
    Mesh Orientation Mode=::NewEnumerator7
    Rotation=&RandomRangeVector(Minimum=<0 0 0>, Maximum=<0.25 0 0>)
    Use Rotation Vector=true
   ) {enabled=false, module=InitialMeshOrientation001}
  ]
  particleUpdateScript [
   &ParticleState()
   &self:Drift_Horror_Get_Set_of_4_Bones(New Skeletal Mesh:NiagaraDataInterfaceSkeletalMesh=$User.SkeletalMesh, StartIndex:NiagaraInt32=$Particles.Initial.Pos Choice) {module=Drift_Horror_Get_Set_of_4_Bones}
   &OrientMeshToVector(
    Look At Direction:Vector3f=&Subtract_Vector(A=$Output.Drift_Horror_Get_Set_of_4_Bones.BonePos2, B=$Output.Drift_Horror_Get_Set_of_4_Bones.BonePos1)
    Override Quat=false
    Up Axis Direction:Vector3f=<0 0 1>
    UseOptionalSideAxis=false
    UseOptionalUpAxis=false
   ) {enabled=false}
   /SetVariables(
    Particles.Pos A:Vector3f=&Lerp_Vector(
     A=$Output.Drift_Horror_Get_Set_of_4_Bones.BonePos1
     B=$Output.Drift_Horror_Get_Set_of_4_Bones.BonePos2
     Alpha=$Particles.Initial.MaterialRandom
    )
    Particles.Pos B:Vector3f=&Lerp_Vector(
     A=$Output.Drift_Horror_Get_Set_of_4_Bones.BonePos3
     B=$Output.Drift_Horror_Get_Set_of_4_Bones.BonePos4
     Alpha=$Particles.Initial.MaterialRandom
    )
   ) {isSetParameters=true, module=SetVariables_#2}
   /SetVariables(Particles.Position=&Lerp_Vector(A=$Particles.Pos A, B=$Particles.Pos B, Alpha=0.5)) {isSetParameters=true, module=SetVariables_#3}
   /SetVariables(Particles.MeshOrientation=&LerpQuaternion(
     Quaternion A:Quat4f=$Particles.SkeletalMeshLocation.SampledRotation
     Quaternion B:Quat4f=$Particles.SkeletalMeshLocation001.SampledRotation
     Lerp Factor:NiagaraFloat=$Particles.Initial.MaterialRandom
     Enable Spherical Linear Interpolation:NiagaraBool=true
    )) {enabled=false, isSetParameters=true, module=SetVariables_#4}
   &DynamicMaterialParameters(
    Index 0 Param 1=1
    Index 0 Param 3=-0.1
    Module.Param0WriteEnabled=true
    Module.Param1WriteEnabled=true
    Module.Param2WriteEnabled=true
    Module.Param3WriteEnabled=true
   )
   &ScaleMeshSize(Scale Factor=&MakeVector(
     X=&RandomRangeFloat(Minimum=0.75, Maximum=1.5)
     Y=&Lerp_Float(A=1, B=&RandomRangeFloat(Minimum=1.2, Maximum=1.5), Alpha=$Particles.NormalizedAge)
     Z=&Lerp_Float(A=1, B=&RandomRangeFloat(Minimum=0.5, Maximum=2), Alpha=$Particles.NormalizedAge)
    ))
   &FN_Light_Attributes(
    Color=<rgba 10 2.5 8.1923 1>
    Create Color Attribute=true
    Create Exponent Attribute=true
    Create Position Attribute=true
    Exponent=1
    Position=&Add_Vector(
     A=Linked() {linkedVariable {name=Particles.Position, type=NiagaraPosition}}
     B=&Multiply_VectorByFloat(
      Vector=&ReturnMeshOrientationAxis(Mesh Orientation Axis=::YAxis, Mesh Orientation Quat=$Particles.MeshOrientation)
      Float=-8
     )
    )
    Scale Color=&FloatFromCurve(FloatCurve=DataInterface() {propertyValues {Curve {keys [
         {interpMode=RCIM_Cubic, tangentMode=RCTM_User, time=0, value=0}
         {arriveTangentWeight=0.0667, interpMode=RCIM_Cubic, leaveTangentWeight=0.2667, tangentMode=RCTM_User, time=0.2, value=1}
         {arriveTangent=-0.0833, interpMode=RCIM_Cubic, leaveTangent=-0.0833, time=0.5, value=0.2}
         {arriveTangent=-0.4, interpMode=RCIM_Cubic, leaveTangent=-0.4, time=0.8, value=0.95}
         <at 1 0 RCIM_Cubic>
        ]}}})
   )
   &ScaleColor(
    Scale Alpha=&Multiply_Float(
     A=&FloatFromCurve(FloatCurve=DataInterface() {propertyValues {Curve {keys [
          {
           arriveTangentWeight=0.9771
           interpMode=RCIM_Cubic
           leaveTangentWeight=0.9771
           tangentMode=RCTM_User
           tangentWeightMode=RCTWM_WeightedBoth
           time=0
           value=1
          }
          {
           arriveTangent=-4.27
           arriveTangentWeight=0.8915
           interpMode=RCIM_Cubic
           leaveTangent=-4.27
           leaveTangentWeight=0.8915
           tangentMode=RCTM_User
           tangentWeightMode=RCTWM_WeightedBoth
           time=1
           value=0
          }
         ]}}})
     B=&Sine(
      Normalized Angle=$Emitter.Age
      Period=&RandomRangeFloat(Minimum=0.05, Maximum=0.2)
      Scale=2
      Bias=&RandomRangeFloat(Minimum=0.5, Maximum=5)
     )
    )
    ScaleA=true
    ScaleRGB=true
   )
  ]
  rendererClasses [&NiagaraMeshRendererProperties, &NiagaraLightRendererProperties]
  renderers [
   {rendererClass=&NiagaraMeshRendererProperties, rendererIndex=0}
   {rendererClass=&NiagaraLightRendererProperties, rendererIndex=1}
  ]
 }
 Lightning_Back_Upper {
  emitterUpdateScript [
   &EmitterState(
    Life Cycle Mode=::NewEnumerator1
    Loop Duration=&RandomRangeFloat(Minimum=1, Maximum=5)
    Recalculate Duration Each Loop=true
    UseLoopDelay=false
   )
   &SpawnBurst_Instantaneous(
    Spawn Count=&MakeCustomIntFromBool(
     Bool=&SetBoolByFloatComparison(A=&RandomRangeFloat(Minimum=0, Maximum=0.5), B=0.25)
     True INT=1
     False INT=2
    )
    Spawn Probability=0.5
    Use Spawn Probability=true
    UseLoopCountLimit=false
   )
   &SpawnRate(SpawnRate=30, Use Spawn Probability=false) {enabled=false}
  ]
  particleSpawnScript [
   &InitializeParticle(
    Color=<rgba 2 0 1.5333 1>
    Lifetime Max=0.3
    Lifetime Min=0.25
    Lifetime Mode=::NewEnumerator1
    Mesh Scale=&Multiply_VectorByFloat(
     Vector=&RandomRangeVector(Minimum=<1 0.5 0.5>, Maximum=<0.5 0.5 0.5>)
     Float=&FloatFromCurve(FloatCurve=DataInterface() {propertyValues {Curve {keys [{arriveTangent=-3.9075, interpMode=RCIM_Cubic, leaveTangent=-3.9075, tangentMode=RCTM_User, time=0, value=1}, <at 1 0 RCIM_Cubic>]}}})
    )
    Mesh Scale Mode=::NewEnumerator4
    Position Mode=::NewEnumerator1
    Sprite Size Mode=::NewEnumerator3
    Uniform Sprite Size=2
    UsePositionOffset=false
    Write Material Random=true
   )
   &InitialMeshOrientation(
    Mesh Orientation Mode=::NewEnumerator7
    Rotation=&RandomRangeVector(Minimum=<0.05 0 -0.25>, Maximum=<-0.05 0 -0.25>)
    Use Rotation Vector=true
   )
   &InitialMeshOrientation(
    Mesh Orientation Mode=::NewEnumerator7
    Rotation=&RandomRangeVector(Minimum=<0 1 -0.1>, Maximum=<0 0 0.1>)
    Use Rotation Vector=true
   ) {module=InitialMeshOrientation001}
  ]
  particleUpdateScript [
   &ParticleState()
   &SkeletalMeshLocation(
    Bone / Socket Index=&RandomRangeInt(Minimum=1, Maximum=4)
    Bone Sampling Mode=::NewEnumerator6
    MaskSpawnGroup=false
    Sampled Position Offset=<0 -10 0>
    Skeletal Mesh=$User.SkeletalMesh
   )
   &OrientMeshToVector(
    Look At Direction:Vector3f=&Subtract_Vector(A=$Output.Drift_Horror_Get_Set_of_4_Bones.BonePos2, B=$Output.Drift_Horror_Get_Set_of_4_Bones.BonePos1)
    Override Quat=false
    Up Axis Direction:Vector3f=<0 0 1>
    UseOptionalSideAxis=false
    UseOptionalUpAxis=false
   ) {enabled=false}
   /SetVariables(Particles.MeshOrientation=&LerpQuaternion(
     Quaternion A:Quat4f=$Particles.SkeletalMeshLocation.SampledRotation
     Quaternion B:Quat4f=$Particles.SkeletalMeshLocation001.SampledRotation
     Lerp Factor:NiagaraFloat=$Particles.Initial.MaterialRandom
     Enable Spherical Linear Interpolation:NiagaraBool=true
    )) {enabled=false, isSetParameters=true, module=SetVariables_#4}
   &DynamicMaterialParameters(
    Index 0 Param 1=1
    Index 0 Param 3=-0.1
    Module.Param0WriteEnabled=true
    Module.Param1WriteEnabled=true
    Module.Param2WriteEnabled=true
    Module.Param3WriteEnabled=true
   )
   &ScaleMeshSize(Scale Factor=&MakeVector(
     X=&FloatFromCurve(FloatCurve=DataInterface() {propertyValues {Curve {keys [{tangentMode=RCTM_User, time=0, value=0.5}, {tangentMode=RCTM_User, time=0.1, value=1}, <at 1 1.1>]}}})
     Y=&Lerp_Float(A=1, B=&RandomRangeFloat(Minimum=1.2, Maximum=2), Alpha=$Particles.NormalizedAge)
     Z=&Lerp_Float(A=1, B=&RandomRangeFloat(Minimum=1.2, Maximum=2), Alpha=$Particles.NormalizedAge)
    ))
   &FN_Light_Attributes(
    Color=<rgba 10 2.5 8.1923 1>
    Create Color Attribute=true
    Create Exponent Attribute=true
    Create Position Attribute=true
    Exponent=3
    Position=&Add_Vector(
     A=Linked() {linkedVariable {name=Particles.Position, type=NiagaraPosition}}
     B=&Multiply_VectorByFloat(
      Vector=&ReturnMeshOrientationAxis(Mesh Orientation Axis=::XAxis, Mesh Orientation Quat=$Particles.MeshOrientation)
      Float=10
     )
    )
   )
   &ScaleColor(
    Scale Alpha=&Multiply_Float(
     A=&FloatFromCurve(FloatCurve=DataInterface() {propertyValues {Curve {keys [
          {
           arriveTangentWeight=0.9771
           interpMode=RCIM_Cubic
           leaveTangentWeight=0.9771
           tangentMode=RCTM_User
           tangentWeightMode=RCTWM_WeightedBoth
           time=0
           value=1
          }
          {
           arriveTangent=-4.27
           arriveTangentWeight=0.8915
           interpMode=RCIM_Cubic
           leaveTangent=-4.27
           leaveTangentWeight=0.8915
           tangentMode=RCTM_User
           tangentWeightMode=RCTWM_WeightedBoth
           time=1
           value=0
          }
         ]}}})
     B=&Sine(
      Normalized Angle=$Emitter.Age
      Period=&RandomRangeFloat(Minimum=0.05, Maximum=0.2)
      Scale=2
      Bias=&RandomRangeFloat(Minimum=0.5, Maximum=5)
     )
    )
    ScaleA=true
    ScaleRGB=true
   )
   &UpdateMeshOrientation(
    Rotation Coordinate Space=::NewEnumerator3
    Rotation Rate=&RandomRangeFloat(Minimum=-0.75, Maximum=0.75)
    Rotation Vector=<1 0 0>
   )
  ]
  rendererClasses [&NiagaraMeshRendererProperties, &NiagaraLightRendererProperties]
  renderers [
   {rendererClass=&NiagaraMeshRendererProperties, rendererIndex=0}
   {rendererClass=&NiagaraLightRendererProperties, rendererIndex=1}
  ]
 }
 Lightning_Back_Upper001 {
  emitterUpdateScript [
   &EmitterState(
    Life Cycle Mode=::NewEnumerator1
    Loop Duration=&RandomRangeFloat(Minimum=1.5, Maximum=5)
    Recalculate Duration Each Loop=true
    UseLoopDelay=false
   )
   &SpawnBurst_Instantaneous(
    Spawn Count=&MakeCustomIntFromBool(
     Bool=&SetBoolByFloatComparison(A=&RandomRangeFloat(Minimum=0, Maximum=0.5), B=0.25)
     True INT=1
     False INT=2
    )
    Spawn Probability=0.4
    Use Spawn Probability=true
    UseLoopCountLimit=false
   )
   &SpawnRate(SpawnRate=10, Use Spawn Probability=false) {enabled=false}
  ]
  particleSpawnScript [
   &InitializeParticle(
    Color=<rgba 2 0 1.5333 0.05>
    Lifetime Max=0.32
    Lifetime Min=0.2
    Lifetime Mode=::NewEnumerator1
    Mesh Scale=&Multiply_VectorByFloat(
     Vector=&RandomRangeVector(Minimum=<0.2 -0.5 1>, Maximum=<0.2 -0.25 0.5>)
     Float=&FloatFromCurve(FloatCurve=DataInterface() {propertyValues {Curve {keys [{arriveTangent=-3.9075, interpMode=RCIM_Cubic, leaveTangent=-3.9075, tangentMode=RCTM_User, time=0, value=1}, <at 1 0 RCIM_Cubic>]}}})
    )
    Mesh Scale Mode=::NewEnumerator4
    Position Mode=::NewEnumerator1
    Sprite Size Mode=::NewEnumerator3
    Uniform Sprite Size=2
    UsePositionOffset=false
    Write Material Random=true
   )
   &InitialMeshOrientation(
    Mesh Orientation Mode=::NewEnumerator7
    Rotation=&RandomRangeVector(Minimum=<0.05 0 -0.25>, Maximum=<-0.05 0 -0.25>)
    Use Rotation Vector=true
   ) {enabled=false}
   &InitialMeshOrientation(
    Mesh Orientation Mode=::NewEnumerator7
    Rotation=&RandomRangeVector(Minimum=<0.05 0.05 0.15>, Maximum=<-0.05 -0.05 -0.15>)
    Use Rotation Vector=true
   ) {module=InitialMeshOrientation001}
  ]
  particleUpdateScript [
   &ParticleState()
   &SkeletalMeshLocation(
    Bone / Socket Index=&RandomRangeInt(Minimum=1, Maximum=4)
    Bone Sampling Mode=::NewEnumerator6
    MaskSpawnGroup=false
    Sampled Position Offset=<0 -10 0>
    Skeletal Mesh=$User.SkeletalMesh
   )
   &OrientMeshToVector(
    Look At Direction:Vector3f=&Subtract_Vector(A=$Output.Drift_Horror_Get_Set_of_4_Bones.BonePos2, B=$Output.Drift_Horror_Get_Set_of_4_Bones.BonePos1)
    Override Quat=false
    Up Axis Direction:Vector3f=<0 0 1>
    UseOptionalSideAxis=false
    UseOptionalUpAxis=false
   ) {enabled=false}
   /SetVariables(Particles.MeshOrientation=&LerpQuaternion(
     Quaternion A:Quat4f=$Particles.SkeletalMeshLocation.SampledRotation
     Quaternion B:Quat4f=$Particles.SkeletalMeshLocation001.SampledRotation
     Lerp Factor:NiagaraFloat=$Particles.Initial.MaterialRandom
     Enable Spherical Linear Interpolation:NiagaraBool=true
    )) {enabled=false, isSetParameters=true, module=SetVariables_#4}
   &DynamicMaterialParameters(
    Index 0 Param 1=1
    Index 0 Param 3=-0.1
    Module.Param0WriteEnabled=true
    Module.Param1WriteEnabled=true
    Module.Param2WriteEnabled=true
    Module.Param3WriteEnabled=true
   )
   &FN_Light_Attributes(
    Color=<rgba 2 0.5 1.6385 1>
    Create Color Attribute=true
    Create Exponent Attribute=true
    Create Position Attribute=true
    Exponent=3
    Position=&Add_Vector(
     A=Linked() {linkedVariable {name=Particles.Position, type=NiagaraPosition}}
     B=&Multiply_VectorByFloat(
      Vector=&ReturnMeshOrientationAxis(Mesh Orientation Axis=::YAxis, Mesh Orientation Quat=$Particles.MeshOrientation)
      Float=-10
     )
    )
    Scale Color=&Multiply_Float(
     A=&FloatFromCurve(FloatCurve=DataInterface() {propertyValues {Curve {keys [
          {
           arriveTangentWeight=0.9771
           interpMode=RCIM_Cubic
           leaveTangentWeight=0.9771
           tangentMode=RCTM_User
           tangentWeightMode=RCTWM_WeightedBoth
           time=0
           value=1
          }
          {
           arriveTangent=-4.27
           arriveTangentWeight=0.8915
           interpMode=RCIM_Cubic
           leaveTangent=-4.27
           leaveTangentWeight=0.8915
           tangentMode=RCTM_User
           tangentWeightMode=RCTWM_WeightedBoth
           time=1
           value=0
          }
         ]}}})
     B=&Sine(
      Normalized Angle=$Emitter.Age
      Period=&RandomRangeFloat(Minimum=0.05, Maximum=0.2)
      Scale=2
      Bias=&RandomRangeFloat(Minimum=0.5, Maximum=2)
     )
    )
   )
   &ScaleMeshSize(Scale Factor=&MakeVector(
     X=&RandomRangeFloat(Minimum=0.75, Maximum=1.5)
     Y=&Lerp_Float(A=1, B=&RandomRangeFloat(Minimum=1.2, Maximum=1.5), Alpha=$Particles.NormalizedAge)
     Z=&Lerp_Float(A=1, B=&RandomRangeFloat(Minimum=0.5, Maximum=2), Alpha=$Particles.NormalizedAge)
    )) {module=ScaleMeshSize001}
   &ScaleColor(
    Scale Alpha=&Multiply_Float(
     A=&FloatFromCurve(FloatCurve=DataInterface() {propertyValues {Curve {keys [
          {
           arriveTangentWeight=0.9771
           interpMode=RCIM_Cubic
           leaveTangentWeight=0.9771
           tangentMode=RCTM_User
           tangentWeightMode=RCTWM_WeightedBoth
           time=0
           value=1
          }
          {
           arriveTangent=-4.27
           arriveTangentWeight=0.8915
           interpMode=RCIM_Cubic
           leaveTangent=-4.27
           leaveTangentWeight=0.8915
           tangentMode=RCTM_User
           tangentWeightMode=RCTWM_WeightedBoth
           time=1
           value=0
          }
         ]}}})
     B=&Sine(
      Normalized Angle=$Emitter.Age
      Period=&RandomRangeFloat(Minimum=0.05, Maximum=0.2)
      Scale=2
      Bias=&RandomRangeFloat(Minimum=0.5, Maximum=5)
     )
    )
    ScaleA=true
    ScaleRGB=true
   )
  ]
  rendererClasses [&NiagaraMeshRendererProperties, &NiagaraLightRendererProperties]
  renderers [
   {rendererClass=&NiagaraMeshRendererProperties, rendererIndex=0}
   {rendererClass=&NiagaraLightRendererProperties, rendererIndex=1}
  ]
 }
 Lightning_Body {
  emitterUpdateScript [
   &EmitterState(
    Life Cycle Mode=::NewEnumerator1
    Loop Duration=&RandomRangeFloat(Minimum=0.1, Maximum=1.6)
    Recalculate Duration Each Loop=true
    UseLoopDelay=false
   )
   &SpawnBurst_Instantaneous(
    Spawn Count=&MakeCustomIntFromBool(
     Bool=&SetBoolByFloatComparison(A=&RandomRangeFloat(Minimum=0, Maximum=0.5), B=0.25)
     True INT=1
     False INT=2
    )
    Spawn Probability=0.5
    Use Spawn Probability=true
    UseLoopCountLimit=false
   )
   &SpawnRate(SpawnRate=500, Use Spawn Probability=false) {enabled=false}
  ]
  particleSpawnScript [
   &InitializeParticle(
    Color=<rgba 2 0 1.5333 1>
    Lifetime Max=0.3
    Lifetime Min=0.2
    Lifetime Mode=::NewEnumerator1
    Mesh Scale=&Multiply_VectorByFloat(
     Vector=&RandomRangeVector(Minimum=<1 0.5 0.5>, Maximum=<0.5 0.5 0.5>)
     Float=&FloatFromCurve(FloatCurve=DataInterface() {propertyValues {Curve {keys [{arriveTangent=-3.9075, interpMode=RCIM_Cubic, leaveTangent=-3.9075, tangentMode=RCTM_User, time=0, value=1}, <at 1 0 RCIM_Cubic>]}}})
    )
    Mesh Scale Mode=::NewEnumerator4
    Position Mode=::NewEnumerator1
    Sprite Size Mode=::NewEnumerator3
    Uniform Sprite Size=2
    UsePositionOffset=false
    Write Material Random=true
   )
   /SetVariables(Particles.Pos Choice:NiagaraInt32=&SelectIntFromArray(
     Array Sampling Mode=::NewEnumerator1
     Direct Array Index=?
     Int Selection Array=DataInterface() {propertyValues {IntData [0,1,2,3,4,5,7,8,9,10,11,12]}}
     Random Seed=0
    )) {isSetParameters=true, module=SetVariables_#1}
   &InitialMeshOrientation(
    Mesh Orientation Mode=::NewEnumerator7
    Rotation=&RandomRangeVector(Minimum=<0 0 -0.1>, Maximum=<1 0 0.1>)
    Use Rotation Vector=true
   )
   &InitialMeshOrientation(
    Mesh Orientation Mode=::NewEnumerator7
    Rotation=&RandomRangeVector(Minimum=<0 0 0>, Maximum=<0.25 0 0>)
    Use Rotation Vector=true
   ) {enabled=false, module=InitialMeshOrientation001}
  ]
  particleUpdateScript [
   &ParticleState()
   &OrientMeshToVector(
    Look At Direction:Vector3f=&Subtract_Vector(A=$Output.Drift_Horror_Get_Set_of_4_Bones.BonePos2, B=$Output.Drift_Horror_Get_Set_of_4_Bones.BonePos1)
    Override Quat=false
    Up Axis Direction:Vector3f=<0 0 1>
    UseOptionalSideAxis=false
    UseOptionalUpAxis=false
   ) {enabled=false}
   &SkeletalMeshLocation(
    Bone / Socket Index=$Particles.Initial.Pos Choice
    MaskSpawnGroup=false
    Mesh Sampling Type=::NewEnumerator5
    Position Sampling=::NewEnumerator2
    Skeletal Mesh=$User.SkeletalMesh
    Socket Sampling Mode=::NewEnumerator5
    Use Sampled Mesh Attribute as Ribbon ID=false
   )
   &SkeletalMeshLocation(
    Bone / Socket Index=&Add_Integer(A=$Particles.Initial.Pos Choice, B=1)
    MaskSpawnGroup=false
    Mesh Sampling Type=::NewEnumerator5
    Position Sampling=::NewEnumerator2
    Skeletal Mesh=$User.SkeletalMesh
    Socket Sampling Mode=::NewEnumerator5
   ) {module=SkeletalMeshLocation001}
   /SetVariables(Particles.Position=&Lerp_Vector(
     A=Linked() {linkedVariable {name=Particles.SkeletalMeshLocation.SampledPosition, type=NiagaraPosition}}
     B=Linked() {linkedVariable {name=Particles.SkeletalMeshLocation001.SampledPosition, type=NiagaraPosition}}
     Alpha=$Particles.Initial.MaterialRandom
    )) {isSetParameters=true, module=SetVariables_#3}
   /SetVariables(Particles.MeshOrientation=&LerpQuaternion(
     Quaternion A:Quat4f=$Particles.SkeletalMeshLocation.SampledRotation
     Quaternion B:Quat4f=$Particles.SkeletalMeshLocation001.SampledRotation
     Lerp Factor:NiagaraFloat=$Particles.Initial.MaterialRandom
     Enable Spherical Linear Interpolation:NiagaraBool=true
    )) {enabled=false, isSetParameters=true, module=SetVariables_#4}
   &DynamicMaterialParameters(
    Index 0 Param 1=1
    Index 0 Param 3=-0.1
    Module.Param0WriteEnabled=true
    Module.Param1WriteEnabled=true
    Module.Param2WriteEnabled=true
    Module.Param3WriteEnabled=true
   )
   &ScaleMeshSize(Scale Factor=&MakeVector(
     X=&FloatFromCurve(FloatCurve=DataInterface() {propertyValues {Curve {keys [{tangentMode=RCTM_User, time=0, value=0.5}, {tangentMode=RCTM_User, time=0.1, value=1}, <at 1 1.1>]}}})
     Y=&Lerp_Float(A=1, B=&RandomRangeFloat(Minimum=1.2, Maximum=2.5), Alpha=$Particles.NormalizedAge)
     Z=&Lerp_Float(A=1, B=&RandomRangeFloat(Minimum=1.2, Maximum=2.5), Alpha=$Particles.NormalizedAge)
    ))
   &FN_Light_Attributes(
    Color=<rgba 25 6.25 20.4807 1>
    Create Color Attribute=true
    Create Exponent Attribute=true
    Create Position Attribute=true
    Exponent=3
    Position=&Add_Vector(
     A=Linked() {linkedVariable {name=Particles.Position, type=NiagaraPosition}}
     B=&Multiply_VectorByFloat(
      Vector=&ReturnMeshOrientationAxis(Mesh Orientation Axis=::XAxis, Mesh Orientation Quat=$Particles.MeshOrientation)
      Float=2
     )
    )
   )
   &ScaleColor(
    Scale Alpha=&Multiply_Float(
     A=&FloatFromCurve(FloatCurve=DataInterface() {propertyValues {Curve {keys [
          {
           arriveTangentWeight=0.9771
           interpMode=RCIM_Cubic
           leaveTangentWeight=0.9771
           tangentMode=RCTM_User
           tangentWeightMode=RCTWM_WeightedBoth
           time=0
           value=1
          }
          {
           arriveTangent=-4.27
           arriveTangentWeight=0.8915
           interpMode=RCIM_Cubic
           leaveTangent=-4.27
           leaveTangentWeight=0.8915
           tangentMode=RCTM_User
           tangentWeightMode=RCTWM_WeightedBoth
           time=1
           value=0
          }
         ]}}})
     B=&Sine(
      Normalized Angle=$Emitter.Age
      Period=&RandomRangeFloat(Minimum=0.05, Maximum=0.2)
      Scale=2
      Bias=&RandomRangeFloat(Minimum=0.5, Maximum=5)
     )
    )
    ScaleA=true
    ScaleRGB=true
   )
   &UpdateMeshOrientation(
    Rotation Coordinate Space=::NewEnumerator3
    Rotation Rate=&RandomRangeFloat(Minimum=-0.75, Maximum=0.75)
    Rotation Vector=<1 0 0>
   )
  ]
  rendererClasses [&NiagaraMeshRendererProperties, &NiagaraLightRendererProperties]
  renderers [
   {rendererClass=&NiagaraMeshRendererProperties, rendererIndex=0}
   {rendererClass=&NiagaraLightRendererProperties, rendererIndex=1}
  ]
 }
 Lightning_Body001 {
  emitterUpdateScript [
   &EmitterState(
    Life Cycle Mode=::NewEnumerator1
    Loop Duration=&RandomRangeFloat(Minimum=0.5, Maximum=1.6)
    Recalculate Duration Each Loop=true
    UseLoopDelay=false
   )
   &SpawnBurst_Instantaneous(
    Spawn Count=&MakeCustomIntFromBool(
     Bool=&SetBoolByFloatComparison(A=&RandomRangeFloat(Minimum=0, Maximum=0.5), B=0.25)
     True INT=1
     False INT=2
    )
    Spawn Probability=0.4
    Use Spawn Probability=true
    UseLoopCountLimit=false
   )
   &SpawnRate(SpawnRate=20, Use Spawn Probability=false) {enabled=false}
  ]
  particleSpawnScript [
   &InitializeParticle(
    Color=<rgba 2 0 1.5333 0.05>
    Lifetime Max=0.35
    Lifetime Min=0.18
    Lifetime Mode=::NewEnumerator1
    Mesh Scale=&Multiply_VectorByFloat(
     Vector=&RandomRangeVector(Minimum=<0.25 0.5 0.2>, Maximum=<0.15 0.1 0.1>)
     Float=&FloatFromCurve(FloatCurve=DataInterface() {propertyValues {Curve {keys [{arriveTangent=-3.9075, interpMode=RCIM_Cubic, leaveTangent=-3.9075, tangentMode=RCTM_User, time=0, value=1}, <at 1 0 RCIM_Cubic>]}}})
    )
    Mesh Scale Mode=::NewEnumerator4
    Position Mode=::NewEnumerator1
    Sprite Size Mode=::NewEnumerator3
    Uniform Sprite Size=2
    UsePositionOffset=false
    Write Material Random=true
   )
   /SetVariables(Particles.Pos Choice:NiagaraInt32=&SelectIntFromArray(
     Array Sampling Mode=::NewEnumerator1
     Direct Array Index=?
     Int Selection Array=DataInterface() {propertyValues {IntData [1,2,3,4,5,7,8,9,10,11]}}
     Random Seed=0
    )) {isSetParameters=true, module=SetVariables_#1}
   &InitialMeshOrientation(
    Mesh Orientation Mode=::NewEnumerator7
    Rotation=&RandomRangeVector(Minimum=<0.1 -0.05 -0.25>, Maximum=<0 0.05 -0.25>)
    Use Rotation Vector=true
   )
   &InitialMeshOrientation(
    Mesh Orientation Mode=::NewEnumerator7
    Rotation=&RandomRangeVector(Minimum=<0 0 0>, Maximum=<0.25 0 0>)
    Use Rotation Vector=true
   ) {enabled=false, module=InitialMeshOrientation001}
  ]
  particleUpdateScript [
   &ParticleState()
   &OrientMeshToVector(
    Look At Direction:Vector3f=&Subtract_Vector(A=$Output.Drift_Horror_Get_Set_of_4_Bones.BonePos2, B=$Output.Drift_Horror_Get_Set_of_4_Bones.BonePos1)
    Override Quat=false
    Up Axis Direction:Vector3f=<0 0 1>
    UseOptionalSideAxis=false
    UseOptionalUpAxis=false
   ) {enabled=false}
   &SkeletalMeshLocation(
    Bone / Socket Index=$Particles.Initial.Pos Choice
    MaskSpawnGroup=false
    Mesh Sampling Type=::NewEnumerator5
    Position Sampling=::NewEnumerator2
    Skeletal Mesh=$User.SkeletalMesh
    Socket Sampling Mode=::NewEnumerator5
    Use Sampled Mesh Attribute as Ribbon ID=false
   )
   &SkeletalMeshLocation(
    Bone / Socket Index=&Add_Integer(A=$Particles.Initial.Pos Choice, B=1)
    MaskSpawnGroup=false
    Mesh Sampling Type=::NewEnumerator5
    Position Sampling=::NewEnumerator2
    Skeletal Mesh=$User.SkeletalMesh
    Socket Sampling Mode=::NewEnumerator5
   ) {module=SkeletalMeshLocation001}
   /SetVariables(Particles.Position=&Lerp_Vector(
     A=Linked() {linkedVariable {name=Particles.SkeletalMeshLocation.SampledPosition, type=NiagaraPosition}}
     B=Linked() {linkedVariable {name=Particles.SkeletalMeshLocation001.SampledPosition, type=NiagaraPosition}}
     Alpha=$Particles.Initial.MaterialRandom
    )) {isSetParameters=true, module=SetVariables_#3}
   /SetVariables(Particles.MeshOrientation=&LerpQuaternion(
     Quaternion A:Quat4f=$Particles.SkeletalMeshLocation.SampledRotation
     Quaternion B:Quat4f=$Particles.SkeletalMeshLocation001.SampledRotation
     Lerp Factor:NiagaraFloat=$Particles.Initial.MaterialRandom
     Enable Spherical Linear Interpolation:NiagaraBool=true
    )) {enabled=false, isSetParameters=true, module=SetVariables_#4}
   &DynamicMaterialParameters(
    Index 0 Param 1=1
    Index 0 Param 3=-0.1
    Module.Param0WriteEnabled=true
    Module.Param1WriteEnabled=true
    Module.Param2WriteEnabled=true
    Module.Param3WriteEnabled=true
   )
   &ScaleMeshSize(Scale Factor=&MakeVector(
     X=1
     Y=&Lerp_Float(A=1, B=&RandomRangeFloat(Minimum=1.25, Maximum=1.1), Alpha=$Particles.NormalizedAge)
     Z=0
    ))
   &FN_Light_Attributes(
    Color=<rgba 1 0.25 0.8192 1>
    Create Color Attribute=true
    Create Exponent Attribute=true
    Create Position Attribute=true
    Exponent=1
    Position=&Add_Vector(
     A=Linked() {linkedVariable {name=Particles.Position, type=NiagaraPosition}}
     B=&Multiply_VectorByFloat(
      Vector=&ReturnMeshOrientationAxis(Mesh Orientation Axis=::XAxis, Mesh Orientation Quat=$Particles.MeshOrientation)
      Float=3
     )
    )
    Scale Color=&Multiply_Float(
     A=&FloatFromCurve(FloatCurve=DataInterface() {propertyValues {Curve {keys [
          {
           arriveTangentWeight=0.9771
           interpMode=RCIM_Cubic
           leaveTangentWeight=0.9771
           tangentMode=RCTM_User
           tangentWeightMode=RCTWM_WeightedBoth
           time=0
           value=1
          }
          {
           arriveTangent=-4.27
           arriveTangentWeight=0.8915
           interpMode=RCIM_Cubic
           leaveTangent=-4.27
           leaveTangentWeight=0.8915
           tangentMode=RCTM_User
           tangentWeightMode=RCTWM_WeightedBoth
           time=1
           value=0
          }
         ]}}})
     B=&Sine(
      Normalized Angle=$Emitter.Age
      Period=&RandomRangeFloat(Minimum=0.05, Maximum=0.2)
      Scale=2
      Bias=&RandomRangeFloat(Minimum=0.5, Maximum=2)
     )
    )
   )
   &ScaleColor(
    Scale Alpha=&Multiply_Float(
     A=&FloatFromCurve(FloatCurve=DataInterface() {propertyValues {Curve {keys [
          {
           arriveTangentWeight=0.9771
           interpMode=RCIM_Cubic
           leaveTangentWeight=0.9771
           tangentMode=RCTM_User
           tangentWeightMode=RCTWM_WeightedBoth
           time=0
           value=1
          }
          {
           arriveTangent=-4.27
           arriveTangentWeight=0.8915
           interpMode=RCIM_Cubic
           leaveTangent=-4.27
           leaveTangentWeight=0.8915
           tangentMode=RCTM_User
           tangentWeightMode=RCTWM_WeightedBoth
           time=1
           value=0
          }
         ]}}})
     B=&Sine(
      Normalized Angle=$Emitter.Age
      Period=&RandomRangeFloat(Minimum=0.05, Maximum=0.2)
      Scale=2
      Bias=&RandomRangeFloat(Minimum=0.5, Maximum=5)
     )
    )
    ScaleA=true
    ScaleRGB=true
   )
  ]
  rendererClasses [&NiagaraMeshRendererProperties, &NiagaraLightRendererProperties]
  renderers [
   {rendererClass=&NiagaraMeshRendererProperties, rendererIndex=0}
   {rendererClass=&NiagaraLightRendererProperties, rendererIndex=1}
  ]
 }
 Lightning_Cape_Bottom_Edge {
  emitterUpdateScript [
   &EmitterState(
    Life Cycle Mode=::NewEnumerator1
    Loop Duration=&RandomRangeFloat(Minimum=0.1, Maximum=2)
    Recalculate Duration Each Loop=true
    UseLoopDelay=false
   )
   &SpawnBurst_Instantaneous(
    Spawn Count=&MakeCustomIntFromBool(
     Bool=&SetBoolByFloatComparison(A=&RandomRangeFloat(Minimum=0, Maximum=0.5), B=0.25)
     True INT=1
     False INT=2
    )
    Spawn Probability=0.5
    Use Spawn Probability=true
    UseLoopCountLimit=false
   )
   &SpawnRate(SpawnRate=30, Use Spawn Probability=false) {enabled=false}
  ]
  particleSpawnScript [
   &InitializeParticle(
    Color=<rgba 2 0 1.5333 1>
    Lifetime Max=0.32
    Lifetime Min=0.15
    Lifetime Mode=::NewEnumerator1
    Mesh Scale Max=<1 1 1>
    Mesh Scale Min=<0.5 0.5 0.5>
    Mesh Scale Mode=::NewEnumerator2
    Position Mode=::NewEnumerator1
    Sprite Size Mode=::NewEnumerator3
    Uniform Sprite Size=2
    UsePositionOffset=false
    Write Material Random=true
   )
   /SetVariables(Particles.Pos Choice:NiagaraInt32=&RandomRangeInt(Minimum=0, Maximum=2)) {isSetParameters=true, module=SetVariables_#1}
   &InitialMeshOrientation(
    Mesh Orientation Mode=::NewEnumerator7
    Rotation=&RandomRangeVector(Minimum=<0 0.25 0.25>, Maximum=<1 -0.25 -0.25>)
    Use Rotation Vector=true
   ) {enabled=false}
   &InitialMeshOrientation(Mesh Orientation Mode=::NewEnumerator7, Rotation=<-0.2 0 -0.13>, Use Rotation Vector=true) {module=InitialMeshOrientation001}
   &InitialMeshOrientation(
    Mesh Orientation Mode=::NewEnumerator7
    Rotation=&RandomRangeVector(Minimum=<0 -0.17 0>, Maximum=<0 0.15 0>)
    Use Rotation Vector=true
   ) {module=InitialMeshOrientation002}
  ]
  particleUpdateScript [
   &ParticleState()
   &SkeletalMeshLocation(
    Bone / Socket Index=$Particles.Initial.Pos Choice
    Bone Sampling Mode=::NewEnumerator4
    MaskSpawnGroup=false
    Orientation Sampling=::NewEnumerator1
    Position Sampling=::NewEnumerator2
    Skeletal Mesh=$User.SkeletalMesh
   )
   &SkeletalMeshLocation(
    Bone / Socket Index=&Add_Integer(A=$Particles.Initial.Pos Choice, B=1)
    Bone Sampling Mode=::NewEnumerator4
    MaskSpawnGroup=false
    Position Sampling=::NewEnumerator2
    Skeletal Mesh=$User.SkeletalMesh
   ) {module=SkeletalMeshLocation001}
   /SetVariables(Particles.Position=&Lerp_Vector(
     A=Linked() {linkedVariable {name=Particles.SkeletalMeshLocation.SampledPosition, type=NiagaraPosition}}
     B=Linked() {linkedVariable {name=Particles.SkeletalMeshLocation001.SampledPosition, type=NiagaraPosition}}
     Alpha=$Particles.Initial.MaterialRandom
    )) {isSetParameters=true, module=SetVariables_#3}
   /SetVariables(Particles.MeshOrientation=&LerpQuaternion(
     Quaternion A:Quat4f=$Particles.SkeletalMeshLocation.SampledRotation
     Quaternion B:Quat4f=$Particles.SkeletalMeshLocation001.SampledRotation
     Lerp Factor:NiagaraFloat=$Particles.Initial.MaterialRandom
     Enable Spherical Linear Interpolation:NiagaraBool=true
    )) {enabled=false, isSetParameters=true, module=SetVariables_#4}
   &DynamicMaterialParameters(
    Index 0 Param 1=1
    Index 0 Param 3=-0.1
    Module.Param0WriteEnabled=true
    Module.Param1WriteEnabled=true
    Module.Param2WriteEnabled=true
    Module.Param3WriteEnabled=true
   )
   &ScaleMeshSize(Scale Factor=&MakeVector(
     X=&FloatFromCurve(FloatCurve=DataInterface() {propertyValues {Curve {keys [{tangentMode=RCTM_User, time=0, value=0.5}, {tangentMode=RCTM_User, time=0.1, value=1}, <at 1 0.9>]}}})
     Y=1
     Z=1
    ))
   &FN_Light_Attributes(
    Color=<rgba 5 1.25 4.0961 1>
    Create Color Attribute=true
    Create Exponent Attribute=true
    Create Position Attribute=true
    Position=&Add_Vector(
     A=&Add_Vector(
      A=Linked() {linkedVariable {name=Particles.Position, type=NiagaraPosition}}
      B=&Multiply_VectorByFloat(
       Vector=&ReturnMeshOrientationAxis(Mesh Orientation Axis=::XAxis, Mesh Orientation Quat=$Particles.MeshOrientation)
       Float=15
      )
     )
     B=&Multiply_VectorByFloat(
      Vector=&Subtract_Vector(A=$Output.Drift_Horror_Get_Set_of_4_Bones.BonePos2, B=$Output.Drift_Horror_Get_Set_of_4_Bones.BonePos1)
      Float=1
     )
    )
   )
   &ScaleColor(
    Scale Alpha=&Multiply_Float(
     A=&FloatFromCurve(FloatCurve=DataInterface() {propertyValues {Curve {keys [
          {
           arriveTangentWeight=0.9771
           interpMode=RCIM_Cubic
           leaveTangentWeight=0.9771
           tangentMode=RCTM_User
           tangentWeightMode=RCTWM_WeightedBoth
           time=0
           value=1
          }
          {
           arriveTangent=-4.27
           arriveTangentWeight=0.8915
           interpMode=RCIM_Cubic
           leaveTangent=-4.27
           leaveTangentWeight=0.8915
           tangentMode=RCTM_User
           tangentWeightMode=RCTWM_WeightedBoth
           time=1
           value=0
          }
         ]}}})
     B=&Sine(
      Normalized Angle=$Emitter.Age
      Period=&RandomRangeFloat(Minimum=0.05, Maximum=0.2)
      Scale=2
      Bias=&RandomRangeFloat(Minimum=0.5, Maximum=5)
     )
    )
    ScaleA=true
    ScaleRGB=true
   )
   &UpdateMeshOrientation(
    Rotation Coordinate Space=::NewEnumerator3
    Rotation Rate=&RandomRangeFloat(Minimum=-0.75, Maximum=0.75)
    Rotation Vector=<1 0 0>
   )
  ]
  rendererClasses [&NiagaraMeshRendererProperties, &NiagaraLightRendererProperties]
  renderers [
   {rendererClass=&NiagaraMeshRendererProperties, rendererIndex=0}
   {rendererClass=&NiagaraLightRendererProperties, rendererIndex=1}
  ]
 }
 Lightning_Cape_Bottom_Edge001 {
  emitterUpdateScript [
   &EmitterState(
    Life Cycle Mode=::NewEnumerator1
    Loop Duration=&RandomRangeFloat(Minimum=0.5, Maximum=2)
    Recalculate Duration Each Loop=true
    UseLoopDelay=false
   )
   &SpawnBurst_Instantaneous(
    Spawn Count=&MakeCustomIntFromBool(
     Bool=&SetBoolByFloatComparison(A=&RandomRangeFloat(Minimum=0, Maximum=0.5), B=0.25)
     True INT=1
     False INT=2
    )
    Spawn Probability=4
    Use Spawn Probability=true
    UseLoopCountLimit=false
   )
   &SpawnRate(SpawnRate=30, Use Spawn Probability=false) {enabled=false}
  ]
  particleSpawnScript [
   &InitializeParticle(
    Color=<rgba 2 0 1.5333 0.1>
    Lifetime Max=0.3
    Lifetime Min=0.2
    Lifetime Mode=::NewEnumerator1
    Mesh Scale Max=<0.5 0.5 0.5>
    Mesh Scale Min=<0.1 0.2 1>
    Mesh Scale Mode=::NewEnumerator2
    Position Mode=::NewEnumerator1
    Sprite Size Mode=::NewEnumerator3
    Uniform Sprite Size=2
    UsePositionOffset=false
    Write Material Random=true
   )
   /SetVariables(Particles.Pos Choice:NiagaraInt32=&RandomRangeInt(Minimum=0, Maximum=2)) {isSetParameters=true, module=SetVariables_#1}
   &InitialMeshOrientation(Mesh Orientation Mode=::NewEnumerator7, Rotation=<0 0.25 -0.25>, Use Rotation Vector=true)
   &InitialMeshOrientation(Mesh Orientation Mode=::NewEnumerator7, Rotation=<0.35 0.25 0>, Use Rotation Vector=true) {enabled=false, module=InitialMeshOrientation001}
   &InitialMeshOrientation(
    Mesh Orientation Mode=::NewEnumerator7
    Rotation=&RandomRangeVector(Minimum=<0 -0.17 0>, Maximum=<0 0.15 0>)
    Use Rotation Vector=true
   ) {enabled=false, module=InitialMeshOrientation002}
  ]
  particleUpdateScript [
   &ParticleState()
   &SkeletalMeshLocation(
    Bone / Socket Index=$Particles.Initial.Pos Choice
    Bone Sampling Mode=::NewEnumerator4
    MaskSpawnGroup=false
    Orientation Sampling=::NewEnumerator1
    Position Sampling=::NewEnumerator2
    Skeletal Mesh=$User.SkeletalMesh
   )
   &SkeletalMeshLocation(
    Bone / Socket Index=&Add_Integer(A=$Particles.Initial.Pos Choice, B=1)
    Bone Sampling Mode=::NewEnumerator4
    MaskSpawnGroup=false
    Orientation Sampling=::NewEnumerator1
    Position Sampling=::NewEnumerator2
    Skeletal Mesh=$User.SkeletalMesh
   ) {module=SkeletalMeshLocation001}
   /SetVariables(Particles.Position=&Lerp_Vector(
     A=Linked() {linkedVariable {name=Particles.SkeletalMeshLocation.SampledPosition, type=NiagaraPosition}}
     B=Linked() {linkedVariable {name=Particles.SkeletalMeshLocation001.SampledPosition, type=NiagaraPosition}}
     Alpha=$Particles.Initial.MaterialRandom
    )) {isSetParameters=true, module=SetVariables_#3}
   /SetVariables(Particles.MeshOrientation=&MultiplyQuaternion(
     Quaternion A=&FindQuatBetweenTwoVectors(From Vector=<0 1 0>, To Vector=<0 0 0>)
     Quaternion B=&LerpQuaternion(
      Quaternion A:Quat4f=$Particles.SkeletalMeshLocation.SampledRotation
      Quaternion B:Quat4f=$Particles.SkeletalMeshLocation001.SampledRotation
      Lerp Factor:NiagaraFloat=$Particles.Initial.MaterialRandom
      Enable Spherical Linear Interpolation:NiagaraBool=true
     )
    )) {isSetParameters=true, module=SetVariables_#4}
   &DynamicMaterialParameters(
    Index 0 Param 1=1
    Index 0 Param 3=-0.1
    Module.Param0WriteEnabled=true
    Module.Param1WriteEnabled=true
    Module.Param2WriteEnabled=true
    Module.Param3WriteEnabled=true
   )
   &ScaleMeshSize(Scale Factor=&MakeVector(
     X=1
     Y=&Lerp_Float(A=1, B=&RandomRangeFloat(Minimum=1.25, Maximum=1.1), Alpha=$Particles.NormalizedAge)
     Z=0
    ))
   &FN_Light_Attributes(
    Color=<rgba 0.5 0.125 0.4096 1>
    Create Color Attribute=true
    Create Exponent Attribute=true
    Create Position Attribute=true
    Position=&Add_Vector(
     A=&Add_Vector(
      A=Linked() {linkedVariable {name=Particles.Position, type=NiagaraPosition}}
      B=&Multiply_VectorByFloat(
       Vector=&ReturnMeshOrientationAxis(Mesh Orientation Axis=::YAxis, Mesh Orientation Quat=$Particles.MeshOrientation)
       Float=10
      )
     )
     B=&Multiply_VectorByFloat(
      Vector=&Subtract_Vector(A=$Output.Drift_Horror_Get_Set_of_4_Bones.BonePos2, B=$Output.Drift_Horror_Get_Set_of_4_Bones.BonePos1)
      Float=1
     )
    )
    Scale Color=&Multiply_Float(
     A=&FloatFromCurve(FloatCurve=DataInterface() {propertyValues {Curve {keys [
          {
           arriveTangentWeight=0.9771
           interpMode=RCIM_Cubic
           leaveTangentWeight=0.9771
           tangentMode=RCTM_User
           tangentWeightMode=RCTWM_WeightedBoth
           time=0
           value=1
          }
          {
           arriveTangent=-4.27
           arriveTangentWeight=0.8915
           interpMode=RCIM_Cubic
           leaveTangent=-4.27
           leaveTangentWeight=0.8915
           tangentMode=RCTM_User
           tangentWeightMode=RCTWM_WeightedBoth
           time=1
           value=0
          }
         ]}}})
     B=&Sine(
      Normalized Angle=$Emitter.Age
      Period=&RandomRangeFloat(Minimum=0.05, Maximum=0.2)
      Scale=2
      Bias=&RandomRangeFloat(Minimum=0.5, Maximum=2)
     )
    )
   )
   &UpdateMeshOrientation() {enabled=false}
   &OrientMeshToVector(
    Look At Direction:Vector3f=<1 0 -1>
    Override Quat=false
    Side Axis Direction:Vector3f=<1 0 0>
    Up Axis Direction:Vector3f=<0 0 1>
    UseOptionalSideAxis=true
    UseOptionalUpAxis=true
   ) {enabled=false}
   /Niagara/Modules/Math/PureRoll_Orientation(
    Rotation Rate=50
    Surface Normal:Vector3f=<1 0 0>
    Update Rotation Rate=false
    Velocity:Vector3f=&ReturnMeshOrientationAxis(Mesh Orientation Axis=::XAxis, Mesh Orientation Quat=$Particles.MeshOrientation)
   ) {enabled=false}
   &ScaleColor(
    Scale Alpha=&Multiply_Float(
     A=&FloatFromCurve(FloatCurve=DataInterface() {propertyValues {Curve {keys [
          {
           arriveTangentWeight=0.9771
           interpMode=RCIM_Cubic
           leaveTangentWeight=0.9771
           tangentMode=RCTM_User
           tangentWeightMode=RCTWM_WeightedBoth
           time=0
           value=1
          }
          {
           arriveTangent=-4.27
           arriveTangentWeight=0.8915
           interpMode=RCIM_Cubic
           leaveTangent=-4.27
           leaveTangentWeight=0.8915
           tangentMode=RCTM_User
           tangentWeightMode=RCTWM_WeightedBoth
           time=1
           value=0
          }
         ]}}})
     B=&Sine(
      Normalized Angle=$Emitter.Age
      Period=&RandomRangeFloat(Minimum=0.05, Maximum=0.2)
      Scale=2
      Bias=&RandomRangeFloat(Minimum=0.5, Maximum=5)
     )
    )
    ScaleA=true
    ScaleRGB=true
   )
  ]
  rendererClasses [&NiagaraMeshRendererProperties, &NiagaraLightRendererProperties]
  renderers [
   {rendererClass=&NiagaraMeshRendererProperties, rendererIndex=0}
   {rendererClass=&NiagaraLightRendererProperties, rendererIndex=1}
  ]
 }
 Lightning_Cape_Bottom_Edge002 {
  emitterUpdateScript [
   &EmitterState(
    Life Cycle Mode=::NewEnumerator1
    Loop Duration=&RandomRangeFloat(Minimum=0.5, Maximum=2)
    Recalculate Duration Each Loop=true
    UseLoopDelay=false
   )
   &SpawnBurst_Instantaneous(
    Spawn Count=&MakeCustomIntFromBool(
     Bool=&SetBoolByFloatComparison(A=&RandomRangeFloat(Minimum=0, Maximum=0.5), B=0.25)
     True INT=1
     False INT=2
    )
    Spawn Probability=0.4
    Use Spawn Probability=true
    UseLoopCountLimit=false
   )
   &SpawnRate(SpawnRate=30, Use Spawn Probability=false) {enabled=false}
  ]
  particleSpawnScript [
   &InitializeParticle(
    Color=<rgba 2 0 1.5333 0.1>
    Lifetime Max=0.3
    Lifetime Min=0.2
    Lifetime Mode=::NewEnumerator1
    Mesh Scale Max=<0.75 0.5 0.5>
    Mesh Scale Min=<0.1 0.2 0.2>
    Mesh Scale Mode=::NewEnumerator2
    Position Mode=::NewEnumerator1
    Sprite Size Mode=::NewEnumerator3
    Uniform Sprite Size=2
    UsePositionOffset=false
    Write Material Random=true
   )
   /SetVariables(Particles.Pos Choice:NiagaraInt32=&RandomRangeInt(Minimum=0, Maximum=2)) {isSetParameters=true, module=SetVariables_#1}
   &InitialMeshOrientation(Mesh Orientation Mode=::NewEnumerator7, Rotation=<0 0.25 -0.25>, Use Rotation Vector=true)
   &InitialMeshOrientation(Mesh Orientation Mode=::NewEnumerator7, Rotation=<0.35 0.25 0>, Use Rotation Vector=true) {enabled=false, module=InitialMeshOrientation001}
   &InitialMeshOrientation(
    Mesh Orientation Mode=::NewEnumerator7
    Rotation=&RandomRangeVector(Minimum=<0 -0.17 0>, Maximum=<0 0.15 0>)
    Use Rotation Vector=true
   ) {enabled=false, module=InitialMeshOrientation002}
  ]
  particleUpdateScript [
   &ParticleState()
   &SkeletalMeshLocation(
    Bone / Socket Index=$Particles.Initial.Pos Choice
    Bone Sampling Mode=::NewEnumerator4
    MaskSpawnGroup=false
    Orientation Sampling=::NewEnumerator1
    Position Sampling=::NewEnumerator2
    Skeletal Mesh=$User.SkeletalMesh
   )
   &SkeletalMeshLocation(
    Bone / Socket Index=&Add_Integer(A=$Particles.Initial.Pos Choice, B=1)
    Bone Sampling Mode=::NewEnumerator4
    MaskSpawnGroup=false
    Orientation Sampling=::NewEnumerator1
    Position Sampling=::NewEnumerator2
    Skeletal Mesh=$User.SkeletalMesh
   ) {module=SkeletalMeshLocation001}
   /SetVariables(Particles.Position=&Lerp_Vector(
     A=Linked() {linkedVariable {name=Particles.SkeletalMeshLocation.SampledPosition, type=NiagaraPosition}}
     B=Linked() {linkedVariable {name=Particles.SkeletalMeshLocation001.SampledPosition, type=NiagaraPosition}}
     Alpha=$Particles.Initial.MaterialRandom
    )) {isSetParameters=true, module=SetVariables_#3}
   /SetVariables(Particles.MeshOrientation=&MultiplyQuaternion(
     Quaternion A=&FindQuatBetweenTwoVectors(From Vector=<0 1 0>, To Vector=<0 0 0>)
     Quaternion B=&LerpQuaternion(
      Quaternion A:Quat4f=$Particles.SkeletalMeshLocation.SampledRotation
      Quaternion B:Quat4f=$Particles.SkeletalMeshLocation001.SampledRotation
      Lerp Factor:NiagaraFloat=$Particles.Initial.MaterialRandom
      Enable Spherical Linear Interpolation:NiagaraBool=true
     )
    )) {isSetParameters=true, module=SetVariables_#4}
   &DynamicMaterialParameters(
    Index 0 Param 1=1
    Index 0 Param 3=-0.1
    Module.Param0WriteEnabled=true
    Module.Param1WriteEnabled=true
    Module.Param2WriteEnabled=true
    Module.Param3WriteEnabled=true
   )
   &ScaleMeshSize(Scale Factor=&MakeVector(
     X=&Lerp_Float(A=1, B=&RandomRangeFloat(Minimum=1.1, Maximum=1.2), Alpha=$Particles.NormalizedAge)
     Y=&Lerp_Float(A=1, B=&RandomRangeFloat(Minimum=1, Maximum=1), Alpha=$Particles.NormalizedAge)
     Z=1
    ))
   &FN_Light_Attributes(
    Color=<rgba 0.8 0.2 0.6554 1>
    Create Color Attribute=true
    Create Exponent Attribute=true
    Create Position Attribute=true
    Position=&Add_Vector(
     A=&Add_Vector(
      A=Linked() {linkedVariable {name=Particles.Position, type=NiagaraPosition}}
      B=&Multiply_VectorByFloat(
       Vector=&ReturnMeshOrientationAxis(Mesh Orientation Axis=::XAxis, Mesh Orientation Quat=$Particles.MeshOrientation)
       Float=10
      )
     )
     B=&Multiply_VectorByFloat(
      Vector=&Subtract_Vector(A=$Output.Drift_Horror_Get_Set_of_4_Bones.BonePos2, B=$Output.Drift_Horror_Get_Set_of_4_Bones.BonePos1)
      Float=1
     )
    )
    Scale Color=&Multiply_Float(
     A=&FloatFromCurve(FloatCurve=DataInterface() {propertyValues {Curve {keys [
          {
           arriveTangentWeight=0.9771
           interpMode=RCIM_Cubic
           leaveTangentWeight=0.9771
           tangentMode=RCTM_User
           tangentWeightMode=RCTWM_WeightedBoth
           time=0
           value=1
          }
          {
           arriveTangent=-4.27
           arriveTangentWeight=0.8915
           interpMode=RCIM_Cubic
           leaveTangent=-4.27
           leaveTangentWeight=0.8915
           tangentMode=RCTM_User
           tangentWeightMode=RCTWM_WeightedBoth
           time=1
           value=0
          }
         ]}}})
     B=&Sine(
      Normalized Angle=$Emitter.Age
      Period=&RandomRangeFloat(Minimum=0.05, Maximum=0.2)
      Scale=2
      Bias=&RandomRangeFloat(Minimum=0.5, Maximum=2)
     )
    )
   )
   &UpdateMeshOrientation() {enabled=false}
   &OrientMeshToVector(
    Look At Direction:Vector3f=<1 0 -1>
    Override Quat=false
    Side Axis Direction:Vector3f=<1 0 0>
    Up Axis Direction:Vector3f=<0 0 1>
    UseOptionalSideAxis=true
    UseOptionalUpAxis=true
   ) {enabled=false}
   &PureRoll_Orientation(
    Rotation Rate=50
    Surface Normal:Vector3f=<1 0 0>
    Update Rotation Rate=false
    Velocity:Vector3f=&ReturnMeshOrientationAxis(Mesh Orientation Axis=::XAxis, Mesh Orientation Quat=$Particles.MeshOrientation)
   ) {enabled=false}
   &ScaleColor(
    Scale Alpha=&Multiply_Float(
     A=&FloatFromCurve(FloatCurve=DataInterface() {propertyValues {Curve {keys [
          {
           arriveTangentWeight=0.9771
           interpMode=RCIM_Cubic
           leaveTangentWeight=0.9771
           tangentMode=RCTM_User
           tangentWeightMode=RCTWM_WeightedBoth
           time=0
           value=1
          }
          {
           arriveTangent=-4.27
           arriveTangentWeight=0.8915
           interpMode=RCIM_Cubic
           leaveTangent=-4.27
           leaveTangentWeight=0.8915
           tangentMode=RCTM_User
           tangentWeightMode=RCTWM_WeightedBoth
           time=1
           value=0
          }
         ]}}})
     B=&Sine(
      Normalized Angle=$Emitter.Age
      Period=&RandomRangeFloat(Minimum=0.05, Maximum=0.2)
      Scale=2
      Bias=&RandomRangeFloat(Minimum=0.5, Maximum=5)
     )
    )
    ScaleA=true
    ScaleRGB=true
   )
  ]
  rendererClasses [&NiagaraMeshRendererProperties, &NiagaraLightRendererProperties]
  renderers [
   {rendererClass=&NiagaraMeshRendererProperties, rendererIndex=0}
   {rendererClass=&NiagaraLightRendererProperties, rendererIndex=1}
  ]
 }
 Lightning_Left_Arm001 {
  emitterUpdateScript [
   &EmitterState(
    Life Cycle Mode=::NewEnumerator1
    Loop Duration=&RandomRangeFloat(Minimum=0.5, Maximum=2)
    Recalculate Duration Each Loop=true
    UseLoopDelay=false
   )
   &SpawnBurst_Instantaneous(
    Spawn Count=&MakeCustomIntFromBool(
     Bool=&SetBoolByFloatComparison(A=&RandomRangeFloat(Minimum=0, Maximum=0.5), B=0.25)
     True INT=1
     False INT=2
    )
    Spawn Probability=0.5
    Use Spawn Probability=true
    UseLoopCountLimit=false
   )
   &SpawnRate(SpawnRate=30, Use Spawn Probability=false) {enabled=false}
  ]
  particleSpawnScript [
   &InitializeParticle(
    Color=<rgba 2 0 1.5333 1>
    Lifetime=0.15
    Mesh Scale=&Multiply_VectorByFloat(
     Vector=&RandomRangeVector(Minimum=<0.75 0.75 0.75>, Maximum=<0.5 0.5 0.5>)
     Float=&FloatFromCurve(FloatCurve=DataInterface() {propertyValues {Curve {keys [{arriveTangent=-3.9075, interpMode=RCIM_Cubic, leaveTangent=-3.9075, tangentMode=RCTM_User, time=0, value=1}, <at 1 0 RCIM_Cubic>]}}})
    )
    Mesh Scale Mode=::NewEnumerator4
    Position Mode=::NewEnumerator1
    Sprite Size Mode=::NewEnumerator3
    Uniform Sprite Size=2
    UsePositionOffset=false
    Write Material Random=true
   )
   &InitialMeshOrientation(
    Mesh Orientation Mode=::NewEnumerator7
    Rotation=&RandomRangeVector(Minimum=<0.25 -0.25 0.1>, Maximum=<-0.05 -0.25 0>)
    Use Rotation Vector=true
   )
   &InitialMeshOrientation(Rotation=&RandomRangeVector(Minimum=<0 0 0>, Maximum=<0.25 0 0>), Use Rotation Vector=true) {enabled=false, module=InitialMeshOrientation001}
  ]
  particleUpdateScript [
   &ParticleState()
   &UpdateMeshOrientation(
    Rotation Coordinate Space=::NewEnumerator3
    Rotation Rate=&RandomRangeFloat(Minimum=-0.75, Maximum=0.75)
    Rotation Vector=<1 0 0>
   )
   &SkeletalMeshLocation(
    Bone / Socket Index=4
    Bone Sampling Mode=::NewEnumerator4
    MaskSpawnGroup=false
    Sampled Position Offset=&RandomRangeVector(Minimum=<-2 -1 4>, Maximum=<19 1 4>)
    Skeletal Mesh=$User.SkeletalMesh
    Use Sampled Mesh Attribute as Ribbon ID=false
   )
   &DynamicMaterialParameters(
    Index 0 Param 1=1
    Index 0 Param 3=-0.1
    Module.Param0WriteEnabled=true
    Module.Param1WriteEnabled=true
    Module.Param2WriteEnabled=true
    Module.Param3WriteEnabled=true
   )
   &ScaleMeshSize(Scale Factor=&MakeVector(
     X=&FloatFromCurve(FloatCurve=DataInterface() {propertyValues {Curve {keys [{tangentMode=RCTM_User, time=0, value=0.5}, {tangentMode=RCTM_User, time=0.1, value=1}, <at 1 1.1>]}}})
     Y=&Lerp_Float(A=1, B=&RandomRangeFloat(Minimum=1.2, Maximum=2.5), Alpha=$Particles.NormalizedAge)
     Z=&Lerp_Float(A=1, B=&RandomRangeFloat(Minimum=1.2, Maximum=2.5), Alpha=$Particles.NormalizedAge)
    ))
   &FN_Light_Attributes(
    Color=<rgba 100 25 81.9227 1>
    Create Color Attribute=true
    Create Exponent Attribute=true
    Create Position Attribute=true
    Exponent=3
    Position=&Add_Vector(
     A=Linked() {linkedVariable {name=Particles.Position, type=NiagaraPosition}}
     B=&Multiply_VectorByFloat(
      Vector=&ReturnMeshOrientationAxis(Mesh Orientation Axis=::XAxis, Mesh Orientation Quat=$Particles.MeshOrientation)
      Float=5
     )
    )
   )
   &ScaleColor(
    Scale Alpha=&Multiply_Float(
     A=&FloatFromCurve(FloatCurve=DataInterface() {propertyValues {Curve {keys [
          {
           arriveTangentWeight=0.9771
           interpMode=RCIM_Cubic
           leaveTangentWeight=0.9771
           tangentMode=RCTM_User
           tangentWeightMode=RCTWM_WeightedBoth
           time=0
           value=1
          }
          {
           arriveTangent=-4.27
           arriveTangentWeight=0.8915
           interpMode=RCIM_Cubic
           leaveTangent=-4.27
           leaveTangentWeight=0.8915
           tangentMode=RCTM_User
           tangentWeightMode=RCTWM_WeightedBoth
           time=1
           value=0
          }
         ]}}})
     B=&Sine(
      Normalized Angle=$Emitter.Age
      Period=&RandomRangeFloat(Minimum=0.05, Maximum=0.2)
      Scale=2
      Bias=&RandomRangeFloat(Minimum=0.5, Maximum=5)
     )
    )
    ScaleA=true
    ScaleRGB=true
   )
  ]
  rendererClasses [&NiagaraMeshRendererProperties, &NiagaraLightRendererProperties]
  renderers [
   {rendererClass=&NiagaraMeshRendererProperties, rendererIndex=0}
   {rendererClass=&NiagaraLightRendererProperties, rendererIndex=1}
  ]
 }
 Lightning_Left_Arm002 {
  emitterUpdateScript [
   &EmitterState(
    Life Cycle Mode=::NewEnumerator1
    Loop Duration=&RandomRangeFloat(Minimum=0.5, Maximum=2)
    Recalculate Duration Each Loop=true
    UseLoopDelay=false
   )
   &SpawnBurst_Instantaneous(
    Spawn Count=&MakeCustomIntFromBool(
     Bool=&SetBoolByFloatComparison(A=&RandomRangeFloat(Minimum=0, Maximum=0.5), B=0.25)
     True INT=1
     False INT=2
    )
    Spawn Probability=0.4
    Use Spawn Probability=true
    UseLoopCountLimit=false
   )
   &SpawnRate(SpawnRate=30, Use Spawn Probability=false) {enabled=false}
  ]
  particleSpawnScript [
   &InitializeParticle(
    Color=<rgba 2 0 1.5333 0.1>
    Lifetime Max=0.3
    Lifetime Min=0.2
    Lifetime Mode=::NewEnumerator1
    Mesh Scale=&Multiply_VectorByFloat(
     Vector=&RandomRangeVector(Minimum=<0.2 0.1 0.2>, Maximum=<0.25 0.35 0.2>)
     Float=&FloatFromCurve(FloatCurve=DataInterface() {propertyValues {Curve {keys [
          {arriveTangent=-3.9075, interpMode=RCIM_Cubic, leaveTangent=-3.9075, tangentMode=RCTM_User, time=0, value=1}
          {
           arriveTangentWeight=0.83
           interpMode=RCIM_Cubic
           leaveTangentWeight=0.83
           tangentMode=RCTM_User
           tangentWeightMode=RCTWM_WeightedBoth
           time=1
           value=0
          }
         ]}}})
    )
    Mesh Scale Mode=::NewEnumerator4
    Position Mode=::NewEnumerator1
    Sprite Size Mode=::NewEnumerator3
    Uniform Sprite Size=2
    UsePositionOffset=false
    Write Material Random=true
   )
   &InitialMeshOrientation(
    Mesh Orientation Mode=::NewEnumerator7
    Rotation=&RandomRangeVector(Minimum=<0.35 0.05 0.05>, Maximum=<0.35 -0.05 -0.05>)
    Use Rotation Vector=true
   )
   &InitialMeshOrientation(Rotation=&RandomRangeVector(Minimum=<0 0 0>, Maximum=<0.25 0 0>), Use Rotation Vector=true) {enabled=false, module=InitialMeshOrientation001}
  ]
  particleUpdateScript [
   &ParticleState()
   &OrientMeshToVector(
    Look At Direction:Vector3f=&Subtract_Vector(A=$Output.Drift_Horror_Get_Set_of_4_Bones.BonePos2, B=$Output.Drift_Horror_Get_Set_of_4_Bones.BonePos1)
    Override Quat=false
    Up Axis Direction:Vector3f=<0 0 1>
    UseOptionalSideAxis=false
    UseOptionalUpAxis=false
   ) {enabled=false}
   &SkeletalMeshLocation(
    Bone / Socket Index=4
    Bone Sampling Mode=::NewEnumerator4
    MaskSpawnGroup=false
    Sampled Position Offset=&RandomRangeVector(Minimum=<-2 -1 4>, Maximum=<12 1 6>)
    Skeletal Mesh=$User.SkeletalMesh
    Use Sampled Mesh Attribute as Ribbon ID=false
   )
   /SetVariables(Particles.Position=&Lerp_Vector(
     A=Linked() {linkedVariable {name=Particles.SkeletalMeshLocation.SampledPosition, type=NiagaraPosition}}
     B=$Particles.SkeletalMeshLocation001.SampledPosition
     Alpha=$Particles.Initial.MaterialRandom
    )) {enabled=false, isSetParameters=true, module=SetVariables_#3}
   /SetVariables(Particles.MeshOrientation=&LerpQuaternion(
     Quaternion A:Quat4f=$Particles.SkeletalMeshLocation.SampledRotation
     Quaternion B:Quat4f=$Particles.SkeletalMeshLocation001.SampledRotation
     Lerp Factor:NiagaraFloat=$Particles.Initial.MaterialRandom
     Enable Spherical Linear Interpolation:NiagaraBool=true
    )) {enabled=false, isSetParameters=true, module=SetVariables_#4}
   &OrientMeshToVector(
    Look At Direction:Vector3f=$Particles.CylinderLocation.CylinderNormal
    Override Quat=false
    Up Axis Direction:Vector3f=<0 0 1>
    UseOptionalSideAxis=true
    UseOptionalUpAxis=true
   ) {enabled=false, module=OrientMeshToVector001}
   &DynamicMaterialParameters(
    Index 0 Param 1=1
    Index 0 Param 3=-0.1
    Module.Param0WriteEnabled=true
    Module.Param1WriteEnabled=true
    Module.Param2WriteEnabled=true
    Module.Param3WriteEnabled=true
   )
   &ScaleMeshSize(Scale Factor=&MakeVector(
     X=1
     Y=&Lerp_Float(A=1, B=&RandomRangeFloat(Minimum=1.25, Maximum=1.1), Alpha=$Particles.NormalizedAge)
     Z=0
    ))
   &FN_Light_Attributes(
    Color=<rgba 25 6.25 20.4807 1>
    Create Color Attribute=true
    Create Exponent Attribute=true
    Create Position Attribute=true
    Exponent=3
    Position=&Add_Vector(
     A=Linked() {linkedVariable {name=Particles.Position, type=NiagaraPosition}}
     B=&Multiply_VectorByFloat(
      Vector=&ReturnMeshOrientationAxis(Mesh Orientation Axis=::YAxis, Mesh Orientation Quat=$Particles.MeshOrientation)
      Float=5
     )
    )
    Scale Color=&Multiply_Float(
     A=&FloatFromCurve(FloatCurve=DataInterface() {propertyValues {Curve {keys [
          {
           arriveTangentWeight=0.9771
           interpMode=RCIM_Cubic
           leaveTangentWeight=0.9771
           tangentMode=RCTM_User
           tangentWeightMode=RCTWM_WeightedBoth
           time=0
           value=1
          }
          {
           arriveTangent=-4.27
           arriveTangentWeight=0.8915
           interpMode=RCIM_Cubic
           leaveTangent=-4.27
           leaveTangentWeight=0.8915
           tangentMode=RCTM_User
           tangentWeightMode=RCTWM_WeightedBoth
           time=1
           value=0
          }
         ]}}})
     B=&Sine(
      Normalized Angle=$Emitter.Age
      Period=&RandomRangeFloat(Minimum=0.05, Maximum=0.2)
      Scale=2
      Bias=&RandomRangeFloat(Minimum=0.5, Maximum=2)
     )
    )
   )
   &ScaleColor(
    Scale Alpha=&Multiply_Float(
     A=&FloatFromCurve(FloatCurve=DataInterface() {propertyValues {Curve {keys [
          {
           arriveTangentWeight=0.9771
           interpMode=RCIM_Cubic
           leaveTangentWeight=0.9771
           tangentMode=RCTM_User
           tangentWeightMode=RCTWM_WeightedBoth
           time=0
           value=1
          }
          {
           arriveTangent=-4.27
           arriveTangentWeight=0.8915
           interpMode=RCIM_Cubic
           leaveTangent=-4.27
           leaveTangentWeight=0.8915
           tangentMode=RCTM_User
           tangentWeightMode=RCTWM_WeightedBoth
           time=1
           value=0
          }
         ]}}})
     B=&Sine(
      Normalized Angle=$Emitter.Age
      Period=&RandomRangeFloat(Minimum=0.05, Maximum=0.2)
      Scale=2
      Bias=&RandomRangeFloat(Minimum=0.5, Maximum=5)
     )
    )
    ScaleA=true
    ScaleRGB=true
   )
  ]
  rendererClasses [&NiagaraMeshRendererProperties, &NiagaraLightRendererProperties]
  renderers [
   {rendererClass=&NiagaraMeshRendererProperties, rendererIndex=0}
   {rendererClass=&NiagaraLightRendererProperties, rendererIndex=1}
  ]
 }
}
summary {
 systemName=NS_DriftHorror_Jacket
 emitters [
  {emitterName=Lightning_Body, rendererClasses [&NiagaraMeshRendererProperties, &NiagaraLightRendererProperties]}
  {
   emitterName=Lightning_Cape_Bottom_Edge
   rendererClasses [&NiagaraMeshRendererProperties, &NiagaraLightRendererProperties]
  }
  {
   emitterName=Lightning_Back_Lower
   rendererClasses [&NiagaraMeshRendererProperties, &NiagaraLightRendererProperties]
  }
  {
   emitterName=Lightning_Back_Upper
   rendererClasses [&NiagaraMeshRendererProperties, &NiagaraLightRendererProperties]
  }
  {
   emitterName=Lightning_Left_Arm001
   rendererClasses [&NiagaraMeshRendererProperties, &NiagaraLightRendererProperties]
  }
  {
   emitterName=Lightning_Back_Lower001
   rendererClasses [&NiagaraMeshRendererProperties, &NiagaraLightRendererProperties]
  }
  {
   emitterName=Lightning_Back_Upper001
   rendererClasses [&NiagaraMeshRendererProperties, &NiagaraLightRendererProperties]
  }
  {
   emitterName=Lightning_Left_Arm002
   rendererClasses [&NiagaraMeshRendererProperties, &NiagaraLightRendererProperties]
  }
  {
   emitterName=Lightning_Body001
   rendererClasses [&NiagaraMeshRendererProperties, &NiagaraLightRendererProperties]
  }
  {
   emitterName=Lightning_Cape_Bottom_Edge001
   rendererClasses [&NiagaraMeshRendererProperties, &NiagaraLightRendererProperties]
  }
  {
   emitterName=Lightning_Cape_Bottom_Edge002
   rendererClasses [&NiagaraMeshRendererProperties, &NiagaraLightRendererProperties]
  }
 ]
 userVariables [
  User.ForwardDot {type=NiagaraFloat, defaultValue=0, description=""}
  User.RibbonIDOffset {type=NiagaraInt32}
  User.RightDot {type=NiagaraFloat, defaultValue=0, description=""}
  User.SkeletalMesh {
   type=NiagaraDataInterfaceSkeletalMesh
   defaultValue {
    struct=NiagaraExt_VariableValue_DataInterface
    value {dataInterface=&self:NiagaraDataInterfaceSkeletalMesh_0}
   }
  }
 ]
}
