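# Physics asset for the collar of the male medium Drift Horror Jacket cosmetic body, defining bodies for secondary collar motion.
# Five box bodies cover the back and paired left and right collar pieces, each constrained to its own tiny MSTR anchor capsule with swing limits and tight twist.
# Head and neck capsules approximate the character's volume beside the collar, joined by a constraint with free swing and twist.
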PhysicsAsset /BRCosmetics/Characters/Player/Male/Medium/Bodies/M_MED_Drift_Horror_Jacket/Meshes/M_MED_Drift_Horror_Jacket_Collar_Physics
bodies {
 dyn_collar_bk_01 [{
   shapeType=Box
   extentX=10.3456
   extentY=4.6961
   extentZ=19.7696
   shapeName=dyn_collar_bk_01_box
   center=<11 0 0>
   rotation=<rot -9.9999 0 89.9999>
  }]
 dyn_collar_bk_MSTR [{shapeType=Capsule, length=0.505, radius=0.505, shapeName=dyn_collar_bk_MSTR_capsule, rotation=<rot 0 -90 0>}]
 dyn_collar_l [{
   shapeType=Box
   extentX=10.8466
   extentY=2.3422
   extentZ=12.7679
   shapeName=dyn_collar_l_box
   center=<-4.5 0 2>
   rotation=<rot -21.8859 -26.7967 90>
  }]
 dyn_collar_l_01 [{shapeType=Box, extentX=10, extentY=3.5, extentZ=17, shapeName=dyn_collar_l_01_box, center=<-5 0 1>, rotation=<rot 24.9603 158.2935 -81.5627>}]
 dyn_collar_l_MSTR [{shapeType=Capsule, length=0.505, radius=0.505, shapeName=dyn_collar_l_MSTR_capsule, rotation=<rot 0 -90 0>}]
 dyn_collar_l_MSTR1 [{shapeType=Capsule, length=0.505, radius=0.505, shapeName=dyn_collar_l_MSTR1_capsule, rotation=<rot 0 -90 0>}]
 dyn_collar_r [{
   shapeType=Box
   extentX=11.6193
   extentY=2.3422
   extentZ=12.7679
   shapeName=dyn_collar_r_box
   center=<-4.5 0 2.5>
   rotation=<rot 17.6931 -152.6907 89.9992>
  }]
 dyn_collar_r_01 [{
   shapeType=Box
   extentX=10
   extentY=3.5
   extentZ=17
   shapeName=dyn_collar_r_01_box
   center=<-5 0 1>
   rotation=<rot 24.9958 -148.5912 -88.9012>
  }]
 dyn_collar_r_MSTR [{shapeType=Capsule, length=0.505, radius=0.505, shapeName=dyn_collar_r_MSTR_capsule, rotation=<rot 0 -90 0>}]
 dyn_collar_r_MSTR1 [{shapeType=Capsule, length=0.505, radius=0.505, shapeName=dyn_collar_r_MSTR1_capsule, rotation=<rot 0 -90 0>}]
 head [
  {shapeType=Capsule, length=7.938, radius=5.0566, shapeName=head_capsule, center=<-2.0464 6.8608 -0.0001>, rotation=<rot 0 90 90>}
  {shapeType=Capsule, length=7.7201, radius=9.7162, shapeName=head_capsule, center=<8.2308 0.8067 0.0001>, rotation=<rot 90 0 0>}
  {shapeType=Capsule, length=11.7497, radius=6.687, shapeName=head_capsule, center=<10.4924 -1.1068 0>}
 ]
 neck_02 [{shapeType=Capsule, length=6.3566, radius=6.0332, shapeName=neck_02_capsule, center=<3.4946 -1.856 -0.1394>, rotation=<rot 90 0 7.7168>}]
}
constraints [
 {bone1Name=dyn_collar_l, bone2Name=dyn_collar_l_MSTR, swing1LimitDegrees=15, swing2LimitDegrees=45, twistLimitDegrees=5}
 {bone1Name=dyn_collar_r, bone2Name=dyn_collar_r_MSTR, swing1LimitDegrees=15, swing2LimitDegrees=45, twistLimitDegrees=5}
 {
  bone1Name=head
  bone2Name=neck_02
  swing1LimitDegrees=45
  swing1Motion=Free
  swing2LimitDegrees=45
  swing2Motion=Free
  twistLimitDegrees=45
  twistMotion=Free
 }
 {bone1Name=dyn_collar_bk_01, bone2Name=dyn_collar_bk_MSTR, swing1LimitDegrees=75, swing2LimitDegrees=60, twistLimitDegrees=5}
 {bone1Name=dyn_collar_r_01, bone2Name=dyn_collar_r_MSTR1, swing1LimitDegrees=25, swing2LimitDegrees=45, twistLimitDegrees=5}
 {bone1Name=dyn_collar_l_01, bone2Name=dyn_collar_l_MSTR1, swing1LimitDegrees=25, swing2LimitDegrees=45, twistLimitDegrees=5}
]
