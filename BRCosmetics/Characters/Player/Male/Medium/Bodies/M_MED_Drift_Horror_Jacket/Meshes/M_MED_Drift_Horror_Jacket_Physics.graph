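# Physics asset for the Drift Horror jacket on the medium male player body, simulating the hanging jacket panels.
# Back, side and front capsule chains on each side hang from tiny master bodies on the pelvis, with limited swing and twist.
# Leg, foot and pelvis capsules plus two root spheres, one very large and low, provide body shapes.
# Angularly free constraints cross-link right back segment 3 to its left twin and to right side segment 3.
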PhysicsAsset /BRCosmetics/Characters/Player/Male/Medium/Bodies/M_MED_Drift_Horror_Jacket/Meshes/M_MED_Drift_Horror_Jacket_Physics
bodies {
 calf_l [{shapeType=Capsule, length=30.4321, radius=9.3593, center=<-25 -2 0>, rotation=<rot 90 0 0>}]
 calf_r [{shapeType=Capsule, length=30.4321, radius=9.3593, center=<25 2 0.1223>, rotation=<rot 90 0 0>}]
 dyn_jacketBack_01_l [{shapeType=Capsule, length=9.1, radius=8.8, shapeName=dyn_jacketBack_01_l_capsule, center=<8 0 6.5>, rotation=<rot 0 0 90>}]
 dyn_jacketBack_01_r [{shapeType=Capsule, length=9.1, radius=8.8, shapeName=dyn_jacketBack_01_r_capsule, center=<8 0 6.5>, rotation=<rot 0 0 90>}]
 dyn_jacketBack_02_l [{shapeType=Capsule, length=10.1, radius=8.8, shapeName=dyn_jacketBack_02_l_capsule, center=<8 0 6>, rotation=<rot 0 0 90>}]
 dyn_jacketBack_02_r [{shapeType=Capsule, length=10.1, radius=8.8, shapeName=dyn_jacketBack_02_r_capsule, center=<8 0 6>, rotation=<rot 0 0 90>}]
 dyn_jacketBack_03_l [{shapeType=Capsule, length=12, radius=8.8, shapeName=dyn_jacketBack_03_l_capsule, center=<8 0 6>, rotation=<rot 0 0 90>}]
 dyn_jacketBack_03_r [{shapeType=Capsule, length=12, radius=8.8, shapeName=dyn_jacketBack_03_r_capsule, center=<8 0 6>, rotation=<rot 0 0 90>}]
 dyn_jacketBack_04_l [{shapeType=Capsule, length=16.158, radius=11.9421, shapeName=dyn_jacketBack_04_l_capsule, center=<8 0 7>, rotation=<rot 0 0 90>}]
 dyn_jacketBack_04_r [{shapeType=Capsule, length=16.158, radius=11.9421, shapeName=dyn_jacketBack_04_r_capsule, center=<8 0 7>, rotation=<rot 0 0 95>}]
 dyn_jacketBack_mstr_l [{shapeType=Capsule, length=0.505, radius=0.505, shapeName=dyn_jacketBack_mstr_l_capsule}]
 dyn_jacketBack_mstr_r [{shapeType=Capsule, length=0.505, radius=0.505, shapeName=dyn_jacketBack_mstr_r_capsule}]
 dyn_jacketFront_01_l [{shapeType=Capsule, length=14.394, radius=8.4708, shapeName=dyn_jacketFront_01_l_capsule, center=<9 2.5 8>, rotation=<rot 0 0 -35>}]
 dyn_jacketFront_01_r [{shapeType=Capsule, length=9.6, radius=8.2, shapeName=dyn_jacketFront_01_r_capsule, center=<9 -2.5 8>, rotation=<rot 0 0 25>}]
 dyn_jacketFront_02_l [{shapeType=Capsule, length=15.3556, radius=9.3513, shapeName=dyn_jacketFront_02_l_capsule, center=<9 2.5 8>, rotation=<rot 0 0 -39.9998>}]
 dyn_jacketFront_02_r [{shapeType=Capsule, length=9.6, radius=8.2, shapeName=dyn_jacketFront_02_r_capsule, center=<9 -2.5 8>, rotation=<rot 0 0 25>}]
 dyn_jacketFront_mstr_l [{shapeType=Capsule, length=0.505, radius=0.505, shapeName=dyn_jacketFront_mstr_l_capsule}]
 dyn_jacketFront_mstr_r [{shapeType=Capsule, length=0.505, radius=0.505, shapeName=dyn_jacketFront_mstr_r_capsule}]
 dyn_jacketSide_01_l [{shapeType=Capsule, length=10.4, radius=9.3, shapeName=dyn_jacketSide_01_l_capsule, center=<8 -2 8>, rotation=<rot 0 0 85>}]
 dyn_jacketSide_01_r [{shapeType=Capsule, length=10.4, radius=9.3, shapeName=dyn_jacketSide_01_r_capsule, center=<8 2 8>, rotation=<rot 0 0 95>}]
 dyn_jacketSide_02_l [{shapeType=Capsule, length=10.4723, radius=10.4277, shapeName=dyn_jacketSide_02_l_capsule, center=<8 -5 6.5>, rotation=<rot 0 0 79.9995>}]
 dyn_jacketSide_02_r [{shapeType=Capsule, length=11.3, radius=9.6, shapeName=dyn_jacketSide_02_r_capsule, center=<8 5 6.5>, rotation=<rot 0 0 100>}]
 dyn_jacketSide_03_l [{shapeType=Capsule, length=15.1262, radius=10.5738, shapeName=dyn_jacketSide_03_l_capsule, center=<8 -5 4.5>, rotation=<rot 0 0 79.9985>}]
 dyn_jacketSide_03_r [{shapeType=Capsule, length=16.1, radius=9.6, shapeName=dyn_jacketSide_03_r_capsule, center=<8 5 4.5>, rotation=<rot 0 0 105>}]
 dyn_jacketSide_mstr_l [{shapeType=Capsule, length=0.505, radius=0.505, shapeName=dyn_jacketSide_mstr_l_capsule}]
 dyn_jacketSide_mstr_r [{shapeType=Capsule, length=0.505, radius=0.505, shapeName=dyn_jacketSide_mstr_r_capsule}]
 foot_l [{shapeType=Capsule, length=17.5237, radius=11.7507, center=<-3.4271 8.491 0.1091>, rotation=<rot 0.7349 0.9586 -85.8278>}]
 foot_r [{shapeType=Capsule, length=21.5376, radius=11.7675, center=<3.4917 -9.9816 -0.1619>, rotation=<rot 0.7221 1.1909 -85.9009>}]
 pelvis [{shapeType=Capsule, length=10.9249, radius=11.9955, center=<-3.1833 -1.2489 0>}]
 root [
  {shapeType=Sphere, radius=187.1686, shapeName=root_sphere, center=<-3.4923 -2.7141 -174.8455>}
  {shapeType=Sphere, radius=24.9585, shapeName=root_sphere, center=<0 2.4787 8.5972>}
 ]
 thigh_l [{shapeType=Capsule, length=30.3476, radius=9.7327, shapeName=thigh_l_capsule, center=<-24 1 0>, rotation=<rot 90 0 -1.6898>}]
 thigh_r [{shapeType=Capsule, length=30.3476, radius=9.7327, shapeName=thigh_r_capsule, center=<24 -1 0>, rotation=<rot 90 0 -1.6898>}]
}
constraints [
 {bone1Name=calf_l, bone2Name=thigh_l, swing1LimitDegrees=45, swing2LimitDegrees=45, twistLimitDegrees=45}
 {bone1Name=foot_l, bone2Name=calf_l, swing1LimitDegrees=45, swing2LimitDegrees=45, twistLimitDegrees=45}
 {bone1Name=calf_r, bone2Name=thigh_r, swing1LimitDegrees=45, swing2LimitDegrees=45, twistLimitDegrees=45}
 {bone1Name=foot_r, bone2Name=calf_r, swing1LimitDegrees=45, swing2LimitDegrees=45, twistLimitDegrees=45}
 {
  bone1Name=thigh_l
  bone2Name=pelvis
  swing1LimitDegrees=45
  swing1Motion=Free
  swing2LimitDegrees=45
  swing2Motion=Free
  twistLimitDegrees=45
  twistMotion=Free
 }
 {
  bone1Name=thigh_r
  bone2Name=pelvis
  swing1LimitDegrees=45
  swing1Motion=Free
  swing2LimitDegrees=45
  swing2Motion=Free
  twistLimitDegrees=45
  twistMotion=Free
 }
 {
  bone1Name=pelvis
  bone2Name=root
  swing1LimitDegrees=45
  swing1Motion=Free
  swing2LimitDegrees=45
  swing2Motion=Free
  twistLimitDegrees=45
  twistMotion=Free
 }
 {bone1Name=dyn_jacketBack_04_l, bone2Name=dyn_jacketBack_03_l, swing1LimitDegrees=35, swing2LimitDegrees=35, twistLimitDegrees=15}
 {bone1Name=dyn_jacketBack_mstr_l, bone2Name=pelvis, swing1LimitDegrees=45, swing2LimitDegrees=45, twistLimitDegrees=15}
 {bone1Name=dyn_jacketBack_01_l, bone2Name=dyn_jacketBack_mstr_l, swing1LimitDegrees=35, swing2LimitDegrees=35, twistLimitDegrees=15}
 {bone1Name=dyn_jacketBack_02_l, bone2Name=dyn_jacketBack_01_l, swing1LimitDegrees=35, swing2LimitDegrees=35, twistLimitDegrees=15}
 {bone1Name=dyn_jacketBack_03_l, bone2Name=dyn_jacketBack_02_l, swing1LimitDegrees=35, swing2LimitDegrees=35, twistLimitDegrees=15}
 {bone1Name=dyn_jacketBack_04_r, bone2Name=dyn_jacketBack_03_r, swing1LimitDegrees=35, swing2LimitDegrees=35, twistLimitDegrees=15}
 {bone1Name=dyn_jacketBack_mstr_r, bone2Name=pelvis, swing1LimitDegrees=45, swing2LimitDegrees=45, twistLimitDegrees=15}
 {bone1Name=dyn_jacketBack_01_r, bone2Name=dyn_jacketBack_mstr_r, swing1LimitDegrees=35, swing2LimitDegrees=35, twistLimitDegrees=15}
 {bone1Name=dyn_jacketBack_02_r, bone2Name=dyn_jacketBack_01_r, swing1LimitDegrees=35, swing2LimitDegrees=35, twistLimitDegrees=15}
 {bone1Name=dyn_jacketBack_03_r, bone2Name=dyn_jacketBack_02_r, swing1LimitDegrees=35, swing2LimitDegrees=35, twistLimitDegrees=15}
 {bone1Name=dyn_jacketSide_03_r, bone2Name=dyn_jacketSide_02_r, swing1LimitDegrees=35, swing2LimitDegrees=35, twistLimitDegrees=15}
 {bone1Name=dyn_jacketSide_mstr_r, bone2Name=pelvis, swing1LimitDegrees=45, swing2LimitDegrees=45, twistLimitDegrees=15}
 {bone1Name=dyn_jacketSide_01_r, bone2Name=dyn_jacketSide_mstr_r, swing1LimitDegrees=35, swing2LimitDegrees=35, twistLimitDegrees=15}
 {bone1Name=dyn_jacketSide_02_r, bone2Name=dyn_jacketSide_01_r, swing1LimitDegrees=35, swing2LimitDegrees=35, twistLimitDegrees=15}
 {bone1Name=dyn_jacketSide_03_l, bone2Name=dyn_jacketSide_02_l, swing1LimitDegrees=35, swing2LimitDegrees=35, twistLimitDegrees=15}
 {bone1Name=dyn_jacketSide_mstr_l, bone2Name=pelvis, swing1LimitDegrees=45, swing2LimitDegrees=45, twistLimitDegrees=15}
 {bone1Name=dyn_jacketSide_01_l, bone2Name=dyn_jacketSide_mstr_l, swing1LimitDegrees=35, swing2LimitDegrees=35, twistLimitDegrees=15}
 {bone1Name=dyn_jacketSide_02_l, bone2Name=dyn_jacketSide_01_l, swing1LimitDegrees=35, swing2LimitDegrees=35, twistLimitDegrees=15}
 {bone1Name=dyn_jacketFront_02_l, bone2Name=dyn_jacketFront_01_l, swing1LimitDegrees=45, swing2LimitDegrees=45, twistLimitDegrees=15}
 {bone1Name=dyn_jacketFront_mstr_l, bone2Name=pelvis, swing1LimitDegrees=45, swing2LimitDegrees=45, twistLimitDegrees=15}
 {bone1Name=dyn_jacketFront_01_l, bone2Name=dyn_jacketFront_mstr_l, swing1LimitDegrees=45, swing2LimitDegrees=45, twistLimitDegrees=15}
 {bone1Name=dyn_jacketFront_02_r, bone2Name=dyn_jacketFront_01_r, swing1LimitDegrees=35, swing2LimitDegrees=35, twistLimitDegrees=15}
 {bone1Name=dyn_jacketFront_mstr_r, bone2Name=pelvis, swing1LimitDegrees=45, swing2LimitDegrees=45, twistLimitDegrees=15}
 {bone1Name=dyn_jacketFront_01_r, bone2Name=dyn_jacketFront_mstr_r, swing1LimitDegrees=35, swing2LimitDegrees=35, twistLimitDegrees=15}
 {
  bone1Name=dyn_jacketBack_03_r
  bone2Name=dyn_jacketBack_03_l
  swing1LimitDegrees=45
  swing1Motion=Free
  swing2LimitDegrees=45
  swing2Motion=Free
  twistLimitDegrees=45
  twistMotion=Free
 }
 {
  bone1Name=dyn_jacketBack_03_r
  bone2Name=dyn_jacketSide_03_r
  swing1LimitDegrees=45
  swing1Motion=Free
  swing2LimitDegrees=45
  swing2Motion=Free
  twistLimitDegrees=45
  twistMotion=Free
 }
]
