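# Physics asset for the armband on the medium male Drift Horror Jacket player character, covering its dynamic bones and the right arm.
# Two two-bone capsule chains, a and b, hang from a master bone with tight twist and moderate swing limits.
# The master bone joins the right upper-arm twist bone with free swing and twist.
# Larger capsules cover the upper and lower arm twist bones, the latter unconstrained.
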PhysicsAsset /Game/Characters/Player/Male/Medium/Bodies/M_MED_Drift_Horror_Jacket/Meshes/M_MED_Drift_Horror_Jacket_armband_Physics
bodies {
 dyn_armband_MSTR [{shapeType=Capsule, length=0.505, radius=0.505, shapeName=dyn_armband_MSTR_capsule, rotation=<rot 0 -90 0>}]
 dyn_armband_a_01 [{shapeType=Capsule, length=1.8444, radius=1.1726, shapeName=dyn_armband_a_01_capsule, center=<2.2136 0 0>, rotation=<rot 90 0 0>}]
 dyn_armband_a_02 [{shapeType=Capsule, length=3.753, radius=1.7511, shapeName=dyn_armband_a_02_capsule, center=<1.6094 0 0>, rotation=<rot 90 0 0>}]
 dyn_armband_b_01 [{shapeType=Capsule, length=2.1467, radius=1.7876, shapeName=dyn_armband_b_01_capsule, center=<3.8546 0 0>, rotation=<rot 90 0 0>}]
 dyn_armband_b_02 [{shapeType=Capsule, length=3.4618, radius=1.7232, shapeName=dyn_armband_b_02_capsule, center=<2.0924 0 0>, rotation=<rot 90 0 0>}]
 lowerarm_twist_02_r [{shapeType=Capsule, length=11.9897, radius=5.6738, shapeName=lowerarm_twist_02_r_capsule, center=<-1.9904 -0.0001 0>, rotation=<rot 90 0 0>}]
 upperarm_twist_02_r [{shapeType=Capsule, length=13.6328, radius=5.1654, center=<-7.5512 -0.2882 1.1541>, rotation=<rot -90 0.0198 -0.0198>}]
}
constraints [
 {bone1Name=dyn_armband_b_02, bone2Name=dyn_armband_b_01, swing1LimitDegrees=15, swing2LimitDegrees=30, twistLimitDegrees=5}
 {
  bone1Name=dyn_armband_MSTR
  bone2Name=upperarm_twist_02_r
  swing1LimitDegrees=45
  swing1Motion=Free
  swing2LimitDegrees=45
  swing2Motion=Free
  twistLimitDegrees=45
  twistMotion=Free
 }
 {bone1Name=dyn_armband_a_01, bone2Name=dyn_armband_MSTR, swing1LimitDegrees=20, swing2LimitDegrees=30, twistLimitDegrees=5}
 {bone1Name=dyn_armband_a_02, bone2Name=dyn_armband_a_01, swing1LimitDegrees=20, swing2LimitDegrees=30, twistLimitDegrees=5}
 {bone1Name=dyn_armband_b_01, bone2Name=dyn_armband_MSTR, swing1LimitDegrees=15, swing2LimitDegrees=30, twistLimitDegrees=5}
]
