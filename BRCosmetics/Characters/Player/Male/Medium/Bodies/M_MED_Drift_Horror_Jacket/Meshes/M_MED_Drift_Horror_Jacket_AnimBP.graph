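# Post-processes the Drift Horror Jacket cosmetic body, adding secondary motion to a pose copied from the attached parent mesh and adjusted by deform Control Rigs.
# bRigidBodyEnabled selects between RigidBody jacket simulation and a procedural alternative.
# The alternative layers Look At and additive bone offsets driven by thigh, crouch, jump and skydive variables, then directional Trail controllers.
# A jacket Control Rig follows, then AnimDynamics for boots, zipper, armbands and sleeves, plus RigidBody collar.
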AnimBlueprint /BRCosmetics/Characters/Player/Male/Medium/Bodies/M_MED_Drift_Horror_Jacket/Meshes/M_MED_Drift_Horror_Jacket_AnimBP
target_skeleton=&/M_MED_Drift_Horror_Jacket_Skeleton
graphs [
 AnimGraph {nodes [
   n1 = Get isNotCrouchRunning() -> (isNotCrouchRunning:double)
   "Transform (Modify) Bone - Bone: dyn_jacketBack_03_l"(Alpha=@n1, Translation="0,0,0", Rotation="25,0,0.000", Scale="1,1,1") {
    rotationMode=BMM_Additive
    rotationSpace=BCS_BoneSpace
    alphaBoolBlend {blendOption=Linear}
    alphaScaleBiasClamp {bInterpResult=true, interpSpeedDecreasing=2, interpSpeedIncreasing=5}
    boneToModify {boneName=dyn_jacketBack_03_l}
    rotation {pitch=25}
   }
   n2 = Get thigh_rot_up_l() -> (thigh_rot_up_l:double)
   "Transform (Modify) Bone - Bone: dyn_jacketFront_01_l"(Alpha=@n2, Translation="0,0,0", Rotation="10,-15.000,-35", Scale="1,1,1") {
    rotationMode=BMM_Additive
    rotationSpace=BCS_BoneSpace
    translationMode=BMM_Additive
    translationSpace=BCS_BoneSpace
    alphaBoolBlend {blendOption=Linear}
    boneToModify {boneName=dyn_jacketFront_01_l}
    rotation=<rot 10 -15 -35>
   }
   "Transform (Modify) Bone - Bone: dyn_jacketSide_01_l"(Alpha=@n2, Translation="0,-2,3", Rotation="0,-10,0", Scale="1,1,1") {
    rotationMode=BMM_Additive
    rotationSpace=BCS_BoneSpace
    translationMode=BMM_Additive
    translationSpace=BCS_BoneSpace
    alphaBoolBlend {blendOption=Linear}
    boneToModify {boneName=dyn_jacketSide_01_l}
    rotation {yaw=-10}
    translation {y=-2, z=3}
   }
   RigidBody(Alpha=Get targeting() -> (targeting:double)) {
    bClampLinearTranslationLimitToRefPose=true
    bOverrideWorldGravity=true
    overridePhysicsAsset=/Game/Characters/Player/Male/Medium/Bodies/M_MED_Drift_Horror_Jacket/Meshes/M_MED_Drift_Horror_Jacket_armband_Physics
    simulationSpace=BaseBoneSpace
    alphaBoolBlend {blendOption=Linear}
    baseBoneRef {boneName=root}
    componentAppliedLinearAccClamp=<3000 3000 3000>
    componentLinearAccScale=<0.3 0.3 0.4>
    componentLinearVelScale=<0.3 0.3 0.4>
    overrideWorldGravity {z=-980}
   }
   n3 = "Transform (Modify) Bone - Bone: dyn_jacketBack_mstr_r"(ComponentPose=Local To Component(LocalPose=Use cached pose 'prerigid'()), Alpha=1, Translation="0,-5,0", Rotation="0,0,0", Scale="1,1,1") {translationMode=BMM_Additive, alphaBoolBlend {blendOption=Linear}, boneToModify {boneName=dyn_jacketBack_mstr_r}, translation {y=-5}}
   n4 = "Transform (Modify) Bone - Bone: dyn_jacketBack_mstr_l"(ComponentPose=@n3, Alpha=1, Translation="0,-5,0", Rotation="0,0,0", Scale="1,1,1") {translationMode=BMM_Additive, alphaBoolBlend {blendOption=Linear}, boneToModify {boneName=dyn_jacketBack_mstr_l}, translation {y=-5}}
   n5 = "Transform (Modify) Bone - Bone: dyn_jacketSide_mstr_r"(ComponentPose=@n4, Alpha=1, Translation="-3,-5,0", Rotation="0,0,0", Scale="1,1,1") {translationMode=BMM_Additive, alphaBoolBlend {blendOption=Linear}, boneToModify {boneName=dyn_jacketSide_mstr_r}, translation=<-3 -5>}
   n6 = "Transform (Modify) Bone - Bone: dyn_jacketSide_mstr_l"(ComponentPose=@n5, Alpha=1, Translation="3,-5,0", Rotation="0,0,0", Scale="1,1,1") {translationMode=BMM_Additive, alphaBoolBlend {blendOption=Linear}, boneToModify {boneName=dyn_jacketSide_mstr_l}, translation=<3 -5>}
   n7 = RigidBody!(ComponentPose=@n6, Alpha, OverrideWorldGravity:Vector=Get GravityOverride(), ExternalForce) -> (Pose) {
    bClampLinearTranslationLimitToRefPose=true
    bOverrideWorldGravity=true
    overridePhysicsAsset=&/M_MED_Drift_Horror_Jacket_Physics
    simulationSpace=BaseBoneSpace
    alphaBoolBlend {blendOption=Linear}
    baseBoneRef {boneName=dyn_root_simspace}
    componentAppliedLinearAccClamp=<3000 3000 3000>
    componentLinearAccScale=<0.25 0.25 0.4>
    componentLinearVelScale=<0.2 0.2 0.2>
   }
   n8 = Component To Local(ComponentPose=@n7)
   n9 = RigidBody(ComponentPose=Local To Component(LocalPose=Use cached pose 'prerigid'()), OverrideWorldGravity=Get GravityOverride()) {
    bClampLinearTranslationLimitToRefPose=true
    bOverrideWorldGravity=true
    overridePhysicsAsset=&M_MED_Drift_Horror_Jacket_Physics
    simulationSpace=BaseBoneSpace
    alphaBoolBlend {blendOption=Linear}
    baseBoneRef {boneName=dyn_root_simspace}
    componentAppliedLinearAccClamp=<3000 3000 3000>
    componentLinearAccScale=<0.25 0.25 0.4>
    componentLinearVelScale=<0.2 0.2 0.2>
   }
   n10 = Component To Local(ComponentPose=@n9)
   n11 = Get bIsSkyDiving()
   n12 = Get thigh_rot_inr_r() -> (thigh_rot_inr_r:double)
   n13 = Look At - Bone: dyn_jacketFront_mstr_r(ComponentPose=Local To Component(LocalPose=Use cached pose 'prerigid'()), Alpha=@n12) {
    alphaBoolBlend {blendOption=Linear}
    alphaScaleBiasClamp {bInterpResult=true}
    boneToModify {boneName=dyn_jacketFront_mstr_r}
    lookAtLocation=<0 -12 5>
    lookAtTarget {boneReference {boneName=calf_r}}
    lookAt_Axis {axis=<1 0>}
    lookUp_Axis {axis {x=0}}
   }
   n14 = Look At - Bone: dyn_jacketSide_mstr_r(ComponentPose=@n13, Alpha=@n12) {
    alphaBoolBlend {blendOption=Linear}
    alphaScaleBiasClamp {bInterpResult=true}
    boneToModify {boneName=dyn_jacketSide_mstr_r}
    lookAtLocation {x=0, z=11}
    lookAtTarget {boneReference {boneName=calf_r}}
    lookAt_Axis {axis=<1 0>}
    lookUp_Axis {axis {x=0}}
   }
   n15 = Get thigh_rot_up_r() -> (thigh_rot_up_r:double)
   n16 = "Transform (Modify) Bone - Bone: dyn_jacketFront_01_r"(ComponentPose=@n14, Alpha=@n15, Translation="0,0,0", Rotation="9.096,4.987,10.829", Scale="1,1,1") {
    rotationMode=BMM_Additive
    rotationSpace=BCS_BoneSpace
    translationMode=BMM_Additive
    translationSpace=BCS_BoneSpace
    alphaBoolBlend {blendOption=Linear}
    boneToModify {boneName=dyn_jacketFront_01_r}
    rotation=<rot 9.096 4.987 10.829>
   }
   n17 = "Transform (Modify) Bone - Bone: dyn_jacketSide_01_r"(ComponentPose=@n16, Alpha=@n15, Translation="0,0,0", Rotation="5.000,0.000,0.000", Scale="1,1,1") {
    rotationMode=BMM_Additive
    rotationSpace=BCS_BoneSpace
    translationMode=BMM_Additive
    translationSpace=BCS_BoneSpace
    alphaBoolBlend {blendOption=Linear}
    boneToModify {boneName=dyn_jacketSide_01_r}
    rotation {pitch=5}
   }
   n18 = Get thigh_rot_inr_l() -> (thigh_rot_inr_l:double)
   n19 = Look At - Bone: dyn_jacketFront_mstr_l(ComponentPose=@n17, Alpha=@n18) {
    alphaBoolBlend {blendOption=Linear}
    alphaScaleBiasClamp {bInterpResult=true}
    boneToModify {boneName=dyn_jacketFront_mstr_l}
    lookAtLocation=<0 12 -5>
    lookAtTarget {boneReference {boneName=calf_l}}
    lookAt_Axis {axis=<1 0>}
    lookUp_Axis {axis {x=0}}
   }
   n20 = Look At - Bone: dyn_jacketSide_mstr_l(ComponentPose=@n19, Alpha=@n18) {
    alphaBoolBlend {blendOption=Linear}
    alphaScaleBiasClamp {bInterpResult=true}
    boneToModify {boneName=dyn_jacketSide_mstr_l}
    lookAtLocation {x=0, z=-11}
    lookAtTarget {boneReference {boneName=calf_l}}
    lookAt_Axis {axis=<1 0>}
    lookUp_Axis {axis {x=0}}
   }
   n21 = Get thigh_rot_up_l()
   n22 = "Transform (Modify) Bone - Bone: dyn_jacketFront_01_l"(ComponentPose=@n20, Alpha=@n21, Translation="0,0,0", Rotation="9.096,-4.987,-10.828", Scale="1,1,1") {
    rotationMode=BMM_Additive
    rotationSpace=BCS_BoneSpace
    translationMode=BMM_Additive
    translationSpace=BCS_BoneSpace
    alphaBoolBlend {blendOption=Linear}
    boneToModify {boneName=dyn_jacketFront_01_l}
    rotation=<rot 9.096 -4.987 -10.8284>
   }
   n23 = "Transform (Modify) Bone - Bone: dyn_jacketSide_01_l"(ComponentPose=@n22, Alpha=@n21, Translation="0,0,0", Rotation="5,0,0", Scale="1,1,1") {
    rotationMode=BMM_Additive
    rotationSpace=BCS_BoneSpace
    translationMode=BMM_Additive
    translationSpace=BCS_BoneSpace
    alphaBoolBlend {blendOption=Linear}
    boneToModify {boneName=dyn_jacketSide_01_l}
    rotation {pitch=5}
   }
   n24 = Get bIsMoving2D()
   n25 = "Transform (Modify) Bone - Bone: dyn_jacketBack_01_l"!(ComponentPose=@n23, bAlphaBoolEnabled:bool=@n24, Translation="0,0,0", Rotation="20,5,0.000", Scale="1,1,1") -> (Pose) {
    alphaInputType=Bool
    rotationMode=BMM_Additive
    rotationSpace=BCS_BoneSpace
    alphaBoolBlend {blendInTime=0.2, blendOption=Linear, blendOutTime=0.2}
    alphaScaleBiasClamp {bInterpResult=true, interpSpeedDecreasing=2, interpSpeedIncreasing=5}
    boneToModify {boneName=dyn_jacketBack_01_l}
    rotation {pitch=20, yaw=5}
   }
   n26 = "Transform (Modify) Bone - Bone: dyn_jacketBack_01_r"!(ComponentPose=@n25, bAlphaBoolEnabled:bool=@n24, Translation="0,0,0", Rotation="20,-5,0.000", Scale="1,1,1") -> (Pose) {
    alphaInputType=Bool
    rotationMode=BMM_Additive
    rotationSpace=BCS_BoneSpace
    alphaBoolBlend {blendInTime=0.2, blendOption=Linear, blendOutTime=0.2}
    alphaScaleBiasClamp {bInterpResult=true, interpSpeedDecreasing=2, interpSpeedIncreasing=5}
    boneToModify {boneName=dyn_jacketBack_01_r}
    rotation {pitch=20, yaw=-5}
   }
   n27 = Get jumpingAlpha() -> (jumpingAlpha:double)
   n28 = "Transform (Modify) Bone - Bone: dyn_jacketBack_01_l"!(ComponentPose=@n26, Alpha:float=@n27, Translation="0,0,0", Rotation="35,10,0.000", Scale="1,1,1") -> (Pose) {
    rotationMode=BMM_Additive
    rotationSpace=BCS_BoneSpace
    alphaBoolBlend {blendOption=Linear}
    alphaScaleBiasClamp {bInterpResult=true, interpSpeedDecreasing=2, interpSpeedIncreasing=5}
    boneToModify {boneName=dyn_jacketBack_01_l}
    rotation {pitch=35, yaw=10}
   }
   n29 = "Transform (Modify) Bone - Bone: dyn_jacketBack_01_r"!(ComponentPose=@n28, Alpha:float=@n27, Translation="0,0,0", Rotation="35,-10,0.000", Scale="1,1,1") -> (Pose) {
    rotationMode=BMM_Additive
    rotationSpace=BCS_BoneSpace
    alphaBoolBlend {blendOption=Linear}
    alphaScaleBiasClamp {bInterpResult=true, interpSpeedDecreasing=2, interpSpeedIncreasing=5}
    boneToModify {boneName=dyn_jacketBack_01_r}
    rotation {pitch=35, yaw=-10}
   }
   n30 = "Transform (Modify) Bone - Bone: dyn_jacketSide_01_l"(ComponentPose=@n29, Alpha=@n27, Translation="0,0,0", Rotation="15,-15,0.000", Scale="1,1,1") {
    rotationMode=BMM_Additive
    rotationSpace=BCS_BoneSpace
    alphaBoolBlend {blendOption=Linear}
    alphaScaleBiasClamp {bInterpResult=true, interpSpeedDecreasing=2, interpSpeedIncreasing=5}
    boneToModify {boneName=dyn_jacketSide_01_l}
    rotation {pitch=15, yaw=-15}
   }
   n31 = "Transform (Modify) Bone - Bone: dyn_jacketSide_01_l"(ComponentPose=@n30, Alpha=@n27, Translation="0,0,0", Rotation="15,15,0.000", Scale="1,1,1") {
    rotationMode=BMM_Additive
    rotationSpace=BCS_BoneSpace
    alphaBoolBlend {blendOption=Linear}
    alphaScaleBiasClamp {bInterpResult=true, interpSpeedDecreasing=2, interpSpeedIncreasing=5}
    boneToModify {boneName=dyn_jacketSide_01_l}
    rotation {pitch=15, yaw=15}
   }
   n32 = Get skydiveAlpha() -> (skydiveAlpha:double)
   n33 = "Transform (Modify) Bone - Bone: dyn_jacketBack_mstr_l"(ComponentPose=@n31, Alpha=@n32, Translation="0,0,0", Rotation="15,0,0.000", Scale="1,1,1") {
    rotationMode=BMM_Additive
    rotationSpace=BCS_BoneSpace
    alphaBoolBlend {blendOption=Linear}
    alphaScaleBiasClamp {bInterpResult=true, interpSpeedDecreasing=2, interpSpeedIncreasing=5}
    boneToModify {boneName=dyn_jacketBack_mstr_l}
    rotation {pitch=15}
   }
   n34 = "Transform (Modify) Bone - Bone: dyn_jacketBack_03_l"(ComponentPose=@n33, Alpha=@n32, Translation="0,0,0", Rotation="20,0,0.000", Scale="1,1,1") {
    rotationMode=BMM_Additive
    rotationSpace=BCS_BoneSpace
    alphaBoolBlend {blendOption=Linear}
    alphaScaleBiasClamp {bInterpResult=true, interpSpeedDecreasing=2, interpSpeedIncreasing=5}
    boneToModify {boneName=dyn_jacketBack_03_l}
    rotation {pitch=20}
   }
   n35 = "Transform (Modify) Bone - Bone: dyn_jacketBack_mstr_r"(ComponentPose=@n34, Alpha=@n32, Translation="0,0,0", Rotation="15,0,0.000", Scale="1,1,1") {
    rotationMode=BMM_Additive
    rotationSpace=BCS_BoneSpace
    alphaBoolBlend {blendOption=Linear}
    alphaScaleBiasClamp {bInterpResult=true, interpSpeedDecreasing=2, interpSpeedIncreasing=5}
    boneToModify {boneName=dyn_jacketBack_mstr_r}
    rotation {pitch=15}
   }
   n36 = Get ParachuteAlpha() -> (ParachuteAlpha:double)
   n37 = "Transform (Modify) Bone - Bone: dyn_jacketBack_01_l"(ComponentPose=@n35, Alpha=@n36, Translation="0,0,0", Rotation="15,10,0.000", Scale="1,1,1") {
    rotationMode=BMM_Additive
    rotationSpace=BCS_BoneSpace
    alphaBoolBlend {blendOption=Linear}
    alphaScaleBiasClamp {bInterpResult=true, interpSpeedDecreasing=2, interpSpeedIncreasing=5}
    boneToModify {boneName=dyn_jacketBack_01_l}
    rotation {pitch=15, yaw=10}
   }
   n38 = "Transform (Modify) Bone - Bone: dyn_jacketBack_01_r"(ComponentPose=@n37, Alpha=@n36, Translation="0,0,0", Rotation="15,-10,0.000", Scale="1,1,1") {
    rotationMode=BMM_Additive
    rotationSpace=BCS_BoneSpace
    alphaBoolBlend {blendOption=Linear}
    alphaScaleBiasClamp {bInterpResult=true, interpSpeedDecreasing=2, interpSpeedIncreasing=5}
    boneToModify {boneName=dyn_jacketBack_01_r}
    rotation {pitch=15, yaw=-10}
   }
   n39 = "Transform (Modify) Bone - Bone: dyn_jacketSide_01_l"(ComponentPose=@n38, Alpha=@n36, Translation="0,0,0", Rotation="15,-15,0.000", Scale="1,1,1") {
    rotationMode=BMM_Additive
    rotationSpace=BCS_BoneSpace
    alphaBoolBlend {blendOption=Linear}
    alphaScaleBiasClamp {bInterpResult=true, interpSpeedDecreasing=2, interpSpeedIncreasing=5}
    boneToModify {boneName=dyn_jacketSide_01_l}
    rotation {pitch=15, yaw=-15}
   }
   n40 = "Transform (Modify) Bone - Bone: dyn_jacketSide_01_l"(ComponentPose=@n39, Alpha=@n36, Translation="0,0,0", Rotation="15,15,0.000", Scale="1,1,1") {
    rotationMode=BMM_Additive
    rotationSpace=BCS_BoneSpace
    alphaBoolBlend {blendOption=Linear}
    alphaScaleBiasClamp {bInterpResult=true, interpSpeedDecreasing=2, interpSpeedIncreasing=5}
    boneToModify {boneName=dyn_jacketSide_01_l}
    rotation {pitch=15, yaw=15}
   }
   n41 = Component To Local(ComponentPose=@n40)
   n42 = "Transform (Modify) Bone - Bone: dyn_jacketFront_01_r"(ComponentPose=Local To Component(LocalPose=Use cached pose 'prerigid'()), Alpha=1, Translation="0,0,0", Rotation="15,0,0", Scale="1,1,1") {
    rotationMode=BMM_Additive
    rotationSpace=BCS_BoneSpace
    translationMode=BMM_Additive
    translationSpace=BCS_BoneSpace
    alphaBoolBlend {blendOption=Linear}
    boneToModify {boneName=dyn_jacketFront_01_r}
    rotation {pitch=15}
   }
   n43 = "Transform (Modify) Bone - Bone: dyn_jacketSide_01_r"(ComponentPose=@n42, Alpha=1, Translation="0,0,0", Rotation="20,0,0", Scale="1,1,1") {
    rotationMode=BMM_Additive
    rotationSpace=BCS_BoneSpace
    translationMode=BMM_Additive
    translationSpace=BCS_BoneSpace
    alphaBoolBlend {blendOption=Linear}
    boneToModify {boneName=dyn_jacketSide_01_r}
    rotation {pitch=20}
   }
   n44 = "Transform (Modify) Bone - Bone: dyn_jacketBack_01_l"(ComponentPose=@n43, Alpha=@n1, Translation="0,0,0", Rotation="10,10,0.000", Scale="1,1,1") {
    rotationMode=BMM_Additive
    rotationSpace=BCS_BoneSpace
    alphaBoolBlend {blendOption=Linear}
    alphaScaleBiasClamp {bInterpResult=true, interpSpeedDecreasing=5}
    boneToModify {boneName=dyn_jacketBack_01_l}
    rotation {pitch=10, yaw=10}
   }
   n45 = "Transform (Modify) Bone - Bone: dyn_jacketBack_01_r"(ComponentPose=@n44, Alpha=@n1, Translation="0,0,0", Rotation="25,-10,0.000", Scale="1,1,1") {
    rotationMode=BMM_Additive
    rotationSpace=BCS_BoneSpace
    alphaBoolBlend {blendOption=Linear}
    alphaScaleBiasClamp {bInterpResult=true, interpSpeedDecreasing=5}
    boneToModify {boneName=dyn_jacketBack_01_r}
    rotation {pitch=25, yaw=-10}
   }
   n46 = "Transform (Modify) Bone - Bone: dyn_jacketBack_02_r"(ComponentPose=@n45, Alpha=@n1, Translation="0,0,1", Rotation="45,0,0.000", Scale="1,1,1") {
    rotationMode=BMM_Additive
    rotationSpace=BCS_BoneSpace
    alphaBoolBlend {blendOption=Linear}
    alphaScaleBiasClamp {bInterpResult=true, interpSpeedDecreasing=5}
    boneToModify {boneName=dyn_jacketBack_02_r}
    rotation {pitch=45}
    translation {z=1}
   }
   n47 = "Transform (Modify) Bone - Bone: dyn_jacketBack_03_r"(ComponentPose=@n46, Alpha=@n1, Translation="0,0,0", Rotation="5,0,0.000", Scale="1,1,1") {
    rotationMode=BMM_Additive
    rotationSpace=BCS_BoneSpace
    alphaBoolBlend {blendOption=Linear}
    alphaScaleBiasClamp {bInterpResult=true, interpSpeedDecreasing=5}
    boneToModify {boneName=dyn_jacketBack_03_r}
    rotation {pitch=5}
   }
   n48 = "Transform (Modify) Bone - Bone: dyn_jacketBack_04_r"(ComponentPose=@n47, Alpha=@n1, Translation="0,0,0", Rotation="-20,0,0.000", Scale="1,1,1") {
    rotationMode=BMM_Additive
    rotationSpace=BCS_BoneSpace
    alphaBoolBlend {blendOption=Linear}
    alphaScaleBiasClamp {bInterpResult=true, interpSpeedDecreasing=5}
    boneToModify {boneName=dyn_jacketBack_04_r}
    rotation {pitch=-20}
   }
   n49 = "Transform (Modify) Bone - Bone: dyn_jacketBack_02_l"(ComponentPose=@n48, Alpha=@n1, Translation="0,0,0", Rotation="15,0,0.000", Scale="1,1,1") {
    rotationMode=BMM_Additive
    rotationSpace=BCS_BoneSpace
    alphaBoolBlend {blendOption=Linear}
    alphaScaleBiasClamp {bInterpResult=true, interpSpeedDecreasing=5}
    boneToModify {boneName=dyn_jacketBack_02_l}
    rotation {pitch=15}
   }
   n50 = "Transform (Modify) Bone - Bone: dyn_jacketBack_03_l"(ComponentPose=@n49, Alpha=@n1, Translation="0,0,0", Rotation="30,0,0.000", Scale="1,1,1") {
    rotationMode=BMM_Additive
    rotationSpace=BCS_BoneSpace
    alphaBoolBlend {blendOption=Linear}
    alphaScaleBiasClamp {bInterpResult=true, interpSpeedDecreasing=5}
    boneToModify {boneName=dyn_jacketBack_03_l}
    rotation {pitch=30}
   }
   n51 = "Transform (Modify) Bone - Bone: dyn_jacketBack_04_l"(ComponentPose=@n50, Alpha=@n1, Translation="0,0,0", Rotation="5,0,0.000", Scale="1,1,1") {
    rotationMode=BMM_Additive
    rotationSpace=BCS_BoneSpace
    alphaBoolBlend {blendOption=Linear}
    alphaScaleBiasClamp {bInterpResult=true, interpSpeedDecreasing=5}
    boneToModify {boneName=dyn_jacketBack_04_l}
    rotation {pitch=5}
   }
   n52 = "Transform (Modify) Bone - Bone: dyn_jacketFront_mstr_r"(ComponentPose=@n51, Alpha=Get thigh_rot_up_r(), Translation="0,0,3", Rotation="0,0,0", Scale="1,1,1") {
    rotationMode=BMM_Additive
    rotationSpace=BCS_BoneSpace
    translationMode=BMM_Additive
    translationSpace=BCS_BoneSpace
    alphaBoolBlend {blendOption=Linear}
    boneToModify {boneName=dyn_jacketFront_mstr_r}
    translation {z=3}
   }
   n53 = Get thigh_rot_otr_r() -> (thigh_rot_otr_r:double)
   n54 = "Transform (Modify) Bone - Bone: dyn_jacketFront_01_r"(ComponentPose=@n52, Alpha=@n53, Translation="0,0,0", Rotation="35,0,0", Scale="1,1,1") {
    rotationMode=BMM_Additive
    rotationSpace=BCS_BoneSpace
    translationMode=BMM_Additive
    translationSpace=BCS_BoneSpace
    alphaBoolBlend {blendOption=Linear}
    boneToModify {boneName=dyn_jacketFront_01_r}
    rotation {pitch=35}
   }
   n55 = "Transform (Modify) Bone - Bone: dyn_jacketFront_01_r"(ComponentPose=@n54, Alpha=@n53, Translation="0,0,2", Rotation="30,0,0", Scale="1,1,1") {
    rotationMode=BMM_Additive
    rotationSpace=BCS_BoneSpace
    translationMode=BMM_Additive
    translationSpace=BCS_BoneSpace
    alphaBoolBlend {blendOption=Linear}
    boneToModify {boneName=dyn_jacketFront_01_r}
    rotation {pitch=30}
    translation {z=2}
   }
   n56 = "Transform (Modify) Bone - Bone: dyn_jacketSide_02_r"(ComponentPose=@n55, Alpha=@n53, Translation="0,0,0", Rotation="30,0,0", Scale="1,1,1") {
    rotationMode=BMM_Additive
    rotationSpace=BCS_BoneSpace
    translationMode=BMM_Additive
    translationSpace=BCS_BoneSpace
    alphaBoolBlend {blendOption=Linear}
    boneToModify {boneName=dyn_jacketSide_02_r}
    rotation {pitch=30}
   }
   n57 = "Transform (Modify) Bone - Bone: dyn_jacketFront_mstr_l"(ComponentPose=@n56, Alpha=1, Translation="-1,3,7", Rotation="5,10,15", Scale="1,1,1") {
    rotationMode=BMM_Additive
    rotationSpace=BCS_BoneSpace
    scaleMode=BMM_Additive
    scaleSpace=BCS_BoneSpace
    translationMode=BMM_Additive
    translationSpace=BCS_BoneSpace
    alphaBoolBlend {blendOption=Linear}
    boneToModify {boneName=dyn_jacketFront_mstr_l}
    rotation=<rot 5 10 15>
    translation=<-1 3 7>
   }
   n58 = "Transform (Modify) Bone - Bone: dyn_jacketFront_02_l"(ComponentPose=@n57, Alpha=@n2, Translation="0,0,0", Rotation="-25,0,0", Scale="1,1,1") {
    rotationMode=BMM_Additive
    rotationSpace=BCS_BoneSpace
    translationMode=BMM_Additive
    translationSpace=BCS_BoneSpace
    alphaBoolBlend {blendOption=Linear}
    boneToModify {boneName=dyn_jacketFront_02_l}
    rotation {pitch=-25}
   }
   n59 = "Transform (Modify) Bone - Bone: dyn_jacketSide_mstr_l"(ComponentPose=@n58, Alpha=1, Translation="-2,-2,0", Rotation="10,5,0", Scale="1,1,1") {
    rotationMode=BMM_Additive
    rotationSpace=BCS_BoneSpace
    scaleMode=BMM_Additive
    scaleSpace=BCS_BoneSpace
    translationMode=BMM_Additive
    translationSpace=BCS_BoneSpace
    alphaBoolBlend {blendOption=Linear}
    boneToModify {boneName=dyn_jacketSide_mstr_l}
    rotation {pitch=10, yaw=5}
    translation=<-2 -2>
   }
   n60 = Get thigh_rot_otr_l() -> (thigh_rot_otr_l:double)
   n61 = "Transform (Modify) Bone - Bone: dyn_jacketFront_01_l"(ComponentPose=@n59, Alpha=@n60, Translation="0,0,0", Rotation="45,0,0", Scale="1,1,1") {
    rotationMode=BMM_Additive
    rotationSpace=BCS_BoneSpace
    translationMode=BMM_Additive
    translationSpace=BCS_BoneSpace
    alphaBoolBlend {blendOption=Linear}
    boneToModify {boneName=dyn_jacketFront_01_l}
    rotation {pitch=45}
   }
   n62 = "Transform (Modify) Bone - Bone: dyn_jacketFront_02_l"(ComponentPose=@n61, Alpha=@n60, Translation="0,0,2", Rotation="30,0,0", Scale="1,1,1") {
    rotationMode=BMM_Additive
    rotationSpace=BCS_BoneSpace
    translationMode=BMM_Additive
    translationSpace=BCS_BoneSpace
    alphaBoolBlend {blendOption=Linear}
    boneToModify {boneName=dyn_jacketFront_02_l}
    rotation {pitch=30}
    translation {z=2}
   }
   n63 = "Transform (Modify) Bone - Bone: dyn_jacketSide_02_l"(ComponentPose=@n62, Alpha=@n60, Translation="0,0,0", Rotation="45,0,0", Scale="1,1,1") {
    rotationMode=BMM_Additive
    rotationSpace=BCS_BoneSpace
    translationMode=BMM_Additive
    translationSpace=BCS_BoneSpace
    alphaBoolBlend {blendOption=Linear}
    boneToModify {boneName=dyn_jacketSide_02_l}
    rotation {pitch=45}
   }
   n64 = Component To Local(ComponentPose=@n63)
   n65 = Get bIsCrouching()
   n66 = Get BwdMovement() -> (BwdMovement:double)
   n67 = Trail controller - Bone: dyn_jacketBack_05_r!(
    ComponentPose=Local To Component(LocalPose=Blend Poses by bool(BlendPose_0=@n64, BlendPose_1=@n41, bActiveValue=@n65) {blendType=Linear})
    Alpha=@n66
    FakeVelocity:Vector="0,0,0"
   ) -> (Pose) {
    bActorSpaceFakeVel=true
    bShowTrailLocation=true
    chainLength=4
    debugLifeTime=0.1
    maxDeltaTime=0.0333
    relaxationSpeedScale=3
    alphaScaleBiasClamp {bInterpResult=true}
    baseJoint {boneName=root}
    planarLimits [
     {drivingBone {boneName=root}, planeTransform {location=<0 0 10>, rotation=<rot 0 0 0>, scale=<1 1 1>}}
     {drivingBone {boneName=spine_03}, planeTransform {location=<0 -20 0>, rotation=<rot 0 15 -89.9999>, scale=<1 1 1>}}
    ]
    rotationLimits [{limitMax=<180 180 180>}, {limitMax=<180 180 180>}, {limitMax=<180 180 180>}, {limitMax=<180 180 180>}]
    rotationOffsets [<0 0 0>, <0 0 0>, <0 0 0>, <0 0 0>]
    trailBone {boneName=dyn_jacketBack_05_r}
    trailRelaxationSpeed {editorCurveData {keys [{arriveTangent=-72.0739, interpMode=RCIM_Cubic, leaveTangent=-72.0739, tangentMode=RCTM_User, time=0, value=35}, <at 1 5 RCIM_Cubic>]}}
   }
   n68 = Trail controller - Bone: dyn_jacketBack_05_l!(ComponentPose=@n67, Alpha=@n66, FakeVelocity:Vector="0,0,0") -> (Pose) {
    bActorSpaceFakeVel=true
    bShowTrailLocation=true
    chainLength=4
    debugLifeTime=0.1
    maxDeltaTime=0.0333
    relaxationSpeedScale=3
    alphaScaleBiasClamp {bInterpResult=true}
    baseJoint {boneName=root}
    planarLimits [
     {drivingBone {boneName=root}, planeTransform {location=<0 0 10>, rotation=<rot 0 0 0>, scale=<1 1 1>}}
     {drivingBone {boneName=spine_03}, planeTransform {location=<0 -20 0>, rotation=<rot 0 15 -89.9999>, scale=<1 1 1>}}
    ]
    rotationLimits [{limitMax=<180 180 180>}, {limitMax=<180 180 180>}, {limitMax=<180 180 180>}, {limitMax=<180 180 180>}]
    rotationOffsets [<0 0 0>, <0 0 0>, <0 0 0>, <0 0 0>]
    trailBone {boneName=dyn_jacketBack_05_l}
    trailRelaxationSpeed {editorCurveData {keys [{arriveTangent=-72.0739, interpMode=RCIM_Cubic, leaveTangent=-72.0739, tangentMode=RCTM_User, time=0, value=35}, <at 1 5 RCIM_Cubic>]}}
   }
   n69 = Get LeftMovement() -> (LeftMovement:double)
   n70 = Trail controller - Bone: dyn_jacketFront_03_r!(ComponentPose=@n68, Alpha=@n69, FakeVelocity:Vector="0,0,0") -> (Pose) {
    bActorSpaceFakeVel=true
    bShowTrailLocation=true
    debugLifeTime=0.1
    maxDeltaTime=0.0333
    relaxationSpeedScale=3
    alphaScaleBiasClamp {bInterpResult=true}
    baseJoint {boneName=root}
    planarLimits [
     {drivingBone {boneName=root}, planeTransform {location=<0 0 10>, rotation=<rot 0 0 0>, scale=<1 1 1>}}
     {drivingBone {boneName=spine_03}, planeTransform {location=<0 -20 0>, rotation=<rot 0 15 -89.9999>, scale=<1 1 1>}}
    ]
    rotationLimits [{limitMax=<180 180 180>}, {limitMax=<180 180 180>}]
    rotationOffsets [<0 0 0>, <0 0 0>]
    trailBone {boneName=dyn_jacketFront_03_r}
    trailRelaxationSpeed {editorCurveData {keys [{arriveTangent=-72.0739, interpMode=RCIM_Cubic, leaveTangent=-72.0739, tangentMode=RCTM_User, time=0, value=35}, <at 1 5 RCIM_Cubic>]}}
   }
   n71 = Trail controller - Bone: dyn_jacketSide_04_r!(ComponentPose=@n70, Alpha=@n69, FakeVelocity:Vector="0,0,0") -> (Pose) {
    bActorSpaceFakeVel=true
    bShowTrailLocation=true
    chainLength=3
    debugLifeTime=0.1
    maxDeltaTime=0.0333
    relaxationSpeedScale=3
    alphaScaleBiasClamp {bInterpResult=true}
    baseJoint {boneName=root}
    planarLimits [
     {drivingBone {boneName=root}, planeTransform {location=<0 0 10>, rotation=<rot 0 0 0>, scale=<1 1 1>}}
     {drivingBone {boneName=spine_03}, planeTransform {location=<0 -20 0>, rotation=<rot 0 15 -89.9999>, scale=<1 1 1>}}
    ]
    rotationLimits [{limitMax=<180 180 180>}, {limitMax=<180 180 180>}, {limitMax=<180 180 180>}]
    rotationOffsets [<0 0 0>, <0 0 0>, <0 0 0>]
    trailBone {boneName=dyn_jacketSide_04_r}
    trailRelaxationSpeed {editorCurveData {keys [{arriveTangent=-72.0739, interpMode=RCIM_Cubic, leaveTangent=-72.0739, tangentMode=RCTM_User, time=0, value=35}, <at 1 5 RCIM_Cubic>]}}
   }
   n72 = Get RightMovement() -> (RightMovement:double)
   n73 = Trail controller - Bone: dyn_jacketFront_03_l!(ComponentPose=@n71, Alpha=@n72, FakeVelocity:Vector="0,0,0") -> (Pose) {
    bActorSpaceFakeVel=true
    bShowTrailLocation=true
    debugLifeTime=0.1
    maxDeltaTime=0.0333
    relaxationSpeedScale=3
    alphaScaleBiasClamp {bInterpResult=true}
    baseJoint {boneName=root}
    planarLimits [
     {drivingBone {boneName=root}, planeTransform {location=<0 0 10>, rotation=<rot 0 0 0>, scale=<1 1 1>}}
     {drivingBone {boneName=spine_03}, planeTransform {location=<0 -20 0>, rotation=<rot 0 15 -89.9999>, scale=<1 1 1>}}
    ]
    rotationLimits [{limitMax=<180 180 180>}, {limitMax=<180 180 180>}]
    rotationOffsets [<0 0 0>, <0 0 0>]
    trailBone {boneName=dyn_jacketFront_03_l}
    trailRelaxationSpeed {editorCurveData {keys [{arriveTangent=-72.0739, interpMode=RCIM_Cubic, leaveTangent=-72.0739, tangentMode=RCTM_User, time=0, value=35}, <at 1 5 RCIM_Cubic>]}}
   }
   n74 = Trail controller - Bone: dyn_jacketSide_04_l!(ComponentPose=@n73, Alpha=@n72, FakeVelocity:Vector="0,0,0") -> (Pose) {
    bActorSpaceFakeVel=true
    bShowTrailLocation=true
    chainLength=3
    debugLifeTime=0.1
    maxDeltaTime=0.0333
    relaxationSpeedScale=3
    alphaScaleBiasClamp {bInterpResult=true}
    baseJoint {boneName=root}
    planarLimits [
     {drivingBone {boneName=root}, planeTransform {location=<0 0 10>, rotation=<rot 0 0 0>, scale=<1 1 1>}}
     {drivingBone {boneName=spine_03}, planeTransform {location=<0 -20 0>, rotation=<rot 0 15 -89.9999>, scale=<1 1 1>}}
    ]
    rotationLimits [{limitMax=<180 180 180>}, {limitMax=<180 180 180>}, {limitMax=<180 180 180>}]
    rotationOffsets [<0 0 0>, <0 0 0>, <0 0 0>]
    trailBone {boneName=dyn_jacketSide_04_l}
    trailRelaxationSpeed {editorCurveData {keys [{arriveTangent=-72.0739, interpMode=RCIM_Cubic, leaveTangent=-72.0739, tangentMode=RCTM_User, time=0, value=35}, <at 1 5 RCIM_Cubic>]}}
   }
   n75 = Control Rig(Source=Blend Poses by bool(
     BlendPose_0=Blend Poses by bool(BlendPose_0=@n8, BlendPose_1=@n10, bActiveValue=@n11) {blendType=Linear}
     BlendPose_1=Component To Local(ComponentPose=@n74)
     bActiveValue=Get bRigidBodyEnabled()
    ) {blendType=Linear}) {
    alphaBoolBlend {blendOption=Linear}
    controlRigAssetReference {blueprintRigClass=&/M_MED_Drift_Horror_Jacket_CtrlRig.M_MED_Drift_Horror_Jacket_CtrlRig_C}
   }
   n76 = Anim Dynamics - Bone: dyn_boot_l!(ComponentPose=Local To Component(LocalPose=@n75), Alpha=0.5, GravityOverride:Vector="0,0,0", ExternalForce="0,0,0") -> (Pose) {
    angularDampingOverride=0.9999
    angularSpringConstant=1500
    bAngularSpring=true
    bGravityOverrideInSimSpace=true
    bOverrideAngularDamping=true
    gravityScale=0.8
    simulationSpace=BoneRelative
    alphaBoolBlend {blendOption=Linear}
    boundBone {boneName=dyn_boot_l}
    componentAppliedLinearAccClamp=<5000 5000 3500>
    componentLinearAccScale=<1.5 1.5 1.25>
    componentLinearVelScale=<1.5 1.5 1.25>
    physicsBodyDefinitions [{
      sphereCollisionRadius=0
      boundBone {boneName=dyn_boot_l}
      boxExtents=<10 20 20>
      constraintSetup {angularLimitsMax=<15 35 15>, angularLimitsMin=<-15 -35 -15>, angularTarget=<0 0 0>}
      localJointOffset=<2 0 0>
     }]
    relativeSpaceBone {boneName=root}
   }
   n77 = Anim Dynamics - Bone: dyn_boot_r!(ComponentPose=@n76, Alpha=0.5, GravityOverride:Vector="0,0,0", ExternalForce="0,0,0") -> (Pose) {
    angularDampingOverride=0.9999
    angularSpringConstant=1500
    bAngularSpring=true
    bGravityOverrideInSimSpace=true
    bOverrideAngularDamping=true
    gravityScale=0.8
    simulationSpace=BoneRelative
    alphaBoolBlend {blendOption=Linear}
    boundBone {boneName=dyn_boot_r}
    componentAppliedLinearAccClamp=<5000 5000 3500>
    componentLinearAccScale=<1.5 1.5 1.25>
    componentLinearVelScale=<1.5 1.5 1.25>
    physicsBodyDefinitions [{
      sphereCollisionRadius=0
      boundBone {boneName=dyn_boot_r}
      boxExtents=<10 20 20>
      constraintSetup {angularLimitsMax=<15 35 15>, angularLimitsMin=<-15 -35 -15>, angularTarget=<0 0 0>}
      localJointOffset=<2 0 0>
     }]
    relativeSpaceBone {boneName=root}
   }
   n78 = Anim Dynamics - Bone: dyn_zipper!(ComponentPose=@n77, Alpha=0.25, GravityOverride:Vector="0,0,0", ExternalForce) -> (Pose) {
    angularDampingOverride=1
    angularSpringConstant=1500
    bAngularSpring=true
    bGravityOverrideInSimSpace=true
    bOverrideAngularDamping=true
    gravityScale=0.45
    simulationSpace=RootRelative
    alphaBoolBlend {blendOption=Linear}
    alphaScaleBias {scale=2}
    boundBone {boneName=dyn_zipper}
    componentAppliedLinearAccClamp=<2500 2500 2500>
    componentLinearAccScale=<1 1 1>
    componentLinearVelScale=<1 1 1>
    physicsBodyDefinitions [{
      sphereCollisionRadius=0
      boundBone {boneName=dyn_zipper}
      boxExtents=<10 5 5>
      constraintSetup {angularLimitsMax=<15 35 15>, angularLimitsMin=<-15 -35 -15>, angularTarget=<0 0 0>}
      localJointOffset=<2 0 0>
     }]
    relativeSpaceBone {boneName=root}
   }
   n79 = RigidBody!(ComponentPose=@n78, Alpha, ExternalForce) -> (Pose) {
    bClampLinearTranslationLimitToRefPose=true
    bOverrideWorldGravity=true
    overridePhysicsAsset=&/M_MED_Drift_Horror_Jacket_Collar_Physics
    simulationSpace=BaseBoneSpace
    alphaBoolBlend {blendOption=Linear}
    baseBoneRef {boneName=spine_05}
    componentAppliedLinearAccClamp=<3000 3000 3000>
    componentLinearAccScale=<0.3 0.3 0.3>
    componentLinearVelScale=<0.3 0.3 0.3>
    overrideWorldGravity {z=-980}
   }
   n80 = Anim Dynamics - Bone: dyn_armband_a_01(ComponentPose=@n79, Alpha=1, ExternalForce="0,0,0") {
    angularDampingOverride=0.9999
    angularSpringConstant=50
    bAngularSpring=true
    bGravityOverrideInSimSpace=true
    bOverrideAngularDamping=true
    bUseGravityOverride=true
    simulationSpace=BoneRelative
    alphaBoolBlend {blendOption=Linear}
    boundBone {boneName=dyn_armband_a_01}
    componentAppliedLinearAccClamp=<3500 3500 3500>
    componentLinearAccScale=<0.8 0.8 0.4>
    componentLinearVelScale=<0.8 0.8 0.4>
    gravityOverride {z=-720}
    physicsBodyDefinitions [{
      sphereCollisionRadius=0
      boundBone {boneName=dyn_armband_a_01}
      boxExtents=<10 10 5>
      constraintSetup {angularLimitsMax=<10 30 25>, angularLimitsMin=<-10 -30 -15>, angularTarget=<0 0 0>}
      localJointOffset=<4 0 0>
     }]
    relativeSpaceBone {boneName=dyn_root_simspace}
   }
   n81 = Anim Dynamics - Bone: dyn_armband_a_02(ComponentPose=@n80, Alpha=1, ExternalForce="0,0,0") {
    angularDampingOverride=0.9999
    angularSpringConstant=50
    bAngularSpring=true
    bGravityOverrideInSimSpace=true
    bOverrideAngularDamping=true
    bUseGravityOverride=true
    simulationSpace=BoneRelative
    alphaBoolBlend {blendOption=Linear}
    boundBone {boneName=dyn_armband_a_02}
    componentAppliedLinearAccClamp=<3500 3500 3500>
    componentLinearAccScale=<0.8 0.8 0.4>
    componentLinearVelScale=<0.8 0.8 0.4>
    gravityOverride {z=-720}
    physicsBodyDefinitions [{
      sphereCollisionRadius=0
      boundBone {boneName=dyn_armband_a_02}
      boxExtents=<10 10 5>
      constraintSetup {angularLimitsMax=<10 30 25>, angularLimitsMin=<-10 -30 -15>, angularTarget=<0 0 0>}
      localJointOffset=<4 0 0>
     }]
    relativeSpaceBone {boneName=dyn_root_simspace}
   }
   n82 = Anim Dynamics - Bone: dyn_armband_b_01(ComponentPose=@n81, Alpha=1, ExternalForce="0,0,0") {
    angularDampingOverride=0.9999
    angularSpringConstant=50
    bAngularSpring=true
    bGravityOverrideInSimSpace=true
    bOverrideAngularDamping=true
    bUseGravityOverride=true
    simulationSpace=BoneRelative
    alphaBoolBlend {blendOption=Linear}
    boundBone {boneName=dyn_armband_b_01}
    componentAppliedLinearAccClamp=<3500 3500 3500>
    componentLinearAccScale=<0.8 0.8 0.4>
    componentLinearVelScale=<0.8 0.8 0.4>
    gravityOverride {z=-720}
    physicsBodyDefinitions [{
      sphereCollisionRadius=0
      boundBone {boneName=dyn_armband_b_01}
      boxExtents=<10 10 5>
      constraintSetup {angularLimitsMax=<10 30 25>, angularLimitsMin=<-10 -30 -5>, angularTarget=<0 0 0>}
      localJointOffset=<4 0 0>
     }]
    relativeSpaceBone {boneName=dyn_root_simspace}
   }
   n83 = Anim Dynamics - Bone: dyn_armband_b_02(ComponentPose=@n82, Alpha=1, ExternalForce="0,0,0") {
    angularDampingOverride=0.9999
    angularSpringConstant=50
    bAngularSpring=true
    bGravityOverrideInSimSpace=true
    bOverrideAngularDamping=true
    bUseGravityOverride=true
    simulationSpace=BoneRelative
    alphaBoolBlend {blendOption=Linear}
    boundBone {boneName=dyn_armband_b_02}
    componentAppliedLinearAccClamp=<3500 3500 3500>
    componentLinearAccScale=<0.8 0.8 0.4>
    componentLinearVelScale=<0.8 0.8 0.4>
    gravityOverride {z=-720}
    physicsBodyDefinitions [{
      sphereCollisionRadius=0
      boundBone {boneName=dyn_armband_b_02}
      boxExtents=<10 10 5>
      constraintSetup {angularLimitsMax=<10 30 25>, angularLimitsMin=<-10 -30 -5>, angularTarget=<0 0 0>}
      localJointOffset=<4 0 0>
     }]
    relativeSpaceBone {boneName=dyn_root_simspace}
   }
   n84 = Anim Dynamics - Bone: dyn_sleeve_01(ComponentPose=@n83, Alpha=1, ExternalForce="0,0,0") {
    angularDampingOverride=0.9999
    angularSpringConstant=50
    bAngularSpring=true
    bGravityOverrideInSimSpace=true
    bOverrideAngularDamping=true
    bUseGravityOverride=true
    simulationSpace=BoneRelative
    alphaBoolBlend {blendOption=Linear}
    boundBone {boneName=dyn_sleeve_01}
    componentAppliedLinearAccClamp=<3500 3500 3500>
    componentLinearAccScale=<0.8 0.8 0.4>
    componentLinearVelScale=<0.8 0.8 0.4>
    gravityOverride {z=-720}
    physicsBodyDefinitions [{
      sphereCollisionRadius=0
      boundBone {boneName=dyn_sleeve_01}
      boxExtents=<10 10 5>
      constraintSetup {angularLimitsMax=<5 15 15>, angularLimitsMin=<-5 -30 -15>, angularTarget=<0 0 0>}
      localJointOffset=<4 0 0>
     }]
    relativeSpaceBone {boneName=dyn_root_simspace}
   }
   n85 = Anim Dynamics - Bone: dyn_sleeve_02(ComponentPose=@n84, Alpha=1, ExternalForce="0,0,0") {
    angularDampingOverride=0.9999
    angularSpringConstant=50
    bAngularSpring=true
    bGravityOverrideInSimSpace=true
    bOverrideAngularDamping=true
    bUseGravityOverride=true
    simulationSpace=BoneRelative
    alphaBoolBlend {blendOption=Linear}
    boundBone {boneName=dyn_sleeve_02}
    componentAppliedLinearAccClamp=<3500 3500 3500>
    componentLinearAccScale=<0.8 0.8 0.4>
    componentLinearVelScale=<0.8 0.8 0.4>
    gravityOverride {z=-810}
    physicsBodyDefinitions [{
      sphereCollisionRadius=0
      boundBone {boneName=dyn_sleeve_02}
      boxExtents=<15 15 5>
      constraintSetup {angularLimitsMax=<5 10 15>, angularLimitsMin=<-5 -30 -15>, angularTarget=<0 0 0>}
      localJointOffset=<3 0 0>
     }]
    relativeSpaceBone {boneName=dyn_root_simspace}
   }
   n86 = Anim Dynamics - Bone: dyn_sleeve_03(ComponentPose=@n85, Alpha=1, ExternalForce="0,0,0") {
    angularDampingOverride=0.9999
    angularSpringConstant=50
    bAngularSpring=true
    bGravityOverrideInSimSpace=true
    bOverrideAngularDamping=true
    bUseGravityOverride=true
    simulationSpace=BoneRelative
    alphaBoolBlend {blendOption=Linear}
    boundBone {boneName=dyn_sleeve_03}
    componentAppliedLinearAccClamp=<3500 3500 3500>
    componentLinearAccScale=<0.8 0.8 0.4>
    componentLinearVelScale=<0.8 0.8 0.4>
    gravityOverride {z=-980}
    physicsBodyDefinitions [{
      sphereCollisionRadius=0
      boundBone {boneName=dyn_sleeve_03}
      boxExtents=<15 10 5>
      constraintSetup {angularLimitsMax=<5 15 15>, angularLimitsMin=<-5 -30 -15>, angularTarget=<0 0 0>}
      localJointOffset=<2 0 0>
     }]
    relativeSpaceBone {boneName=dyn_root_simspace}
   }
   n87 = Anim Dynamics - Bone: dyn_sleeve_04(ComponentPose=@n86, Alpha=1, ExternalForce="0,0,0") {
    angularDampingOverride=0.9999
    angularSpringConstant=50
    bAngularSpring=true
    bGravityOverrideInSimSpace=true
    bOverrideAngularDamping=true
    bUseGravityOverride=true
    simulationSpace=BoneRelative
    alphaBoolBlend {blendOption=Linear}
    boundBone {boneName=dyn_sleeve_04}
    componentAppliedLinearAccClamp=<3500 3500 3500>
    componentLinearAccScale=<0.8 0.8 0.4>
    componentLinearVelScale=<0.8 0.8 0.4>
    gravityOverride {z=-720}
    physicsBodyDefinitions [{
      sphereCollisionRadius=0
      boundBone {boneName=dyn_sleeve_04}
      boxExtents=<10 10 5>
      constraintSetup {angularLimitsMax=<5 15 15>, angularLimitsMin=<-5 -30 -15>, angularTarget=<0 0 0>}
      localJointOffset=<2 0 0>
     }]
    relativeSpaceBone {boneName=dyn_root_simspace}
   }
   Output Pose(Result=Component To Local(ComponentPose=@n87)) {layerGroup=None}
   n88 = Control Rig(Source=Copy Pose From Mesh() {bUseAttachedParent=true}) {
    alphaBoolBlend {blendOption=Linear}
    controlRigAssetReference {blueprintRigClass=/Game/Characters/Player/Male/Medium/Bodies/M_MED_Emperor/Meshes/M_MED_Emperor_Deform_CtrlRig_twists.M_MED_Emperor_Deform_CtrlRig_twists_C}
   }
   n89 = Control Rig!(
    Source=@n88
    Alpha
    KneeRot_fr_l_114deg:Vector
    KneeTrans_fr_l_114deg:Vector="0.258,-3.591,0.009"
    KneeScale_fr_l_114deg:Vector="1,1,1"
    KneeScale_fr_l_130deg:Vector="1,1,1"
    KneeRot_fr_l_130deg:Vector="0,0,-51.2"
    KneeTrans_fr_l_130deg:Vector="0.258,-3.591,0.009"
   ) -> (Pose) {
    alphaBoolBlend {blendOption=Linear}
    controlRigAssetReference {blueprintRigClass=/Game/Characters/Player/Male/Medium/Bodies/M_MED_Emperor/Meshes/M_MED_Emperor_Deform_CtrlRig_sculpters.M_MED_Emperor_Deform_CtrlRig_sculpters_C}
   }
   n90 = "Constraint - dyn_jacketFront_mstr_l (PPPPPPPPPP)"!(
    ComponentPose=Local To Component(LocalPose=@n89)
    Alpha=1
    ConstraintWeights_0=0.0246
    ConstraintWeights_1:float=0.0146
    ConstraintWeights_2:float=0.4309
    ConstraintWeights_3:float=0.1507
    ConstraintWeights_4:float=0.0505
    ConstraintWeights_5:float=0.0333
    ConstraintWeights_6:float=0.0019
    ConstraintWeights_7:float=0.0385
    ConstraintWeights_8:float=0.2448
    ConstraintWeights_9:float=0.0101
   ) -> (Pose) {
    alphaBoolBlend {blendOption=Linear}
    boneToModify {boneName=dyn_jacketFront_mstr_l}
    constraintSetup [
     {targetBone {boneName=deform_thigh_fr_l}}
     {targetBone {boneName=dyn_jacketFront_02_l}}
     {targetBone {boneName=dyn_jacketFront_01_l}}
     {targetBone {boneName=spine_01}}
     {targetBone {boneName=spine_02}}
     {targetBone {boneName=spine_03}}
     {targetBone {boneName=spine_04}}
     {targetBone {boneName=dyn_jacketSide_01_l}}
     {targetBone {boneName=pelvis}}
     {targetBone {boneName=thigh_twist_01_l}}
    ]
    constraintWeights [1,1,1,1,1,1,1,1,1,1]
   }
   n91 = "Constraint - dyn_jacketFront_mstr_r (PPPPPPPPPP)"!(
    ComponentPose=@n90
    Alpha=1
    ConstraintWeights_0=0.0406
    ConstraintWeights_1:float=0.015
    ConstraintWeights_2:float=0.3136
    ConstraintWeights_3:float=0.1819
    ConstraintWeights_4:float=0.0707
    ConstraintWeights_5:float=0.0489
    ConstraintWeights_6:float=0.0199
    ConstraintWeights_7:float=0.2801
    ConstraintWeights_8:float=0.0021
    ConstraintWeights_9:float=0.0273
   ) -> (Pose) {
    alphaBoolBlend {blendOption=Linear}
    boneToModify {boneName=dyn_jacketFront_mstr_r}
    constraintSetup [
     {targetBone {boneName=dyn_jacketSide_01_r}}
     {targetBone {boneName=thigh_twist_01_r}}
     {targetBone {boneName=dyn_jacketFront_01_r}}
     {targetBone {boneName=spine_01}}
     {targetBone {boneName=spine_02}}
     {targetBone {boneName=spine_03}}
     {targetBone {boneName=spine_04}}
     {targetBone {boneName=pelvis}}
     {targetBone {boneName=thigh_twist_01_l}}
     {targetBone {boneName=deform_thigh_fr_r}}
    ]
    constraintWeights [1,1,1,1,1,1,1,1,1,1]
   }
   n92 = Get thigh_rot_l()
   n93 = "Transform (Modify) Bone - Bone: dyn_jacketFront_mstr_l"(ComponentPose=@n91, Alpha=@n92, Translation="-10,0,10", Rotation="0,10,-55.000", Scale="1,1,1") {
    rotationMode=BMM_Additive
    rotationSpace=BCS_BoneSpace
    scaleMode=BMM_Additive
    scaleSpace=BCS_BoneSpace
    translationMode=BMM_Additive
    translationSpace=BCS_BoneSpace
    alphaBoolBlend {blendOption=Linear}
    boneToModify {boneName=dyn_jacketFront_mstr_l}
    rotation {roll=-55, yaw=10}
    translation {x=-10, z=10}
   }
   n94 = "Transform (Modify) Bone - Bone: dyn_jacketSide_mstr_l"(ComponentPose=@n93, Alpha=@n92, Translation="-2,-2,0", Rotation="10,5,0", Scale="1,1,1") {
    rotationMode=BMM_Additive
    rotationSpace=BCS_BoneSpace
    scaleMode=BMM_Additive
    scaleSpace=BCS_BoneSpace
    translationMode=BMM_Additive
    translationSpace=BCS_BoneSpace
    alphaBoolBlend {blendOption=Linear}
    boneToModify {boneName=dyn_jacketSide_mstr_l}
    rotation {pitch=10, yaw=5}
    translation=<-2 -2>
   }
   n95 = Get thigh_rot_r()
   n96 = "Transform (Modify) Bone - Bone: dyn_jacketFront_mstr_r"(ComponentPose=@n94, Alpha=@n95, Translation="-7,0,10", Rotation="0.000,0.000,50", Scale="1,1,1") {
    rotationMode=BMM_Additive
    rotationSpace=BCS_BoneSpace
    scaleMode=BMM_Additive
    scaleSpace=BCS_BoneSpace
    translationMode=BMM_Additive
    translationSpace=BCS_BoneSpace
    alphaBoolBlend {blendOption=Linear}
    boneToModify {boneName=dyn_jacketFront_mstr_r}
    rotation {roll=50}
    translation {x=-7, z=10}
   }
   n97 = "Transform (Modify) Bone - Bone: dyn_jacketSide_mstr_r"(ComponentPose=@n96, Alpha=@n95, Translation="-2,2,0", Rotation="10,5,0", Scale="1,1,1") {
    rotationMode=BMM_Additive
    rotationSpace=BCS_BoneSpace
    scaleMode=BMM_Additive
    scaleSpace=BCS_BoneSpace
    translationMode=BMM_Additive
    translationSpace=BCS_BoneSpace
    alphaBoolBlend {blendOption=Linear}
    boneToModify {boneName=dyn_jacketSide_mstr_r}
    rotation {pitch=10, yaw=5}
    translation=<-2 2>
   }
   n98 = "Constraint - dyn_zipper_MSTR (PPPPPP)"!(
    ComponentPose=@n97
    Alpha=1
    ConstraintWeights_0=0.0013
    ConstraintWeights_1:float=0.0029
    ConstraintWeights_2:float=0.5713
    ConstraintWeights_3:float=0.0002
    ConstraintWeights_4:float=0.2146
    ConstraintWeights_5:float=0.2098
   ) -> (Pose) {
    alphaBoolBlend {blendOption=Linear}
    boneToModify {boneName=dyn_zipper_MSTR}
    constraintSetup [
     {targetBone {boneName=clavicle_l}}
     {targetBone {boneName=spine_04}}
     {targetBone {boneName=spine_05}}
     {targetBone {boneName=clavicle_r}}
     {targetBone {boneName=neck_02}}
     {targetBone {boneName=neck_01}}
    ]
    constraintWeights [1,1,1,1,1,1]
   }
   n99 = "Constraint - dyn_boot_l_MSTR (PPPP)"!(
    ComponentPose=@n98
    Alpha=1
    ConstraintWeights_0=0.0129
    ConstraintWeights_1:float=0.5734
    ConstraintWeights_2:float=0.1423
    ConstraintWeights_3:float=0.2714
   ) -> (Pose) {
    alphaBoolBlend {blendOption=Linear}
    boneToModify {boneName=dyn_boot_l_MSTR}
    constraintSetup [
     {targetBone {boneName=calf_l}}
     {targetBone {boneName=calf_twist_01_l}}
     {targetBone {boneName=foot_l}}
     {targetBone {boneName=calf_twist_02_l}}
    ]
    constraintWeights [1,1,1,1]
   }
   n100 = "Constraint - dyn_boot_r_MSTR (PPPP)"!(
    ComponentPose=@n99
    Alpha=1
    ConstraintWeights_0=0.2543
    ConstraintWeights_1:float=0.5804
    ConstraintWeights_2:float=0.154
    ConstraintWeights_3:float=0.0113
   ) -> (Pose) {
    alphaBoolBlend {blendOption=Linear}
    boneToModify {boneName=dyn_boot_r_MSTR}
    constraintSetup [
     {targetBone {boneName=calf_twist_02_r}}
     {targetBone {boneName=calf_twist_01_r}}
     {targetBone {boneName=foot_r}}
     {targetBone {boneName=calf_r}}
    ]
    constraintWeights [1,1,1,1]
   }
   n101 = "Constraint - dyn_collar_l_MSTR (PPPPPPPP)"!(
    ComponentPose=@n100
    Alpha=1
    ConstraintWeights_0=0.0507
    ConstraintWeights_1:float=0.0042
    ConstraintWeights_2:float=0.0024
    ConstraintWeights_3:float=0.0248
    ConstraintWeights_4:float=0.8475
    ConstraintWeights_5:float=0.0023
    ConstraintWeights_6:float=0.0304
    ConstraintWeights_7:float=0.0376
   ) -> (Pose) {
    alphaBoolBlend {blendOption=Linear}
    boneToModify {boneName=dyn_collar_l_MSTR}
    constraintSetup [
     {targetBone {boneName=clavicle_l}}
     {targetBone {boneName=deform_pectoral_fr_l}}
     {targetBone {boneName=spine_03}}
     {targetBone {boneName=spine_04}}
     {targetBone {boneName=spine_05}}
     {targetBone {boneName=clavicle_r}}
     {targetBone {boneName=neck_02}}
     {targetBone {boneName=neck_01}}
    ]
    constraintWeights [1,1,1,1,1,1,1,1]
   }
   n102 = "Constraint - dyn_collar_r_MSTR (PPPPPPPP)"!(
    ComponentPose=@n101
    Alpha=1
    ConstraintWeights_0=0.0061
    ConstraintWeights_1:float=0.0014
    ConstraintWeights_2:float=0.0032
    ConstraintWeights_3:float=0.0269
    ConstraintWeights_4:float=0.8524
    ConstraintWeights_5:float=0.0291
    ConstraintWeights_6:float=0.0377
    ConstraintWeights_7:float=0.0432
   ) -> (Pose) {
    alphaBoolBlend {blendOption=Linear}
    boneToModify {boneName=dyn_collar_r_MSTR}
    constraintSetup [
     {targetBone {boneName=clavicle_l}}
     {targetBone {boneName=deform_pectoral_fr_r}}
     {targetBone {boneName=spine_03}}
     {targetBone {boneName=spine_04}}
     {targetBone {boneName=spine_05}}
     {targetBone {boneName=clavicle_r}}
     {targetBone {boneName=neck_02}}
     {targetBone {boneName=neck_01}}
    ]
    constraintWeights [1,1,1,1,1,1,1,1]
   }
   n103 = "Constraint - dyn_collar_r_MSTR1 (PPPPPPPPPP)"!(
    ComponentPose=@n102
    Alpha=1
    ConstraintWeights_0=0.0196
    ConstraintWeights_1:float=0.0831
    ConstraintWeights_2:float=0.0255
    ConstraintWeights_3:float=0.0154
    ConstraintWeights_4:float=0.0368
    ConstraintWeights_5:float=0.444
    ConstraintWeights_6:float=0.1629
    ConstraintWeights_7:float=0.1544
    ConstraintWeights_8:float=0.0503
    ConstraintWeights_9:float=0.0079
   ) -> (Pose) {
    alphaBoolBlend {blendOption=Linear}
    boneToModify {boneName=dyn_collar_r_MSTR1}
    constraintSetup [
     {targetBone {boneName=upperarm_twist_01_r}}
     {targetBone {boneName=deform_pectoral_fr_r}}
     {targetBone {boneName=deform_deltoid_fr_r}}
     {targetBone {boneName=dyn_collar_r}}
     {targetBone {boneName=spine_04}}
     {targetBone {boneName=spine_05}}
     {targetBone {boneName=clavicle_r}}
     {targetBone {boneName=neck_02}}
     {targetBone {boneName=neck_01}}
     {targetBone {boneName=dyn_collar_bk_01}}
    ]
    constraintWeights [1,1,1,1,1,1,1,1,1,1]
   }
   n104 = "Constraint - dyn_collar_l_MSTR1 (PPPPPPPP)"!(
    ComponentPose=@n103
    Alpha=1
    ConstraintWeights_0=0.0922
    ConstraintWeights_1:float=0.0432
    ConstraintWeights_2:float=0.1006
    ConstraintWeights_3:float=0.0307
    ConstraintWeights_4:float=0.419
    ConstraintWeights_5:float=0.0323
    ConstraintWeights_6:float=0.1845
    ConstraintWeights_7:float=0.0976
   ) -> (Pose) {
    alphaBoolBlend {blendOption=Linear}
    boneToModify {boneName=dyn_collar_l_MSTR1}
    constraintSetup [
     {targetBone {boneName=clavicle_l}}
     {targetBone {boneName=deform_deltoid_fr_l}}
     {targetBone {boneName=deform_pectoral_fr_l}}
     {targetBone {boneName=spine_04}}
     {targetBone {boneName=spine_05}}
     {targetBone {boneName=upperarm_twist_01_l}}
     {targetBone {boneName=neck_02}}
     {targetBone {boneName=neck_01}}
    ]
    constraintWeights [1,1,1,1,1,1,1,1]
   }
   n105 = "Constraint - dyn_collar_bk_MSTR (PPPPPPP)"!(
    ComponentPose=@n104
    Alpha=1
    ConstraintWeights_0=0.0342
    ConstraintWeights_1:float=0.0096
    ConstraintWeights_2:float=0.0164
    ConstraintWeights_3:float=0.3585
    ConstraintWeights_4:float=0.0457
    ConstraintWeights_5:float=0.3579
    ConstraintWeights_6:float=0.1777
   ) -> (Pose) {
    alphaBoolBlend {blendOption=Linear}
    boneToModify {boneName=dyn_collar_bk_MSTR}
    constraintSetup [
     {targetBone {boneName=clavicle_l}}
     {targetBone {boneName=spine_03}}
     {targetBone {boneName=spine_04}}
     {targetBone {boneName=spine_05}}
     {targetBone {boneName=clavicle_r}}
     {targetBone {boneName=neck_02}}
     {targetBone {boneName=neck_01}}
    ]
    constraintWeights [1,1,1,1,1,1,1]
   }
   n106 = "Constraint - dyn_armband_MSTR (PPPPPPPP)"!(
    ComponentPose=@n105
    Alpha=1
    ConstraintWeights_0=0.0197
    ConstraintWeights_1:float=0.076
    ConstraintWeights_2:float=0.4332
    ConstraintWeights_3:float=0.0572
    ConstraintWeights_4:float=0.0217
    ConstraintWeights_5:float=0.0728
    ConstraintWeights_6:float=0.3181
    ConstraintWeights_7:float=0.0012
   ) -> (Pose) {
    alphaBoolBlend {blendOption=Linear}
    boneToModify {boneName=dyn_armband_MSTR}
    constraintSetup [
     {targetBone {boneName=lowerarm_twist_02_r}}
     {targetBone {boneName=deform_elbow_bk_r}}
     {targetBone {boneName=upperarm_r}}
     {targetBone {boneName=deform_elbow_fr_r}}
     {targetBone {boneName=lowerarm_r}}
     {targetBone {boneName=upperarm_twist_01_r}}
     {targetBone {boneName=upperarm_twist_02_r}}
     {targetBone {boneName=clavicle_r}}
    ]
    constraintWeights [1,1,1,1,1,1,1,1]
   }
   n107 = "Constraint - dyn_sleeve_01_MSTR (PPPPPPPP)"!(
    ComponentPose=@n106
    Alpha=1
    ConstraintWeights_0=0.2113
    ConstraintWeights_1:float=0.0052
    ConstraintWeights_2:float=0.0107
    ConstraintWeights_3:float=0.0424
    ConstraintWeights_4:float=0.0593
    ConstraintWeights_5:float=0.335
    ConstraintWeights_6:float=0.2183
    ConstraintWeights_7:float=0.1179
   ) -> (Pose) {
    alphaBoolBlend {blendOption=Linear}
    boneToModify {boneName=dyn_sleeve_01_MSTR}
    constraintSetup [
     {targetBone {boneName=upperarm_twist_01_r}}
     {targetBone {boneName=upperarm_twist_02_r}}
     {targetBone {boneName=spine_01}}
     {targetBone {boneName=spine_02}}
     {targetBone {boneName=spine_03}}
     {targetBone {boneName=spine_04}}
     {targetBone {boneName=spine_05}}
     {targetBone {boneName=clavicle_r}}
    ]
    constraintWeights [1,1,1,1,1,1,1,1]
   }
   n108 = "Constraint - dyn_sleeve_02_MSTR (PPPPPPPP)"!(
    ComponentPose=@n107
    Alpha=1
    ConstraintWeights_0=0.2007
    ConstraintWeights_1:float=0.0304
    ConstraintWeights_2:float=0.0178
    ConstraintWeights_3:float=0.0442
    ConstraintWeights_4:float=0.0238
    ConstraintWeights_5:float=0.2347
    ConstraintWeights_6:float=0.2426
    ConstraintWeights_7:float=0.2057
   ) -> (Pose) {
    alphaBoolBlend {blendOption=Linear}
    boneToModify {boneName=dyn_sleeve_02_MSTR}
    constraintSetup [
     {targetBone {boneName=upperarm_twist_01_r}}
     {targetBone {boneName=deform_pectoral_fr_r}}
     {targetBone {boneName=upperarm_twist_02_r}}
     {targetBone {boneName=deform_deltoid_fr_r}}
     {targetBone {boneName=spine_03}}
     {targetBone {boneName=spine_04}}
     {targetBone {boneName=spine_05}}
     {targetBone {boneName=clavicle_r}}
    ]
    constraintWeights [1,1,1,1,1,1,1,1]
   }
   n109 = "Constraint - dyn_sleeve_03_MSTR (PPPPPPPP)"!(
    ComponentPose=@n108
    Alpha=1
    ConstraintWeights_0=0.2964
    ConstraintWeights_1:float=0.0592
    ConstraintWeights_2:float=0.0325
    ConstraintWeights_3:float=0.0913
    ConstraintWeights_4:float=0.0102
    ConstraintWeights_5:float=0.121
    ConstraintWeights_6:float=0.1772
    ConstraintWeights_7:float=0.2123
   ) -> (Pose) {
    alphaBoolBlend {blendOption=Linear}
    boneToModify {boneName=dyn_sleeve_03_MSTR}
    constraintSetup [
     {targetBone {boneName=upperarm_twist_01_r}}
     {targetBone {boneName=deform_pectoral_fr_r}}
     {targetBone {boneName=upperarm_twist_02_r}}
     {targetBone {boneName=deform_deltoid_fr_r}}
     {targetBone {boneName=spine_03}}
     {targetBone {boneName=spine_04}}
     {targetBone {boneName=spine_05}}
     {targetBone {boneName=clavicle_r}}
    ]
    constraintWeights [1,1,1,1,1,1,1,1]
   }
   n110 = "Constraint - dyn_sleeve_04_MSTR (PPPPPPPP)"!(
    ComponentPose=@n109
    Alpha=1
    ConstraintWeights_0=0.0087
    ConstraintWeights_1:float=0.2365
    ConstraintWeights_2:float=0.1438
    ConstraintWeights_3:float=0.076
    ConstraintWeights_4:float=0.2647
    ConstraintWeights_5:float=0.0453
    ConstraintWeights_6:float=0.0711
    ConstraintWeights_7:float=0.1539
   ) -> (Pose) {
    alphaBoolBlend {blendOption=Linear}
    boneToModify {boneName=dyn_sleeve_04_MSTR}
    constraintSetup [
     {targetBone {boneName=upperarm_r}}
     {targetBone {boneName=upperarm_twist_01_r}}
     {targetBone {boneName=deform_pectoral_fr_r}}
     {targetBone {boneName=upperarm_twist_02_r}}
     {targetBone {boneName=deform_deltoid_fr_r}}
     {targetBone {boneName=spine_04}}
     {targetBone {boneName=spine_05}}
     {targetBone {boneName=clavicle_r}}
    ]
    constraintWeights [1,1,1,1,1,1,1,1]
   }
   n111 = Get bIsSkyDiving()
   n112 = "Transform (Modify) Bone - Bone: dyn_sleeve_01_MSTR"!(ComponentPose=@n110, bAlphaBoolEnabled:bool=@n111, Translation="0.000,-3,0.000", Rotation="0,0,0", Scale="1,1,1") -> (Pose) {
    alphaInputType=Bool
    translationMode=BMM_Additive
    alphaBoolBlend {blendInTime=0.1, blendOption=Linear, blendOutTime=0.1}
    boneToModify {boneName=dyn_sleeve_01_MSTR}
    translation {y=-3}
   }
   n113 = "Transform (Modify) Bone - Bone: dyn_sleeve_02_MSTR"!(ComponentPose=@n112, bAlphaBoolEnabled:bool=@n111, Translation="0.000,-2,0.000", Rotation="-6.355,53.683,-32.519", Scale="1,1,1") -> (Pose) {
    alphaInputType=Bool
    rotationMode=BMM_Additive
    translationMode=BMM_Additive
    alphaBoolBlend {blendInTime=0.1, blendOption=Linear, blendOutTime=0.1}
    boneToModify {boneName=dyn_sleeve_02_MSTR}
    rotation=<rot -6.355 53.683 -32.519>
    translation {y=-2}
   }
   n114 = "Transform (Modify) Bone - Bone: dyn_sleeve_03_MSTR"!(ComponentPose=@n113, bAlphaBoolEnabled:bool=@n111, Translation="0.000,-1.510,1.517", Rotation="-33.032,38.320,-6.824", Scale="1,1,1") -> (Pose) {
    alphaInputType=Bool
    rotationMode=BMM_Additive
    translationMode=BMM_Additive
    alphaBoolBlend {blendInTime=0.1, blendOption=Linear, blendOutTime=0.1}
    boneToModify {boneName=dyn_sleeve_03_MSTR}
    rotation=<rot -33.032 38.3199 -6.824>
    translation {y=-1.51, z=1.517}
   }
   n115 = "Transform (Modify) Bone - Bone: dyn_sleeve_04_MSTR"!(ComponentPose=@n114, bAlphaBoolEnabled:bool=@n111, Translation="0,0,0", Rotation="0,0,0", Scale="1,1,1") -> (Pose) {
    alphaInputType=Bool
    translationMode=BMM_Additive
    alphaBoolBlend {blendInTime=0.1, blendOption=Linear, blendOutTime=0.1}
    boneToModify {boneName=dyn_sleeve_04_MSTR}
   }
   Save cached pose 'prerigid'(Pose=Component To Local(ComponentPose=@n115)) {cacheName=prerigid}
  ]}
 EventGraph {nodes [
   Event BlueprintUpdateAnimation()
   > n1 = Sequence!(execute) -> (then_0, then_1, then_2:Exec, then_3:Exec, then_4:Exec, then_5:Exec, then_6:Exec)
   n2 = Get Skel_Mesh()
   > Set thigh_rot_l(thigh_rot_l=GetAlphaFromReferenceTransformDelta(SkeletalMeshComponent=@n2, BoneName=thigh_l, BaseName=pelvis, InputRange=(X=-20,Y=-115), TwistAxis="1,0,0"))
   > Set thigh_rot_r(thigh_rot_r=GetAlphaFromReferenceTransformDelta(SkeletalMeshComponent=@n2, BoneName=thigh_r, BaseName=pelvis, InputRange=(X=-20,Y=-100), TwistAxis="1,0,0"))
   n3 = GetOwnerSkeletalMeshComponent()
   n4 = BreakRotator(InRot=BreakTransform(InTransform=GetDeltaTransformFromRefPose(self=@n3, BoneName=thigh_r, BaseName=pelvis)).Rotation)
   > Set thigh_rot_up_r(thigh_rot_up_r:double=MapRangeClamped(Value=@n4.Yaw, InRangeA=-15, InRangeB=-100)) -> (Output_Get:double)
   > Set thigh_rot_inr_r(thigh_rot_inr_r:double=MapRangeClamped(Value=@n4.Pitch, InRangeA=0, InRangeB=-10, OutRangeA=1, OutRangeB=0)) -> (Output_Get:double)
   > Set thigh_rot_otr_r(thigh_rot_otr_r:double=MapRangeClamped(Value=@n4.Pitch, InRangeA=20, InRangeB=50)) -> (Output_Get:double)
   n5 = GetOwnerSkeletalMeshComponent()
   n6 = BreakRotator(InRot=BreakTransform(InTransform=GetDeltaTransformFromRefPose(self=@n5, BoneName=thigh_l, BaseName=pelvis)).Rotation)
   > Set thigh_rot_up_l(thigh_rot_up_l:double=MapRangeClamped(Value=@n6.Yaw, InRangeA=-10, InRangeB=-100)) -> (Output_Get:double)
   > Set thigh_rot_inr_l(thigh_rot_inr_l:double=MapRangeClamped(Value=@n6.Pitch, InRangeA=0, InRangeB=-10, OutRangeA=1, OutRangeB=0)) -> (Output_Get:double)
   > Set thigh_rot_otr_l(thigh_rot_otr_l:double=MapRangeClamped(Value=@n6.Pitch, InRangeA=20, InRangeB=50)) -> (Output_Get:double)
   > Set jumpingAlpha(jumpingAlpha:double="To Float (Boolean)"(InBool=Get bIsJumping())) -> (Output_Get:double)
   n7 = Get bIsMoving2D()
   > Set isNotCrouchRunning(isNotCrouchRunning:double="To Float (Boolean)"(InBool=NOT Boolean(A=@n7))) -> (Output_Get:double)
   > Set ParachuteAlpha(ParachuteAlpha:double="To Float (Boolean)"(InBool=Get bIsParachuteOpened())) -> (Output_Get:double)
   > Set skydiveAlpha(skydiveAlpha:double="To Float (Boolean)"(InBool=Get bIsSkyDiving())) -> (Output_Get:double)
   n8 = Break Fort Anim Input Slope Info!(FortAnimInput_SlopeInfo:"Fort Anim Input Slope Info Structure (by ref)"=Get SlopeInfoInput() -> (SlopeInfoInput:Fort Anim Input Slope Info Structure)) -> (RootSlopeTranslation:Vector=~, LeftFootSlopeTranslation:Vector=~, RightFootSlopeTranslation:Vector=~, SlopeRotation:Rotator=~) {is=title K2Node_BreakStruct}
   Set slope_root_trans(execute=@n1.then_1, slope_root_trans:Vector=@n8) -> (Output_Get:Vector)
   > Set slope_foot_rot(slope_foot_rot:Rotator=@n8.SlopeRotation) -> (Output_Get:Rotator)
   n9 = Gravity Override State Function DEPRECATED(
    execute=@n1.then_2
    CharacterPartAnimInstance=Self-Reference()
    GameTimeInSeconds=GetGameTimeInSeconds()
    GravityOverrideParams=Get Gravity Override Params()
   )
   > Set GravityOverride(GravityOverride=@n9)
   n10 = ComplexSimSpaceFunction DEPRECATED!(
    execute=@n1.then_3
    CharacterPartAnimInstance:Custom Character Part Anim Instance Object Reference=Self-Reference()
    InterpSpeedMax:double=9
    PawnSpeedMax:double=20
    RInterpErrorTolerance:double=1
    DeltaTime:double=0
    DebugDraw:bool=false
    sim_space_rotation_in:Rotator=Get sim_space_rotation() -> (sim_space_rotation:Rotator)
    inComponentSpaceIn:bool=Get InComponentSpace() -> (InComponentSpace:bool)
    inComponentSpaceLastFrameIn:bool=Get InComponentSpaceLastFrame() -> (InComponentSpaceLastFrame:bool)
    lockedOnSimSpaceIn:bool=Get LockedOnSimSpace() -> (LockedOnSimSpace:bool)
    InterpSpeedIn:double=Get InterpSpeed() -> (InterpSpeed:double)
   ) -> (
    then
    sim_space_rotation:Rotator
    inComponentSpaceOut:bool
    inComponentSpaceLastFrameOut:bool
    lockedOnSimSpaceOut:bool
    InterpSpeedOut:double
   )
   > Set sim_space_rotation(sim_space_rotation:Rotator=@n10) -> (Output_Get:Rotator)
   > Set InComponentSpace(InComponentSpace:bool=@n10.inComponentSpaceOut) -> (Output_Get:bool)
   > Set InComponentSpaceLastFrame(InComponentSpaceLastFrame:bool=@n10.inComponentSpaceLastFrameOut) -> (Output_Get:bool)
   > Set LockedOnSimSpace(LockedOnSimSpace:bool=@n10.lockedOnSimSpaceOut) -> (Output_Get:bool)
   > Set InterpSpeed(InterpSpeed:double=@n10.InterpSpeedOut) -> (Output_Get:double)
   n11 = PlayerMovement(execute=@n1.then_4, AnimInstance=Self-Reference(), XDotProductCoefficient=1, YDotProductCoefficient=1, ZDotProductCoefficient=1)
   n12 = MapRangeClamped(Value=@n11.YMovement, InRangeA=0, InRangeB=-1)
   n13 = float * float(A="Absolute (Float)"(A=@n11), B=0.5)
   > Set BwdMovement(BwdMovement:double=MapRangeClamped(
     Value=float + float(A=@n13, B=@n12) {is=title K2Node_CommutativeAssociativeBinaryOperator}
     InRangeA=0
     InRangeB=1.1
     OutRangeA=0.4
     OutRangeB=0
    )) -> (Output_Get:double)
   n14 = MapRangeClamped(Value=@n11, InRangeA=-0.2, InRangeB=1)
   n15 = float * float(A="Absolute (Float)"(A=@n11.YMovement), B=0.2)
   > Set RightMovement(RightMovement:double=MapRangeClamped(Value=float + float(A=@n15, B=@n14), InRangeA=0, InRangeB=1.2, OutRangeB=0.5)) -> (Output_Get:double)
   n16 = MapRangeClamped(Value=@n11, InRangeA=0.2, InRangeB=-1, OutRangeB=0.3)
   n17 = float * float(A="Absolute (Float)"(A=@n11.YMovement), B=0.2)
   > Set LeftMovement(LeftMovement:double=MapRangeClamped(Value=float + float(A=@n17, B=@n16), InRangeA=0, InRangeB=1.2, OutRangeB=0.5)) -> (Output_Get:double)
   n18 = Branch(execute=@n1.then_5, Condition=Get bIsSkyDiving())
   > n19 = Branch(Condition="float > float"(A=Get PawnVelocityZ(), B=50) {is=title K2Node_CallFunction})
   > Set BwdMovement(BwdMovement=float * float(A=Get BwdMovement(), B=0))
   Set BwdMovement(execute=[@n18.else, @n19.else], BwdMovement=float * float(A=Get BwdMovement(), B=1))
   n20 = Branch(execute=@n1.then_6, Condition=Get bIsTargeting())
   > Set targeting(targeting:double=0.5) -> (Output_Get:double)
   Set targeting(execute=@n20.else, targeting=1)
   Event BlueprintInitializeAnimation()
   > Set Skel_Mesh(Skel_Mesh=GetOwnerSkeletalMeshComponent())
   Get bIsCrouchSprinting()
  ]}
]
variables [
 BwdMovement {type=number}
 Gravity Override Params {
  type=GravityOverrideParamsStruct
  default {
   emoteOrMelee {gravityOverride=<0 0 -720>}
   front_End {gravityOverride=<0 0 -720>}
   onGround_Moving {gravityOverride=<0 0 -720>}
   onGround_Standing {gravityOverride=<0 0 -720>}
   skydive_DivingDown {gravityOverride=<0 0 0>}
   skydive_FloatingDown {gravityOverride=<0 0 -240>}
   skydive_GoingUp {gravityOverride=<0 -300 100>}
   skydive_Parachuting {gravityOverride=<0 0 -540>}
   swimming {gravityOverride=<0 0 0>}
  }
 }
 GravityOverride {type=Vector}
 InComponentSpace {type=boolean}
 InComponentSpaceLastFrame {type=boolean}
 InterpSpeed {type=number}
 isNotCrouchRunning {type=number}
 jumpingAlpha {type=number}
 LeftMovement {type=number}
 LockedOnSimSpace {type=boolean}
 ParachuteAlpha {type=number}
 RightMovement {type=number}
 sim_space_rotation {type=Rotator}
 Skel_Mesh {type=SkeletalMeshComponent}
 skydiveAlpha {type=number}
 slope_foot_rot {type=Rotator}
 slope_root_trans {type=Vector}
 targeting {type=number}
 thigh_rot_inr_l {type=number}
 thigh_rot_inr_r {type=number}
 thigh_rot_l {type=number}
 thigh_rot_otr_l {type=number}
 thigh_rot_otr_r {type=number}
 thigh_rot_r {type=number}
 thigh_rot_up_l {type=number}
 thigh_rot_up_r {type=number}
]
