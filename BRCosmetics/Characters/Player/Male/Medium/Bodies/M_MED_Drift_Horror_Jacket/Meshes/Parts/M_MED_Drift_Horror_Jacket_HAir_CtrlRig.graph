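# Control Rig for the hair part of the male Drift Horror Jacket character, adjusting dynamic hair bones as the head bends.
# It reads the Z Euler angle of head relative to neck_01.
# Two clamped remaps turn that angle into a uniform shrink toward half scale and a Y rotation toward 90 degrees.
# One ModifyTransforms node applies both to three bones, including dyn_hair_04 and dyn_hair_05, with no simulation in the graph.
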ControlRigBlueprint /BRCosmetics/Characters/Player/Male/Medium/Bodies/M_MED_Drift_Horror_Jacket/Meshes/Parts/M_MED_Drift_Horror_Jacket_HAir_CtrlRig
graphs [
 RigVMFunctionLibrary {links [], nodes []}
 RigVMModel {nodes [
   RigUnit_BeginExecution = &RigUnit_BeginExecution()
   n1 = &RigVMFunction_MathFloatRemap(
    SourceMaximum=-60
    SourceMinimum=-20
    TargetMaximum=0.5
    TargetMinimum=1
    bClamp=true
    Value=&RigVMFunction_MathQuaternionToEuler(
     RotationOrder=ZYX
     Value=&RigUnit_GetRelativeTransformForItem(Child=(Type=Bone,Name="head"), Parent=(Type=Bone,Name="neck_01")).RelativeTransform.Rotation
    ).Result.Z
   )
   n2 = &RigUnit_GetRelativeTransformForItem(Child=(Type=Bone,Name="head"), Parent=(Type=Bone,Name="neck_01"))
   n3 = &RigVMFunction_MathQuaternionFromEuler(
    Euler=(X=0,Y=45,Z=0)
    Euler.Y=&RigVMFunction_MathFloatRemap(
     SourceMaximum=-60
     SourceMinimum=-20
     TargetMaximum=90
     TargetMinimum=0
     bClamp=true
     Value=&RigVMFunction_MathQuaternionToEuler(RotationOrder=ZYX, Value=@n2.RelativeTransform.Rotation).Result.Z
    ).Result
   )
   &RigUnit_ModifyTransforms(
    ItemToModify="((Item=(Type=Bone,Name=\"dyn_hair_04\"),Transform=(Rotation=(X=0,Y=0,Z=0,W=1),Translation=(X=0,Y=0,Z=0),Scale3D=(X=1,Y=1,Z=1))),(Item=(Type=Bone,Name=\"dyn_hair_05\"),Transform=(Rotation=(X=0,Y=0,Z=0,W=1),Translation=(X=0,Y=0,Z=0),Scale3D=(X=1,Y=1,Z=1))),(Item ...(truncated)"
    Weight=1
    ItemToModify.0.Transform.Scale3D.X=@n1.Result
    ItemToModify.0.Transform.Scale3D.Y=@n1.Result
    ItemToModify.0.Transform.Scale3D.Z=@n1.Result
    ItemToModify.1.Transform.Scale3D.X=@n1.Result
    ItemToModify.1.Transform.Scale3D.Y=@n1.Result
    ItemToModify.1.Transform.Scale3D.Z=@n1.Result
    ItemToModify.2.Transform.Scale3D.X=@n1.Result
    ItemToModify.2.Transform.Scale3D.Y=@n1.Result
    ItemToModify.2.Transform.Scale3D.Z=@n1.Result
    ItemToModify.0.Transform.Rotation=@n3.Result
    ItemToModify.1.Transform.Rotation=@n3.Result
    ItemToModify.2.Transform.Rotation=@n3.Result
    ExecutePin=@RigUnit_BeginExecution.ExecutePin
   )
  ]}
]
hierarchy [
 root {loc=<0 0 0>, rot=<0 0 0 1>}
 pelvis {parent=root, loc=<0 -0.7071 93.1191>, rot=<0.0127 0.707 -0.0127 -0.707>}
 spine_01 {parent=pelvis, loc=<6.7494 -0.6388 0>, rot=<0 0 0.034 -0.9994>}
 spine_02 {parent=spine_01, loc=<7.2518 -0.035 0>, rot=<0 0 0.0238 -0.9997>}
 spine_03 {parent=spine_02, loc=<6.0104 0 0>, rot=<0 0 0.0149 -0.9999>}
 spine_04 {parent=spine_03, loc=<8.1024 0 0>, rot=<0 0 -0.0132 -0.9999>}
 spine_05 {parent=spine_04, loc=<10.0146 0 0>, rot=<0 0 -0.0015 -1>}
 clavicle_l {parent=spine_05, loc=<4.3575 -0.1466 -2.3223>, rot=<0.0465 -0.7289 -0.009 -0.6829>}
 clavicle_r {parent=spine_05, loc=<4.3575 -0.1466 2.3223>, rot=<0.7289 0.0465 -0.6829 0.009>}
 neck_01 {parent=spine_05, loc=<11.3138 0.1529 0>, rot=<0 0 -0.2003 -0.9797>}
 neck_02 {parent=neck_01, loc=<4.5627 0 0>, rot=<0 0 0.0504 -0.9987>}
 head {parent=neck_02, loc=<5.3914 0 0>, rot=<0 0 0.115 -0.9934>}
 faceAttach {parent=head, loc=<0.5101 -0.8087 0>, rot=<0.0082 -0.7071 -0.0082 -0.7071>, scale=<0.89 0.89 0.89>}
 C_brow_mid {parent=faceAttach, loc=<-0.0019 13.4115 5.3811>, rot=<0 0 0 -1>}
 L_brow_mid {parent=faceAttach, loc=<4.4655 12.3636 5.9824>, rot=<-0.0001 -0.0042 0.2553 -0.9669>}
 L_brow_outer {parent=faceAttach, loc=<6.4616 9.461 7.4644>, rot=<-0.0286 -0.04 0.4491 -0.8921>}
 L_cheek_inner {parent=faceAttach, loc=<5.704 10.3429 1.6218>, rot=<-0.0706 -0.2697 0.3184 -0.906>}
 C_lip_upper_mid {parent=faceAttach, loc=<0.0024 13.8375 -2.2788>, rot=<0.2225 0.0013 -0.0005 -0.9749>}
 L_lip_upper_outer {parent=faceAttach, loc=<2.0564 12.5052 -2.4039>, rot=<0.0359 -0.0728 0.2627 -0.9615>}
 L_lip_corner {parent=faceAttach, loc=<5.5578 8.4507 0.6355>, rot=<0.0029 0.008 -0.0146 -0.9999>}
 C_nose_bridge {parent=faceAttach, loc=<-0.0198 14.1882 2.5621>, rot=<-0.2218 0 0 -0.9751>}
 R_brow_mid {parent=faceAttach, loc=<-4.4607 12.3554 5.9824>, rot=<-0.0001 0.0042 -0.2553 -0.9669>}
 R_brow_outer {parent=faceAttach, loc=<-6.4429 9.4478 7.4653>, rot=<-0.0287 0.04 -0.4491 -0.8921>}
 R_cheek_inner {parent=faceAttach, loc=<-5.7037 10.3429 1.6221>, rot=<-0.0706 0.2697 -0.3184 -0.906>}
 R_lip_corner {parent=faceAttach, loc=<-5.5578 8.4506 0.6355>, rot=<0.0029 -0.008 0.0146 -0.9999>}
 C_jaw {parent=faceAttach, loc=<-0.0151 5.6976 2.1537>, rot=<0 0 0 1>}
 L_lip_lower_outer {parent=C_jaw, loc=<3.5313 4.2847 -4.9128>, rot=<0.013 0.0445 0.3056 -0.951>}
 C_lip_lower_mid {parent=C_jaw, loc=<0.0251 7.0256 -7.2797>, rot=<0 0 0 -1>}
 R_lip_lower_outer {parent=C_jaw, loc=<-3.5013 4.2849 -4.9129>, rot=<0.013 -0.0445 -0.3057 -0.951>}
 tongue {parent=C_jaw, loc=<0.0158 0.3321 -6.5416>, rot=<-0.1568 0 0 -0.9876>}
 R_eye {parent=faceAttach, loc=<-4.1979 8.9958 4.7341>, rot=<0 0 0.9998 -0.0187>}
 R_eye_lid_upper_mid {parent=R_eye, loc=<-0.0396 -1.1712 0.2636>, rot=<0.0523 -0.0009 0.0167 -0.9985>}
 R_eye_lid_lower_mid {parent=R_eye, loc=<-0.0085 -0.9422 -0.7046>, rot=<0.0329 -0.0005 0.0167 -0.9993>}
 L_eye {parent=faceAttach, loc=<4.1979 8.9958 4.7341>, rot=<0 0 0.9998 0.0187>}
 L_eye_lid_upper_mid {parent=L_eye, loc=<0.0251 -1.1623 0.2388>, rot=<0.0523 0 0 -0.9986>}
 L_eye_lid_lower_mid {parent=L_eye, loc=<0.0223 -0.923 -0.7488>, rot=<0.0329 0 0 -0.9995>}
 teeth_upper {parent=faceAttach, loc=<0 12.8197 -2.5289>, rot=<0 0 0 1>}
 R_lip_upper_outer {parent=faceAttach, loc=<-2.0564 12.5049 -2.4041>, rot=<0.0359 0.0728 -0.2627 -0.9615>}
 dyn_ear_r {parent=head, loc=<7.4265 3.1122 6.9822>, rot=<-0.0009 0.2068 0.9784 -0.004>}
 dyn_ear_l {parent=head, loc=<7.4264 3.1122 -6.9822>, rot=<0.0039 -0.9424 -0.3344 0.0014>}
 dyn_hair_05 {parent=head, loc=<7.4281 -6.2577 -1.54>, rot=<0.0029 -0.7071 0.7071 -0.0029>}
 dyn_hair_04 {parent=head, loc=<9.4686 -7.32 -0.4818>, rot=<0.0029 -0.7071 0.7071 -0.0029>}
 dyn_hair_03 {parent=head, loc=<12.8778 -5.5686 2.462>, rot=<0.3495 -0.6147 0.6147 -0.3495>}
 dyn_hair_02 {parent=head, loc=<14.6591 -5.3398 -0.3594>, rot=<0.3495 -0.6147 0.6147 -0.3495>}
 dyn_hair_01 {parent=head, loc=<16.5418 0.3425 -1.8135>, rot=<0.5021 -0.4979 0.4979 -0.5021>}
]
